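# Control Rig for the M_MED_DJ_Remix male body cosmetic that applies procedural corrective offsets to its collar bones as the arms and head move.
# Upper-arm angles relative to the clavicles and head angles relative to spine_05 become clamped remap weights for additive local bone modifiers.
# The left arm has two stacked angle ranges, the right one; side head correctives take the maximum of two axes.
# Shorts and lapel bones are not driven.
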ControlRigBlueprint /BRCosmetics/Characters/Player/Male/Medium/Bodies/M_MED_DJ_Remix/Meshes/M_MED_DJ_Remix_ControlRig
graphs [
 RigVMFunctionLibrary {links [], nodes []}
 RigVMModel {nodes [
   n1 = &RigUnit_BeginExecution()
   n2 = &RigUnit_GetRelativeBoneTransform(Bone=upperarm_r, Space=clavicle_r)
   /Script/ControlRig.RigUnit_ModifyBoneTransforms(
    BoneToModify=((Bone="dyn_parent_collar_r",Transform=(Rotation=(X=0,Y=0,Z=0,W=1),Translation=(X=0,Y=0,Z=0),Scale3D=(X=1,Y=1,Z=1))))
    Mode=AdditiveLocal
    WeightMaximum=1
    WeightMinimum=0
    ExecutePin=@n1.ExecutePin
    Weight=&RigVMFunction_MathFloatRemap(
     SourceMaximum=120
     SourceMinimum=80
     TargetMinimum=0
     bClamp=true
     Value=&RigVMFunction_MathQuaternionToEuler(RotationOrder=ZXY, Value=@n2.Transform.Rotation).Result.Z
    ).Result
    BoneToModify.0.Transform.Rotation=&RigVMFunction_MathQuaternionFromEuler(Euler=(X=50,Y=5,Z=0)).Result
   ) {is=struct RigVMUnitNode}
   n3 = &RigUnit_GetRelativeBoneTransform(Bone=upperarm_l, Space=clavicle_l)
   > &RigUnit_ModifyBoneTransforms(
    BoneToModify=((Bone="dyn_parent_collar_l",Transform=(Rotation=(X=0,Y=0,Z=0,W=1),Translation=(X=0,Y=0,Z=0),Scale3D=(X=1,Y=1,Z=1))),(Bone="dyn_parent_collar_bck_l",Transform=(Rotation=(X=0,Y=0,Z=0,W=1),Translation=(X=0,Y=0,Z=0),Scale3D=(X=1,Y=1,Z=1))))
    Mode=AdditiveLocal
    WeightMaximum=1
    WeightMinimum=0
    Weight=&RigVMFunction_MathFloatRemap(
     SourceMaximum=120
     SourceMinimum=80
     TargetMinimum=0
     bClamp=true
     Value=&RigVMFunction_MathQuaternionToEuler(RotationOrder=ZXY, Value=@n3.Transform.Rotation).Result.Z
    ).Result
    BoneToModify.0.Transform.Rotation=&RigVMFunction_MathQuaternionFromEuler(Euler=(X=10,Y=-20,Z=0)).Result
    BoneToModify.1.Transform.Rotation=&RigVMFunction_MathQuaternionFromEuler(Euler=(X=10,Y=-20,Z=0)).Result
   )
   n4 = &RigUnit_GetRelativeBoneTransform(Bone=upperarm_l, Space=clavicle_l)
   > &RigUnit_ModifyBoneTransforms(
    BoneToModify=((Bone="dyn_parent_collar_l",Transform=(Rotation=(X=0,Y=0,Z=0,W=1),Translation=(X=0,Y=0,Z=0),Scale3D=(X=1,Y=1,Z=1))),(Bone="dyn_parent_collar_bck_l",Transform=(Rotation=(X=0,Y=0,Z=0,W=1),Translation=(X=0,Y=0,Z=0),Scale3D=(X=1,Y=1,Z=1))))
    Mode=AdditiveLocal
    WeightMaximum=1
    WeightMinimum=0
    Weight=&RigVMFunction_MathFloatRemap(
     SourceMaximum=70
     SourceMinimum=50
     TargetMinimum=0
     bClamp=true
     Value=&RigVMFunction_MathQuaternionToEuler(RotationOrder=ZXY, Value=@n4.Transform.Rotation).Result.Z
    ).Result
    BoneToModify.1.Transform.Rotation=&RigVMFunction_MathQuaternionFromEuler(Euler=(X=10,Y=-10,Z=0)).Result
    BoneToModify.0.Transform.Rotation=&RigVMFunction_MathQuaternionFromEuler(Euler=(X=10,Y=-10,Z=0)).Result
   )
   n5 = &RigUnit_BeginExecution()
   n6 = &RigUnit_GetRelativeBoneTransform(Bone=head, Space=spine_05)
   &RigUnit_ModifyBoneTransforms(
    BoneToModify="((Bone=\"dyn_parent_collar_bck\",Transform=(Rotation=(X=0,Y=0,Z=0,W=1),Translation=(X=0,Y=-2.5,Z=-2.5),Scale3D=(X=1,Y=1,Z=1))),(Bone=\"dyn_parent_collar_bck_r\",Transform=(Rotation=(X=0,Y=0,Z=0,W=1),Translation=(X=0,Y=-1.25,Z=-1.25),Scale3D=(X=1,Y=1,Z=1))),(Bo ...(truncated)"
    Mode=AdditiveLocal
    WeightMaximum=1
    WeightMinimum=0
    ExecutePin=@n5.ExecutePin
    Weight=&RigVMFunction_MathFloatRemap(
     SourceMaximum=-43
     SourceMinimum=0
     TargetMinimum=0
     bClamp=true
     Value=&RigVMFunction_MathQuaternionToEuler(RotationOrder=XZY, Value=@n6.Transform.Rotation).Result.Z
    ).Result
    BoneToModify.0.Transform.Rotation=&RigVMFunction_MathQuaternionFromEuler(Euler=(X=35,Y=0,Z=0)).Result
    BoneToModify.1.Transform.Rotation=&RigVMFunction_MathQuaternionFromEuler(Euler=(X=25,Y=0,Z=0)).Result
    BoneToModify.2.Transform.Rotation=&RigVMFunction_MathQuaternionFromEuler(Euler=(X=25,Y=0,Z=0)).Result
   )
   n7 = &RigVMFunction_MathQuaternionToEuler(RotationOrder=XYZ, Value=&RigUnit_GetRelativeBoneTransform(Bone=head, Space=spine_05).Transform.Rotation)
   > &RigUnit_ModifyBoneTransforms(
    BoneToModify="((Bone=\"dyn_collar_01_l\",Transform=(Rotation=(X=0,Y=0,Z=0,W=1),Translation=(X=0,Y=0,Z=0),Scale3D=(X=1,Y=1,Z=1))),(Bone=\"dyn_collar_01_bck_l\",Transform=(Rotation=(X=0,Y=0,Z=0,W=1),Translation=(X=0,Y=-3,Z=-3),Scale3D=(X=1,Y=1,Z=1))),(Bone=\"dyn_parent_collar_ ...(truncated)"
    Mode=AdditiveLocal
    WeightMaximum=1
    WeightMinimum=0
    Weight=/Script/RigVM.RigVMFunction_MathFloatMax(
     B=&RigVMFunction_MathFloatRemap(SourceMaximum=55, SourceMinimum=0, TargetMinimum=0, bClamp=true, Value=@n7.Result.Y).Result
     A=&RigVMFunction_MathFloatRemap(SourceMaximum=-55, SourceMinimum=0, TargetMinimum=0, bClamp=true, Value=@n7.Result.X).Result
    ) {is=struct RigVMUnitNode, template="Maximum::Execute(in A,in B,out Result)"}.Result
    BoneToModify.0.Transform.Rotation=&RigVMFunction_MathQuaternionFromEuler(Euler=(X=0,Y=0,Z=30)).Result
    BoneToModify.1.Transform.Rotation=&RigVMFunction_MathQuaternionFromEuler(Euler=(X=0,Y=0,Z=-30)).Result
    BoneToModify.2.Transform.Rotation=&RigVMFunction_MathQuaternionFromEuler(Euler=(X=20,Y=0,Z=0)).Result
   )
   n8 = &RigVMFunction_MathQuaternionToEuler(RotationOrder=XYZ, Value=&RigUnit_GetRelativeBoneTransform(Bone=head, Space=spine_05).Transform.Rotation)
   > &RigUnit_ModifyBoneTransforms(
    BoneToModify="((Bone=\"dyn_collar_01_r\",Transform=(Rotation=(X=0,Y=0,Z=0,W=1),Translation=(X=0,Y=0,Z=0),Scale3D=(X=1,Y=1,Z=1))),(Bone=\"dyn_collar_01_bck_r\",Transform=(Rotation=(X=0,Y=0,Z=0,W=1),Translation=(X=0,Y=-3,Z=-3),Scale3D=(X=1,Y=1,Z=1))),(Bone=\"dyn_parent_collar_ ...(truncated)"
    Mode=AdditiveLocal
    WeightMaximum=1
    WeightMinimum=0
    Weight=&RigVMFunction_MathFloatMax(
     B=&RigVMFunction_MathFloatRemap(SourceMaximum=-50, SourceMinimum=0, TargetMinimum=0, bClamp=true, Value=@n8.Result.Y).Result
     A=&RigVMFunction_MathFloatRemap(SourceMaximum=60, SourceMinimum=25, TargetMinimum=0, bClamp=true, Value=@n8.Result.X).Result
    ) {template="Maximum::Execute(in A,in B,out Result)"}.Result
    BoneToModify.0.Transform.Rotation=&RigVMFunction_MathQuaternionFromEuler(Euler=(X=0,Y=0,Z=30)).Result
    BoneToModify.1.Transform.Rotation=&RigVMFunction_MathQuaternionFromEuler(Euler=(X=0,Y=0,Z=-45)).Result
    BoneToModify.2.Transform.Rotation=&RigVMFunction_MathQuaternionFromEuler(Euler=(X=20,Y=0,Z=0)).Result
   )
  ]}
]
hierarchy [
 root {loc=<0 0 0>, rot=<0 0 0 1>}
 pelvis {parent=root, loc=<0 -0.7071 93.1191>, rot=<0.0127 0.707 -0.0127 -0.707>}
 spine_01 {parent=pelvis, loc=<6.7494 -0.6388 0>, rot=<0 0 0.034 -0.9994>}
 spine_02 {parent=spine_01, loc=<7.2518 -0.035 0>, rot=<0 0 0.0238 -0.9997>}
 spine_03 {parent=spine_02, loc=<6.0104 0 0>, rot=<0 0 0.0149 -0.9999>}
 spine_04 {parent=spine_03, loc=<8.1024 0 0>, rot=<0 0 -0.0132 -0.9999>}
 spine_05 {parent=spine_04, loc=<10.0146 0 0>, rot=<0 0 -0.0015 -1>}
 clavicle_l {parent=spine_05, loc=<4.3575 -0.1466 -2.3223>, rot=<0.0465 -0.7289 -0.009 -0.6829>}
 upperarm_l {parent=clavicle_l, loc=<15.5111 0 0>, rot=<0.0226 -0.3234 -0.0088 -0.9459>}
 lowerarm_l {parent=upperarm_l, loc=<26.6514 0 0>, rot=<0.0391 -0.0099 -0.2453 -0.9686>}
 hand_l {parent=lowerarm_l, loc=<22.3491 0 0>, rot=<-0.7081 -0.0507 0.0975 -0.6975>}
 index_metacarpal_l {parent=hand_l, loc=<1.6535 0.0183 -2.96>, rot=<0.0158 -0.0124 -0.0173 -0.9996>}
 index_01_l {parent=index_metacarpal_l, loc=<8.3008 0.0357 -0.0524>, rot=<-0.0044 0.0485 0.0924 -0.9945>}
 index_02_l {parent=index_01_l, loc=<3.3587 -0.3004 0.0452>, rot=<-0.002 0.0027 0.3225 -0.9466>}
 index_03_l {parent=index_02_l, loc=<2.5578 -0.123 -0.0252>, rot=<-0.002 0.0067 0.2558 -0.9667>}
 middle_metacarpal_l {parent=hand_l, loc=<2.1856 -0.1725 -0.5984>, rot=<0.006 0.0168 -0.0329 -0.9993>}
 middle_01_l {parent=middle_metacarpal_l, loc=<7.3275 -0.2447 0.0129>, rot=<-0.0208 0.0042 0.0858 -0.9961>}
 middle_02_l {parent=middle_01_l, loc=<4.0123 0.1401 0.1175>, rot=<-0.0218 0.0258 0.3512 -0.9357>}
 middle_03_l {parent=middle_02_l, loc=<2.6077 0.0534 0.1373>, rot=<0.0082 0.0188 0.1708 -0.9851>}
 pinky_metacarpal_l {parent=hand_l, loc=<1.7616 -0.9753 3.3814>, rot=<-0.0397 0.0814 -0.069 -0.9935>}
 pinky_01_l {parent=pinky_metacarpal_l, loc=<7.1709 -0.4831 -0.1094>, rot=<-0.0654 -0.0035 0.2242 -0.9723>}
 pinky_02_l {parent=pinky_01_l, loc=<2.8274 0.0893 -0.1043>, rot=<0.0037 0.0011 0.2854 -0.9584>}
 pinky_03_l {parent=pinky_02_l, loc=<2.2046 0.0307 -0.0078>, rot=<-0.0002 0.0098 0.1445 -0.9895>}
 ring_metacarpal_l {parent=hand_l, loc=<2.2074 -0.3625 1.3638>, rot=<-0.0284 0.0722 -0.0458 -0.9959>}
 ring_01_l {parent=ring_metacarpal_l, loc=<7.0295 -0.6197 -0.1446>, rot=<-0.0489 -0.0252 0.1892 -0.9804>}
 ring_02_l {parent=ring_01_l, loc=<3.6174 0.4647 -0.1027>, rot=<0 0 0.2806 -0.9598>}
 ring_03_l {parent=ring_02_l, loc=<2.4152 0.1141 0.0199>, rot=<0 0 0.1189 -0.9929>}
 thumb_01_l {parent=hand_l, loc=<1.3118 -0.3191 -2.459>, rot=<0.6546 -0.4042 0.0664 -0.6354>}
 thumb_02_l {parent=thumb_01_l, loc=<5.9336 -0.0448 0.0401>, rot=<-0.0037 0.0213 0.298 -0.9543>}
 thumb_03_l {parent=thumb_02_l, loc=<2.2223 0.3682 -0.0907>, rot=<0.0118 -0.013 0.0066 -0.9998>}
 lowerarm_twist_01_l {parent=lowerarm_l, loc=<17.8549 0 0>, rot=<0 0 0 1>}
 lowerarm_twist_02_l {parent=lowerarm_l, loc=<9.6199 0 0>, rot=<0 0 0 1>}
 upperarm_twist_01_l {parent=upperarm_l, loc=<0.0041 0 0>, rot=<0 0 0 -1>}
 upperarm_twist_02_l {parent=upperarm_l, loc=<13.7941 0 0>, rot=<0 0 0 -1>}
 clavicle_r {parent=spine_05, loc=<4.3575 -0.1466 2.3223>, rot=<0.7289 0.0465 -0.6829 0.009>}
 upperarm_r {parent=clavicle_r, loc=<-15.5111 0 0>, rot=<0.0226 -0.3234 -0.0088 -0.9459>}
 lowerarm_r {parent=upperarm_r, loc=<-26.6514 0 0>, rot=<0.0391 -0.0099 -0.2453 -0.9686>}
 hand_r {parent=lowerarm_r, loc=<-22.3492 0 0>, rot=<-0.7081 -0.0507 0.0975 -0.6975>}
 index_metacarpal_r {parent=hand_r, loc=<-1.6534 -0.0181 2.9601>, rot=<0.0158 -0.0124 -0.0173 -0.9996>}
 index_01_r {parent=index_metacarpal_r, loc=<-8.3008 -0.0357 0.0524>, rot=<-0.0044 0.0485 0.0924 -0.9945>}
 index_02_r {parent=index_01_r, loc=<-3.3589 0.3003 -0.0453>, rot=<-0.002 0.0027 0.3225 -0.9466>}
 index_03_r {parent=index_02_r, loc=<-2.5578 0.1231 0.0251>, rot=<-0.002 0.0067 0.2558 -0.9667>}
 middle_metacarpal_r {parent=hand_r, loc=<-2.1856 0.1726 0.5984>, rot=<0.006 0.0168 -0.0329 -0.9993>}
 middle_01_r {parent=middle_metacarpal_r, loc=<-7.3273 0.2447 -0.0129>, rot=<-0.0208 0.0042 0.0858 -0.9961>}
 middle_02_r {parent=middle_01_r, loc=<-4.0142 -0.1411 -0.07>, rot=<-0.0217 0.0259 0.3512 -0.9357>}
 middle_03_r {parent=middle_02_r, loc=<-2.6074 -0.0534 -0.1312>, rot=<0.0083 0.0188 0.1708 -0.9851>}
 pinky_metacarpal_r {parent=hand_r, loc=<-1.7617 0.9753 -3.3814>, rot=<-0.0397 0.0814 -0.069 -0.9935>}
 pinky_01_r {parent=pinky_metacarpal_r, loc=<-7.1707 0.4832 0.1094>, rot=<-0.0654 -0.0035 0.2242 -0.9723>}
 pinky_02_r {parent=pinky_01_r, loc=<-2.8274 -0.0892 0.1043>, rot=<0.0037 0.0011 0.2854 -0.9584>}
 pinky_03_r {parent=pinky_02_r, loc=<-2.2045 -0.031 0.0078>, rot=<-0.0002 0.0098 0.1445 -0.9895>}
 ring_metacarpal_r {parent=hand_r, loc=<-2.2073 0.3627 -1.3638>, rot=<-0.0284 0.0722 -0.0458 -0.9959>}
 ring_01_r {parent=ring_metacarpal_r, loc=<-7.0297 0.6197 0.1446>, rot=<-0.0489 -0.0252 0.1892 -0.9804>}
 ring_02_r {parent=ring_01_r, loc=<-3.6172 -0.4648 0.1026>, rot=<0 0 0.2806 -0.9598>}
 ring_03_r {parent=ring_02_r, loc=<-2.4153 -0.1141 -0.0199>, rot=<0 0 0.1189 -0.9929>}
 thumb_01_r {parent=hand_r, loc=<-1.3119 0.3192 2.459>, rot=<0.6546 -0.4042 0.0664 -0.6354>}
 thumb_02_r {parent=thumb_01_r, loc=<-5.9335 0.0447 -0.0401>, rot=<-0.0037 0.0213 0.298 -0.9543>}
 thumb_03_r {parent=thumb_02_r, loc=<-2.2223 -0.3682 0.0907>, rot=<0.0118 -0.013 0.0066 -0.9998>}
 lowerarm_twist_01_r {parent=lowerarm_r, loc=<-17.855 0 0>, rot=<0 0 0 1>}
 lowerarm_twist_02_r {parent=lowerarm_r, loc=<-9.6199 0 0>, rot=<0 0 0 1>}
 upperarm_twist_01_r {parent=upperarm_r, loc=<-0.0041 0 0>, rot=<0 0 0 1>}
 upperarm_twist_02_r {parent=upperarm_r, loc=<-13.7942 0 0>, rot=<0 0 0 1>}
 neck_01 {parent=spine_05, loc=<11.3138 0.1529 0>, rot=<0 0 -0.2003 -0.9797>}
 neck_02 {parent=neck_01, loc=<4.5627 0 0>, rot=<0 0 0.0504 -0.9987>}
 head {parent=neck_02, loc=<5.3914 0 0>, rot=<0 0 0.115 -0.9934>}
 dyn_parent_collar_l {parent=spine_05, loc=<10.2076 4.8715 -7.9891>, rot=<0.0284 -0.7065 -0.0284 -0.7065>}
 dyn_collar_01_l {parent=dyn_parent_collar_l, loc=<0 0 0>, rot=<0.8545 0.2187 0.2818 -0.3776>}
 dyn_collar_02_l {parent=dyn_collar_01_l, loc=<6.9505 0 0>, rot=<0 0 0 -1>}
 dyn_collar_03_l {parent=dyn_collar_02_l, loc=<5.7433 0 0>, rot=<0 0 0 -1>}
 dyn_parent_collar_r {parent=spine_05, loc=<10.2075 4.8715 7.9891>, rot=<0.0284 -0.7065 -0.0284 -0.7065>}
 dyn_collar_01_r {parent=dyn_parent_collar_r, loc=<0 0 0>, rot=<-0.3776 -0.2818 0.2187 -0.8545>}
 dyn_collar_02_r {parent=dyn_collar_01_r, loc=<-6.9504 0 0>, rot=<0 0 0 -1>}
 dyn_collar_03_r {parent=dyn_collar_02_r, loc=<-5.7437 0.0003 0.0004>, rot=<0 0 0 -1>}
 dyn_parent_collar_bck {parent=spine_05, loc=<14.4491 -6.1922 -0.0003>, rot=<0.0284 -0.7065 -0.0284 -0.7065>}
 dyn_collar_01_bck {parent=dyn_parent_collar_bck, loc=<0 0 0>, rot=<-0.0429 0.707 0.0701 -0.7024>}
 dyn_collar_02_bck {parent=dyn_collar_01_bck, loc=<5.1054 0 0>, rot=<0 0 0.0355 -0.9994>}
 dyn_collar_03_bck {parent=dyn_collar_02_bck, loc=<5.9177 0 0>, rot=<0 0 0 1>}
 dyn_parent_collar_bck_l {parent=spine_05, loc=<13.7035 -1.8185 -6.3503>, rot=<0.0284 -0.7065 -0.0284 -0.7065>}
 dyn_collar_01_bck_l {parent=dyn_parent_collar_bck_l, loc=<0 0 0>, rot=<-0.376 0.4819 -0.1853 -0.7694>}
 dyn_collar_02_bck_l {parent=dyn_collar_01_bck_l, loc=<5.1054 0 0>, rot=<0 0 0.0355 -0.9994>}
 dyn_collar_03_bck_l {parent=dyn_collar_02_bck_l, loc=<5.9177 0 0>, rot=<0 0 0 -1>}
 dyn_parent_collar_bck_r {parent=spine_05, loc=<13.7038 -1.8185 6.3503>, rot=<0.0284 -0.7065 -0.0284 -0.7065>}
 dyn_collar_01_bck_r {parent=dyn_parent_collar_bck_r, loc=<0 0 0>, rot=<0.7694 -0.1853 -0.4819 -0.376>}
 dyn_collar_02_bck_r {parent=dyn_collar_01_bck_r, loc=<-5.1049 0.0002 0.0001>, rot=<0 0 0.0355 -0.9994>}
 dyn_collar_03_bck_r {parent=dyn_collar_02_bck_r, loc=<-5.9179 -0.0001 -0.0001>, rot=<0 0 0 -1>}
 dyn_lapel_l {parent=spine_04, loc=<9.7062 11.5054 -6.5184>, rot=<0.7269 -0.0808 -0.6702 -0.1262>}
 dyn_rapel_r {parent=spine_04, loc=<9.7063 11.5054 6.5184>, rot=<-0.0808 -0.7269 0.1262 -0.6702>}
 thigh_l {parent=pelvis, loc=<-2.5693 -0.0109 -7.7501>, rot=<0.0012 0.0572 0.0155 -0.9982>}
 calf_l {parent=thigh_l, loc=<-39.0484 0 0>, rot=<0.0006 -0.0161 -0.0391 -0.9991>}
 calf_twist_02_l {parent=calf_l, loc=<-18.8064 -0.0967 0.0397>, rot=<0.0237 0.0095 -0.0035 -0.9997>}
 foot_l {parent=calf_l, loc=<-42.561 0.1077 0.4292>, rot=<0.0418 -0.0455 0.0343 -0.9975>}
 ball_l {parent=foot_l, loc=<-6.1599 13.8264 0.6806>, rot=<0.0001 -0.0001 -0.7063 -0.7079>}
 thigh_twist_01_l {parent=thigh_l, loc=<-21.6303 0 0>, rot=<0.0013 0 0 -1>}
 dyn_parent_shorts_l {parent=thigh_l, loc=<-12.1477 0 0>, rot=<0 -1 0.0013 0>}
 parent_dyn_shorts_main_l {parent=dyn_parent_shorts_l, loc=<-3.9401 -0.9165 0.1197>, rot=<-0.0006 0.6655 -0.0027 -0.7464>}
 driver_dyn_shorts_main_l {parent=parent_dyn_shorts_main_l, loc=<0 0 0>, rot=<0 0 0 1>}
 dyn_shorts_main_rear_l {parent=driver_dyn_shorts_main_l, loc=<0 0 -0.0827>, rot=<0 0 0 1>}
 dyn_shorts_outA_01_l {parent=dyn_shorts_main_rear_l, loc=<8.5254 3.7073 0.8319>, rot=<0.6157 -0.4092 -0.4786 -0.4737>}
 dyn_shorts_outA_02_l {parent=dyn_shorts_outA_01_l, loc=<10.1317 0 0>, rot=<0 0 0 1>}
 dyn_shorts_outB_01_l {parent=dyn_shorts_main_rear_l, loc=<9.3124 -3.0419 1.4292>, rot=<0.5128 -0.5179 -0.3891 -0.5633>}
 dyn_shorts_outB_02_l {parent=dyn_shorts_outB_01_l, loc=<9.8406 0 0>, rot=<0 0 0 -1>}
 dyn_shorts_bkA_01_l {parent=dyn_shorts_main_rear_l, loc=<6.5027 -8.6683 1.808>, rot=<0.3259 -0.626 -0.2184 -0.674>}
 dyn_shorts_bkA_02_l {parent=dyn_shorts_bkA_01_l, loc=<10.252 0 0>, rot=<0 0 0 1>}
 dyn_shorts_bkA_03_l {parent=dyn_shorts_bkA_02_l, loc=<9.8234 0 0>, rot=<0 0 0 1>}
 dyn_shorts_bkB_01_l {parent=dyn_shorts_main_rear_l, loc=<0.5915 -12.933 2.3065>, rot=<0.2065 -0.6816 -0.1009 -0.6947>}
 dyn_shorts_bkB_02_l {parent=dyn_shorts_bkB_01_l, loc=<10.1085 0 0>, rot=<0 -0.0003 0.0052 -1>}
 dyn_shorts_bkB_03_l {parent=dyn_shorts_bkB_02_l, loc=<10.0987 0.0229 0.0012>, rot=<0.0002 -0.001 0.0189 -0.9998>}
 dyn_rearA_01_l {parent=dyn_shorts_main_rear_l, loc=<-7.1569 -14.426 2.2595>, rot=<0.1023 -0.6941 0.0144 -0.7124>}
 dyn_rearA_02_l {parent=dyn_rearA_01_l, loc=<10.023 0.396 -0.0005>, rot=<0.0001 0.0001 -0.0099 -1>}
 dyn_rearA_03_l {parent=dyn_rearA_02_l, loc=<9.9858 0.0974 -0.0011>, rot=<0.0004 0.0002 0.0192 -0.9998>}
 dyn_shorts_main_l {parent=driver_dyn_shorts_main_l, loc=<0 0 0>, rot=<0 0 0.2691 -0.9631>}
 dyn_shorts_frt_01_l {parent=dyn_shorts_main_l, loc=<-3.379 10.6064 0.0652>, rot=<0.7371 0.0334 -0.6738 -0.0398>}
 dyn_shorts_frt_02_l {parent=dyn_shorts_frt_01_l, loc=<9.9311 0 0>, rot=<0 0 0 -1>}
 dyn_shorts_frt_03_l {parent=dyn_shorts_frt_02_l, loc=<9.9336 0 0>, rot=<0 0 0 -1>}
 thigh_r {parent=pelvis, loc=<-2.5691 -0.0109 7.7501>, rot=<0.0572 -0.0012 0.9982 0.0155>}
 calf_r {parent=thigh_r, loc=<39.0487 0 0>, rot=<0.0006 -0.0161 -0.0391 -0.9991>}
 calf_twist_02_r {parent=calf_r, loc=<18.8064 0.0967 -0.0397>, rot=<0.0237 0.0095 -0.0035 -0.9997>}
 foot_r {parent=calf_r, loc=<42.5611 -0.1077 -0.4292>, rot=<0.0418 -0.0455 0.0343 -0.9975>}
 ball_r {parent=foot_r, loc=<6.1598 -13.8264 -0.6806>, rot=<0.0001 -0.0001 -0.7063 -0.7079>}
 thigh_twist_01_r {parent=thigh_r, loc=<21.6303 0 0>, rot=<0.0013 0 0 -1>}
 dyn_parent_shorts_r {parent=thigh_r, loc=<12.1477 0 0>, rot=<0.0013 0 0 -1>}
 parent_dyn_shorts_main_r {parent=dyn_parent_shorts_r, loc=<-3.94 0.9165 0.1197>, rot=<-0.0027 -0.7464 0.0006 -0.6655>}
 driver_dyn_shorts_main_r {parent=parent_dyn_shorts_main_r, loc=<0 0 0>, rot=<0 0 0 1>}
 dyn_shorts_main_rear_r {parent=driver_dyn_shorts_main_r, loc=<0 0 0.0827>, rot=<0 0 0 1>}
 dyn_shorts_outA_01_r {parent=dyn_shorts_main_rear_r, loc=<-8.5329 -3.9354 -0.8561>, rot=<0.5985 -0.4139 -0.4778 -0.4921>}
 dyn_shorts_outA_02_r {parent=dyn_shorts_outA_01_r, loc=<-10.3767 0 0>, rot=<0 0 0 1>}
 dyn_shorts_outB_01_r {parent=dyn_shorts_main_rear_r, loc=<-9.0393 4.0471 -1.5345>, rot=<0.4712 -0.5505 -0.3461 -0.596>}
 dyn_shorts_outB_02_r {parent=dyn_shorts_outB_01_r, loc=<-10.4179 0 0>, rot=<0 0 0 1>}
 dyn_shorts_bkA_01_r {parent=dyn_shorts_main_rear_r, loc=<-5.1078 10.1479 -2.0532>, rot=<0.3047 -0.6472 -0.204 -0.6684>}
 dyn_shorts_bkA_02_r {parent=dyn_shorts_bkA_01_r, loc=<-10.2713 0 0>, rot=<0 0 0 -1>}
 dyn_shorts_bkA_03_r {parent=dyn_shorts_bkA_02_r, loc=<-9.6889 -0.0165 0>, rot=<0 0 0.0251 -0.9997>}
 dyn_shorts_bkB_01_r {parent=dyn_shorts_main_rear_r, loc=<0.9169 13.0602 -2.226>, rot=<0.1666 -0.6902 -0.0668 -0.7011>}
 dyn_shorts_bkB_02_r {parent=dyn_shorts_bkB_01_r, loc=<-10.4106 0 -0.0001>, rot=<0 -0.0003 0.0052 -1>}
 dyn_shorts_bkB_03_r {parent=dyn_shorts_bkB_02_r, loc=<-9.7114 -0.0375 -0.002>, rot=<0.0002 -0.001 0.0189 -0.9998>}
 dyn_rearA_01_r {parent=dyn_shorts_main_rear_r, loc=<6.7214 14.2957 -2.4319>, rot=<0.0719 -0.7014 0.0247 -0.7087>}
 dyn_rearA_02_r {parent=dyn_rearA_01_r, loc=<-10.2143 -0.0035 0.012>, rot=<0 -0.0003 0.0052 -1>}
 dyn_rearA_03_r {parent=dyn_rearA_02_r, loc=<-9.8963 -0.0149 -0.0034>, rot=<0.0002 -0.001 0.0093 -1>}
 dyn_shorts_main_r {parent=driver_dyn_shorts_main_r, loc=<0 0 0>, rot=<0 0 0.2691 -0.9631>}
 dyn_shorts_frt_01_r {parent=dyn_shorts_main_r, loc=<3.2002 -10.541 -0.2279>, rot=<0.7299 -0.0761 -0.6581 -0.1684>}
 dyn_shorts_frt_02_r {parent=dyn_shorts_frt_01_r, loc=<-10.4054 0 0>, rot=<0 0 0 -1>}
 dyn_shorts_frt_03_r {parent=dyn_shorts_frt_02_r, loc=<-9.5707 0 0>, rot=<0 0 0 -1>}
 dyn_shorts_frtA_01_r {parent=dyn_shorts_main_r, loc=<7.2288 -9.8658 -2.9229>, rot=<-0.0893 0.7119 0.0013 -0.6966>}
]
