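# Physics asset for the M_MED_DJ_Remix male player cosmetic body, set up for secondary motion of its collar.
# Box bodies cover the back, back-left and back-right collar flaps as single segments, and the left and right sides as two-segment chains.
# Swing- and twist-limited constraints link each flap to a small parent anchor box or preceding segment, with very tight twist throughout.
# Two head capsules provide a collision volume beside the collar, with no constraint attached.
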PhysicsAsset /BRCosmetics/Characters/Player/Male/Medium/Bodies/M_MED_DJ_Remix/Meshes/M_MED_DJ_Remix_Physics
bodies {
 dyn_collar_01_bck [{shapeType=Box, extentX=17.449, extentY=1.958, extentZ=15.1122, center=<13 0 0>}]
 dyn_collar_01_bck_l [{shapeType=Box, extentX=18.9442, extentY=3.6615, extentZ=16.7521, center=<13 0 0>}]
 dyn_collar_01_bck_r [{shapeType=Box, extentX=20.485, extentY=3.7249, extentZ=16.7805, center=<-13 0 0.0003>}]
 dyn_collar_01_l [{shapeType=Box, extentX=11.1957, extentY=3.97, extentZ=7.5358, center=<7.5 0 0>}]
 dyn_collar_01_r [{shapeType=Box, extentX=13.1957, extentY=4.47, extentZ=7.5358, center=<-7.5 0 0>}]
 dyn_collar_02_l [{shapeType=Box, extentX=12, extentY=2.5, extentZ=8.5, center=<6 0 0>}]
 dyn_collar_02_r [{shapeType=Box, extentX=14, extentY=2.5, extentZ=13.5, center=<-6 0 0>}]
 dyn_parent_collar_bck [{shapeType=Box, extentX=1.01, extentY=1.01, extentZ=1.01, rotation=<rot 0 -90 0>}]
 dyn_parent_collar_bck_l [{shapeType=Box, extentX=1.01, extentY=1.01, extentZ=1.01, rotation=<rot 0 -90 0>}]
 dyn_parent_collar_bck_r [{shapeType=Box, extentX=1.01, extentY=1.01, extentZ=1.01, rotation=<rot 0 -90 0>}]
 dyn_parent_collar_l [{shapeType=Box, extentX=1.01, extentY=1.01, extentZ=1.01, rotation=<rot 0 -90 0>}]
 dyn_parent_collar_r [{shapeType=Box, extentX=1.01, extentY=1.01, extentZ=1.01, rotation=<rot 0 -90 0>}]
 head [
  {shapeType=Capsule, length=13, radius=8.9999, center=<4.6908 0.3216 -0.6014>, rotation=<rot -4.6505 -90.076 0.0071>}
  {shapeType=Capsule, length=3.75, radius=9.75, center=<0.6369 -0.9212 0>, rotation=<rot -0.0001 -179.9998 180>}
 ]
}
constraints [
 {bone1Name=dyn_collar_01_bck, bone2Name=dyn_parent_collar_bck, swing1LimitDegrees=30, swing2LimitDegrees=15, twistLimitDegrees=0.2}
 {bone1Name=dyn_collar_01_bck_r, bone2Name=dyn_parent_collar_bck_r, swing1LimitDegrees=45, swing2LimitDegrees=45, twistLimitDegrees=0.2}
 {bone1Name=dyn_collar_01_bck_l, bone2Name=dyn_parent_collar_bck_l, swing1LimitDegrees=45, swing2LimitDegrees=45, twistLimitDegrees=0.2}
 {bone1Name=dyn_collar_01_l, bone2Name=dyn_parent_collar_l, swing1LimitDegrees=15, swing2LimitDegrees=80, twistLimitDegrees=5}
 {bone1Name=dyn_collar_01_r, bone2Name=dyn_parent_collar_r, swing1LimitDegrees=10, swing2LimitDegrees=80, twistLimitDegrees=5}
 {bone1Name=dyn_collar_02_l, bone2Name=dyn_collar_01_l, swing1LimitDegrees=45, swing2LimitDegrees=45, twistLimitDegrees=5}
 {bone1Name=dyn_collar_02_r, bone2Name=dyn_collar_01_r, swing1LimitDegrees=45, swing2LimitDegrees=45, twistLimitDegrees=5}
]
