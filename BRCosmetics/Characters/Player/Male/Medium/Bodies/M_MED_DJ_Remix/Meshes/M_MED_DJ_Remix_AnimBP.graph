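# Animation Blueprint for the M_MED_DJ_Remix male player body cosmetic, adding secondary motion and pose correctives to shorts, collar and lapel bones.
# One cached-pose branch chains Anim Dynamics, scaled by Shorts_Alpha and pushed by skydiving_force, into additive bone offsets and RigidBody simulation.
# Many offsets use interpolated calf-bend or arm-raise alphas as correctives.
# Another branch uses weighted constraints binding shorts bones to pelvis and leg bones.
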AnimBlueprint /BRCosmetics/Characters/Player/Male/Medium/Bodies/M_MED_DJ_Remix/Meshes/M_MED_DJ_Remix_AnimBP
target_skeleton=&/M_MED_DJ_Remix_Skeleton
graphs [
 AnimGraph {nodes [
   n1 = Get Shorts_Alpha() -> (Shorts_Alpha:double)
   n2 = Get skydiving_force() -> (skydiving_force:Vector)
   "Anim Dynamics - Chain: dyn_shorts_frtA_01_r -> dyn_shorts_frtA_02_r"(Alpha=@n1, ExternalForce=@n2) {
    angularDampingOverride=0.9999
    angularSpringConstant=5
    bAngularSpring=true
    bChain=true
    bGravityOverrideInSimSpace=true
    bOverrideAngularDamping=true
    gravityScale=0.3
    numSolverIterationsPostUpdate=8
    numSolverIterationsPreUpdate=32
    simulationSpace=RootRelative
    alphaBoolBlend {blendOption=Linear}
    boundBone {boneName=dyn_shorts_frtA_01_r}
    chainEnd {boneName=dyn_shorts_frtA_02_r}
    componentAppliedLinearAccClamp=<1300 1300 1300>
    componentLinearAccScale=<1 1 1>
    componentLinearVelScale=<1 1 1>
    physicsBodyDefinitions [{
      sphereCollisionRadius=0
      boundBone {boneName=dyn_shorts_frtA_01_r}
      boxExtents=<18 18 23>
      constraintSetup {angularLimitsMax=<10 35 10>, angularLimitsMin=<-10 -20 -35>, angularTarget=<0 0 1>}
      localJointOffset=<5 0 0>
     }]
    relativeSpaceBone {boneName=thigh_l}
   }
   "Constraint - dyn_shorts_outA_03_r (P)"(Alpha=1, ConstraintWeights_0=1) {
    alphaBoolBlend {blendOption=Linear}
    boneToModify {boneName=dyn_shorts_outA_03_r}
    constraintSetup [{targetBone {boneName=thigh_r}}]
    constraintWeights [1]
   }
   "Constraint - dyn_shorts_outB_03_r (P)"(Alpha=1, ConstraintWeights_0=1) {
    alphaBoolBlend {blendOption=Linear}
    boneToModify {boneName=dyn_shorts_outB_03_r}
    constraintSetup [{targetBone {boneName=thigh_r}}]
    constraintWeights [1]
   }
   "Constraint - dyn_shorts_outB_03_l (PP)"!(ComponentPose, Alpha=1, ConstraintWeights_0=0.3502, ConstraintWeights_1:float=0.6498) -> (Pose) {
    alphaBoolBlend {blendOption=Linear}
    boneToModify {boneName=dyn_shorts_outB_03_l}
    constraintSetup [{targetBone {boneName=thigh_twist_01_l}}, {targetBone {boneName=thigh_l}}]
    constraintWeights [1,1]
   }
   RigidBody() {
    bClampLinearTranslationLimitToRefPose=true
    bOverrideWorldGravity=true
    overridePhysicsAsset=&/M_MED_DJ_Remix_Physics
    simulationSpace=BaseBoneSpace
    alphaBoolBlend {blendOption=Linear}
    baseBoneRef {boneName=root}
    componentAppliedLinearAccClamp=<20000 20000 1300>
    componentLinearAccScale=<1 1 1>
    componentLinearVelScale=<1 1 1>
    overrideWorldGravity {z=-980}
   }
   n3 = Get calf_rot_back_r() -> (calf_rot_back_r:double)
   n4 = "Anim Dynamics - Chain: dyn_shorts_frt_01_l -> dyn_shorts_frt_03_l"(ComponentPose=Local To Component(LocalPose=Use cached pose 'Anim_Cache'()), Alpha=@n1, ExternalForce=@n2) {
    angularDampingOverride=0.9999
    angularSpringConstant=5
    bAngularSpring=true
    bChain=true
    bGravityOverrideInSimSpace=true
    bOverrideAngularDamping=true
    bUseSphericalLimits=true
    gravityScale=0.3
    numSolverIterationsPostUpdate=8
    numSolverIterationsPreUpdate=32
    simulationSpace=RootRelative
    alphaBoolBlend {blendOption=Linear}
    boundBone {boneName=dyn_shorts_frt_01_l}
    chainEnd {boneName=dyn_shorts_frt_03_l}
    componentAppliedLinearAccClamp=<1300 1300 1300>
    componentLinearAccScale=<1 1 1>
    componentLinearVelScale=<1 1 1>
    physicsBodyDefinitions [{
      collisionType=CustomSphere
      sphereCollisionRadius=2.75
      boundBone {boneName=dyn_shorts_frt_01_l}
      boxExtents=<18 18 23>
      constraintSetup {angularLimitsMax=<10 35 10>, angularLimitsMin=<-10 -20 -35>, angularTarget=<0 0 1>}
      localJointOffset=<5 0 0>
     }]
    relativeSpaceBone {boneName=thigh_l}
    sphericalLimits [{limitRadius=7.8, drivingBone {boneName=thigh_l}, sphereLocalOffset=<-28 0 2>}]
   }
   n5 = "Anim Dynamics - Chain: dyn_shorts_outA_01_l -> dyn_shorts_outA_02_l"(ComponentPose=@n4, Alpha=@n1, ExternalForce=@n2) {
    angularDampingOverride=1
    angularSpringConstant=5
    bAngularSpring=true
    bChain=true
    bGravityOverrideInSimSpace=true
    bOverrideAngularDamping=true
    bUseSphericalLimits=true
    gravityScale=0.3
    numSolverIterationsPostUpdate=8
    numSolverIterationsPreUpdate=32
    simulationSpace=RootRelative
    alphaBoolBlend {blendOption=Linear}
    boundBone {boneName=dyn_shorts_outA_01_l}
    chainEnd {boneName=dyn_shorts_outA_02_l}
    componentAppliedLinearAccClamp=<1000 1000 1000>
    componentLinearAccScale=<1 1 1>
    componentLinearVelScale=<1 1 1>
    physicsBodyDefinitions [{
      collisionType=CustomSphere
      sphereCollisionRadius=2.5
      boundBone {boneName=dyn_shorts_outA_01_l}
      boxExtents=<18 18 30>
      constraintSetup {angularLimitsMax=<10 35 10>, angularLimitsMin=<-10 -20 -35>, angularTarget=<0 0 1>}
      localJointOffset=<6.5 0 0>
     }]
    relativeSpaceBone {boneName=thigh_l}
    sphericalLimits [{limitRadius=8.5, drivingBone {boneName=thigh_l}, sphereLocalOffset=<-28 0 0>}]
   }
   n6 = "Anim Dynamics - Chain: dyn_shorts_outB_01_l -> dyn_shorts_outB_02_l"(ComponentPose=@n5, Alpha=@n1, ExternalForce=@n2) {
    angularDampingOverride=1
    angularSpringConstant=5
    bAngularSpring=true
    bChain=true
    bGravityOverrideInSimSpace=true
    bOverrideAngularDamping=true
    gravityScale=0.3
    numSolverIterationsPostUpdate=8
    numSolverIterationsPreUpdate=32
    simulationSpace=RootRelative
    alphaBoolBlend {blendOption=Linear}
    boundBone {boneName=dyn_shorts_outB_01_l}
    chainEnd {boneName=dyn_shorts_outB_02_l}
    componentAppliedLinearAccClamp=<1000 1000 1000>
    componentLinearAccScale=<1 1 1>
    componentLinearVelScale=<1 1 1>
    physicsBodyDefinitions [{
      sphereCollisionRadius=0
      boundBone {boneName=dyn_shorts_outB_01_l}
      boxExtents=<18 18 30>
      constraintSetup {angularLimitsMax=<10 35 10>, angularLimitsMin=<-10 -20 -35>, angularTarget=<0 0 1>}
      localJointOffset=<6.5 0 0>
     }]
    relativeSpaceBone {boneName=thigh_l}
   }
   n7 = "Anim Dynamics - Chain: dyn_shorts_bkA_01_l -> dyn_shorts_bkA_02_l"(ComponentPose=@n6, Alpha=@n1, ExternalForce=@n2) {
    angularDampingOverride=1
    angularSpringConstant=5
    bAngularSpring=true
    bChain=true
    bGravityOverrideInSimSpace=true
    bOverrideAngularDamping=true
    gravityScale=0.3
    numSolverIterationsPostUpdate=8
    numSolverIterationsPreUpdate=32
    simulationSpace=RootRelative
    alphaBoolBlend {blendOption=Linear}
    boundBone {boneName=dyn_shorts_bkA_01_l}
    chainEnd {boneName=dyn_shorts_bkA_02_l}
    componentAppliedLinearAccClamp=<1000 1000 1000>
    componentLinearAccScale=<1 1 1>
    componentLinearVelScale=<1 1 1>
    physicsBodyDefinitions [{
      sphereCollisionRadius=0
      boundBone {boneName=dyn_shorts_bkA_01_l}
      boxExtents=<18 18 30>
      constraintSetup {angularLimitsMax=<10 35 10>, angularLimitsMin=<-10 -20 -35>, angularTarget=<0 0 1>}
      localJointOffset=<4.5 0 0>
     }]
    relativeSpaceBone {boneName=thigh_l}
   }
   n8 = "Anim Dynamics - Chain: dyn_shorts_bkB_01_l -> dyn_shorts_bkB_02_l"(ComponentPose=@n7, Alpha=@n1, ExternalForce=@n2) {
    angularDampingOverride=1
    angularSpringConstant=5
    bAngularSpring=true
    bChain=true
    bGravityOverrideInSimSpace=true
    bOverrideAngularDamping=true
    gravityScale=0.3
    numSolverIterationsPostUpdate=8
    numSolverIterationsPreUpdate=32
    simulationSpace=RootRelative
    alphaBoolBlend {blendOption=Linear}
    boundBone {boneName=dyn_shorts_bkB_01_l}
    chainEnd {boneName=dyn_shorts_bkB_02_l}
    componentAppliedLinearAccClamp=<1000 1000 1000>
    componentLinearAccScale=<1 1 1>
    componentLinearVelScale=<1 1 1>
    physicsBodyDefinitions [{
      sphereCollisionRadius=0
      boundBone {boneName=dyn_shorts_bkB_01_l}
      boxExtents=<18 18 30>
      constraintSetup {angularLimitsMax=<10 35 10>, angularLimitsMin=<-10 -20 -30>, angularTarget=<0 0 1>}
      localJointOffset=<4.5 0 0>
     }]
    relativeSpaceBone {boneName=thigh_l}
   }
   n9 = "Anim Dynamics - Chain: dyn_rearA_01_l -> dyn_rearA_02_l"(ComponentPose=@n8, Alpha=@n1, ExternalForce=@n2) {
    angularDampingOverride=1
    angularSpringConstant=5
    bAngularSpring=true
    bChain=true
    bGravityOverrideInSimSpace=true
    bOverrideAngularDamping=true
    gravityScale=0.3
    numSolverIterationsPostUpdate=8
    numSolverIterationsPreUpdate=32
    simulationSpace=RootRelative
    alphaBoolBlend {blendOption=Linear}
    boundBone {boneName=dyn_rearA_01_l}
    chainEnd {boneName=dyn_rearA_02_l}
    componentAppliedLinearAccClamp=<1000 1000 1000>
    componentLinearAccScale=<1 1 1>
    componentLinearVelScale=<1 1 1>
    physicsBodyDefinitions [{
      sphereCollisionRadius=0
      boundBone {boneName=dyn_rearA_01_l}
      boxExtents=<18 18 30>
      constraintSetup {angularTargetAxis=AxisZ, angularLimitsMax=<10 35 10>, angularLimitsMin=<-10 -20 -35>, angularTarget=<0 0 1>}
      localJointOffset=<4.5 0 0>
     }]
    relativeSpaceBone {boneName=thigh_l}
   }
   n10 = "Anim Dynamics - Chain: dyn_shorts_frt_01_r -> dyn_shorts_frt_02_r"(ComponentPose=@n9, Alpha=@n1, ExternalForce=@n2) {
    angularDampingOverride=1
    angularSpringConstant=5
    bAngularSpring=true
    bChain=true
    bGravityOverrideInSimSpace=true
    bOverrideAngularDamping=true
    gravityScale=0.3
    numSolverIterationsPostUpdate=8
    numSolverIterationsPreUpdate=32
    simulationSpace=RootRelative
    alphaBoolBlend {blendOption=Linear}
    boundBone {boneName=dyn_shorts_frt_01_r}
    chainEnd {boneName=dyn_shorts_frt_02_r}
    componentAppliedLinearAccClamp=<1000 1000 1000>
    componentLinearAccScale=<1 1 1>
    componentLinearVelScale=<1 1 1>
    physicsBodyDefinitions [{
      sphereCollisionRadius=0
      boundBone {boneName=dyn_shorts_frt_01_r}
      boxExtents=<18 18 30>
      constraintSetup {angularLimitsMax=<10 35 8>, angularLimitsMin=<-10 -20 -35>, angularTarget=<0 0 1>}
      localJointOffset=<-5 0 0>
     }]
    relativeSpaceBone {boneName=thigh_l}
   }
   n11 = "Anim Dynamics - Chain: dyn_shorts_outA_01_r -> dyn_shorts_outA_02_r"(ComponentPose=@n10, Alpha=@n1, ExternalForce=@n2) {
    angularDampingOverride=1
    angularSpringConstant=5
    bAngularSpring=true
    bChain=true
    bGravityOverrideInSimSpace=true
    bOverrideAngularDamping=true
    gravityScale=0.3
    numSolverIterationsPostUpdate=8
    numSolverIterationsPreUpdate=32
    simulationSpace=RootRelative
    alphaBoolBlend {blendOption=Linear}
    boundBone {boneName=dyn_shorts_outA_01_r}
    chainEnd {boneName=dyn_shorts_outA_02_r}
    componentAppliedLinearAccClamp=<1000 1000 1000>
    componentLinearAccScale=<1 1 1>
    componentLinearVelScale=<1 1 1>
    physicsBodyDefinitions [{
      sphereCollisionRadius=0
      boundBone {boneName=dyn_shorts_outA_01_r}
      boxExtents=<18 18 30>
      constraintSetup {angularLimitsMax=<10 35 8>, angularLimitsMin=<-10 -20 -35>, angularTarget=<0 0 1>}
      localJointOffset=<-6.5 0 0>
     }]
    relativeSpaceBone {boneName=thigh_l}
   }
   n12 = "Anim Dynamics - Chain: dyn_shorts_outB_01_r -> dyn_shorts_outB_02_r"(ComponentPose=@n11, Alpha=@n1, ExternalForce=@n2) {
    angularDampingOverride=1
    angularSpringConstant=5
    bAngularSpring=true
    bChain=true
    bGravityOverrideInSimSpace=true
    bOverrideAngularDamping=true
    gravityScale=0.3
    numSolverIterationsPostUpdate=8
    numSolverIterationsPreUpdate=32
    simulationSpace=RootRelative
    alphaBoolBlend {blendOption=Linear}
    boundBone {boneName=dyn_shorts_outB_01_r}
    chainEnd {boneName=dyn_shorts_outB_02_r}
    componentAppliedLinearAccClamp=<1000 1000 1000>
    componentLinearAccScale=<1 1 1>
    componentLinearVelScale=<1 1 1>
    physicsBodyDefinitions [{
      sphereCollisionRadius=0
      boundBone {boneName=dyn_shorts_outB_01_r}
      boxExtents=<18 18 30>
      constraintSetup {angularLimitsMax=<10 35 8>, angularLimitsMin=<-10 -20 -35>, angularTarget=<0 0 1>}
      localJointOffset=<-6.5 0 0>
     }]
    relativeSpaceBone {boneName=thigh_l}
   }
   n13 = "Anim Dynamics - Chain: dyn_shorts_bkA_01_r -> dyn_shorts_bkA_02_r"(ComponentPose=@n12, Alpha=@n1, ExternalForce=@n2) {
    angularDampingOverride=1
    angularSpringConstant=5
    bAngularSpring=true
    bChain=true
    bGravityOverrideInSimSpace=true
    bOverrideAngularDamping=true
    gravityScale=0.3
    numSolverIterationsPostUpdate=8
    numSolverIterationsPreUpdate=32
    simulationSpace=RootRelative
    alphaBoolBlend {blendOption=Linear}
    boundBone {boneName=dyn_shorts_bkA_01_r}
    chainEnd {boneName=dyn_shorts_bkA_02_r}
    componentAppliedLinearAccClamp=<1000 1000 1000>
    componentLinearAccScale=<1 1 1>
    componentLinearVelScale=<1 1 1>
    physicsBodyDefinitions [{
      sphereCollisionRadius=0
      boundBone {boneName=dyn_shorts_bkA_01_r}
      boxExtents=<18 18 30>
      constraintSetup {angularLimitsMax=<10 35 8>, angularLimitsMin=<-10 -20 -35>, angularTarget=<0 0 1>}
      localJointOffset=<-6.5 0 0>
     }]
    relativeSpaceBone {boneName=thigh_l}
   }
   n14 = "Anim Dynamics - Chain: dyn_shorts_bkB_01_r -> dyn_shorts_bkB_02_r"(ComponentPose=@n13, Alpha=@n1, ExternalForce=@n2) {
    angularDampingOverride=1
    angularSpringConstant=5
    bAngularSpring=true
    bChain=true
    bGravityOverrideInSimSpace=true
    bOverrideAngularDamping=true
    gravityScale=0.3
    numSolverIterationsPostUpdate=8
    numSolverIterationsPreUpdate=32
    simulationSpace=RootRelative
    alphaBoolBlend {blendOption=Linear}
    boundBone {boneName=dyn_shorts_bkB_01_r}
    chainEnd {boneName=dyn_shorts_bkB_02_r}
    componentAppliedLinearAccClamp=<1000 1000 1000>
    componentLinearAccScale=<1 1 1>
    componentLinearVelScale=<1 1 1>
    physicsBodyDefinitions [{
      sphereCollisionRadius=0
      boundBone {boneName=dyn_shorts_bkB_01_r}
      boxExtents=<18 18 30>
      constraintSetup {angularLimitsMax=<10 35 8>, angularLimitsMin=<-10 -20 -35>, angularTarget=<0 0 1>}
      localJointOffset=<-6.5 0 0>
     }]
    relativeSpaceBone {boneName=thigh_l}
   }
   n15 = "Anim Dynamics - Chain: dyn_rearA_01_r -> dyn_rearA_02_r"(ComponentPose=@n14, Alpha=@n1, ExternalForce=@n2) {
    angularDampingOverride=1
    angularSpringConstant=5
    bAngularSpring=true
    bChain=true
    bGravityOverrideInSimSpace=true
    bOverrideAngularDamping=true
    gravityScale=0.3
    numSolverIterationsPostUpdate=8
    numSolverIterationsPreUpdate=32
    simulationSpace=RootRelative
    alphaBoolBlend {blendOption=Linear}
    boundBone {boneName=dyn_rearA_01_r}
    chainEnd {boneName=dyn_rearA_02_r}
    componentAppliedLinearAccClamp=<1000 1000 1000>
    componentLinearAccScale=<1 1 1>
    componentLinearVelScale=<1 1 1>
    physicsBodyDefinitions [{
      sphereCollisionRadius=0
      boundBone {boneName=dyn_rearA_01_r}
      boxExtents=<18 18 30>
      constraintSetup {angularLimitsMax=<10 35 8>, angularLimitsMin=<-10 -20 -35>, angularTarget=<0 0 1>}
      localJointOffset=<-6.5 0 0>
     }]
    relativeSpaceBone {boneName=thigh_l}
   }
   n16 = "Constraint - dyn_parent_shorts_l (P)"(ComponentPose=@n15, Alpha=1, ConstraintWeights_0=1) {
    alphaBoolBlend {blendOption=Linear}
    boneToModify {boneName=dyn_parent_shorts_l}
    constraintSetup [{targetBone {boneName=thigh_twist_01_l}}]
    constraintWeights [1]
   }
   n17 = "Constraint - dyn_parent_shorts_r (P)"(ComponentPose=@n16, Alpha=1, ConstraintWeights_0=1) {
    alphaBoolBlend {blendOption=Linear}
    boneToModify {boneName=dyn_parent_shorts_r}
    constraintSetup [{targetBone {boneName=thigh_twist_01_r}}]
    constraintWeights [1]
   }
   n18 = Get calf_rot_back_l() -> (calf_rot_back_l:double)
   n19 = "Transform (Modify) Bone - Bone: dyn_shorts_main_l"(ComponentPose=@n17, Alpha=@n18, Translation="0,0,0", Rotation="0,0,0", Scale="1,1,1") {
    rotationMode=BMM_Additive
    rotationSpace=BCS_BoneSpace
    scaleMode=BMM_Additive
    scaleSpace=BCS_BoneSpace
    translationMode=BMM_Additive
    translationSpace=BCS_BoneSpace
    alphaBoolBlend {blendOption=Linear}
    alphaScaleBiasClamp {bInterpResult=true, interpSpeedDecreasing=25, interpSpeedIncreasing=25}
    boneToModify {boneName=dyn_shorts_main_l}
   }
   n20 = "Transform (Modify) Bone - Bone: dyn_shorts_frt_01_l"(ComponentPose=@n19, Alpha=@n18, Translation="3,0.6,0", Rotation="15,-3,0", Scale="1,1,1") {
    rotationMode=BMM_Additive
    rotationSpace=BCS_BoneSpace
    scaleMode=BMM_Additive
    scaleSpace=BCS_BoneSpace
    translationMode=BMM_Additive
    translationSpace=BCS_BoneSpace
    alphaBoolBlend {blendOption=Linear}
    alphaScaleBiasClamp {bInterpResult=true, interpSpeedDecreasing=25, interpSpeedIncreasing=25}
    boneToModify {boneName=dyn_shorts_frt_01_l}
    rotation {pitch=15, yaw=-3}
    translation=<3 0.6>
   }
   n21 = "Transform (Modify) Bone - Bone: dyn_shorts_outA_01_l"(ComponentPose=@n20, Alpha=@n18, Translation="-0.000,0.022,-4", Rotation="5,0,-20", Scale="1,1,1") {
    rotationMode=BMM_Additive
    rotationSpace=BCS_BoneSpace
    scaleMode=BMM_Additive
    scaleSpace=BCS_BoneSpace
    translationMode=BMM_Additive
    translationSpace=BCS_BoneSpace
    alphaBoolBlend {blendOption=Linear}
    alphaScaleBiasClamp {bInterpResult=true, interpSpeedDecreasing=25, interpSpeedIncreasing=25}
    boneToModify {boneName=dyn_shorts_outA_01_l}
    rotation {pitch=5, roll=-20}
    translation {y=0.022, z=-4}
   }
   n22 = "Transform (Modify) Bone - Bone: dyn_shorts_outB_01_l"(ComponentPose=@n21, Alpha=@n18, Translation="-2,5,-8", Rotation="0.000,10.000,0.000", Scale="1,1,1") {
    rotationMode=BMM_Additive
    rotationSpace=BCS_BoneSpace
    scaleMode=BMM_Additive
    scaleSpace=BCS_BoneSpace
    translationMode=BMM_Additive
    translationSpace=BCS_BoneSpace
    alphaBoolBlend {blendOption=Linear}
    alphaScaleBiasClamp {bInterpResult=true, interpSpeedDecreasing=25, interpSpeedIncreasing=25}
    boneToModify {boneName=dyn_shorts_outB_01_l}
    rotation {yaw=10}
    translation=<-2 5 -8>
   }
   n23 = "Transform (Modify) Bone - Bone: dyn_shorts_bkA_01_l"(ComponentPose=@n22, Alpha=@n18, Translation="-10,-8,-10", Rotation="10,-20,0.000", Scale="1,1,1") {
    rotationMode=BMM_Additive
    rotationSpace=BCS_BoneSpace
    scaleMode=BMM_Additive
    scaleSpace=BCS_BoneSpace
    translationMode=BMM_Additive
    translationSpace=BCS_BoneSpace
    alphaBoolBlend {blendOption=Linear}
    alphaScaleBiasClamp {bInterpResult=true, interpSpeedDecreasing=25, interpSpeedIncreasing=25}
    boneToModify {boneName=dyn_shorts_bkA_01_l}
    rotation {pitch=10, yaw=-20}
    translation=<-10 -8 -10>
   }
   n24 = "Transform (Modify) Bone - Bone: dyn_shorts_bkB_01_l"(ComponentPose=@n23, Alpha=@n18, Translation="-20,-8,-13", Rotation="5,-25,0", Scale="1,1,1") {
    rotationMode=BMM_Additive
    rotationSpace=BCS_BoneSpace
    scaleMode=BMM_Additive
    scaleSpace=BCS_BoneSpace
    translationMode=BMM_Additive
    translationSpace=BCS_BoneSpace
    alphaBoolBlend {blendOption=Linear}
    alphaScaleBiasClamp {bInterpResult=true, interpSpeedDecreasing=25, interpSpeedIncreasing=25}
    boneToModify {boneName=dyn_shorts_bkB_01_l}
    rotation {pitch=5, yaw=-25}
    translation=<-20 -8 -13>
   }
   n25 = "Transform (Modify) Bone - Bone: dyn_rearA_01_l"(ComponentPose=@n24, Alpha=@n18, Translation="-30,-10,-15", Rotation="0,-40,0.000", Scale="0.95,0.95,0.95") {
    rotationMode=BMM_Additive
    rotationSpace=BCS_BoneSpace
    scaleMode=BMM_Additive
    scaleSpace=BCS_BoneSpace
    translationMode=BMM_Additive
    translationSpace=BCS_BoneSpace
    alphaBoolBlend {blendOption=Linear}
    alphaScaleBiasClamp {bInterpResult=true, interpSpeedDecreasing=25, interpSpeedIncreasing=25}
    boneToModify {boneName=dyn_rearA_01_l}
    rotation {yaw=-40}
    scale=<0.95 0.95 0.95>
    translation=<-30 -10 -15>
   }
   n26 = "Transform (Modify) Bone - Bone: dyn_shorts_main_r"(ComponentPose=@n25, Alpha=@n3, Translation="0,0,0", Rotation="0,0,0", Scale="1,1,1") {
    rotationMode=BMM_Additive
    rotationSpace=BCS_BoneSpace
    scaleMode=BMM_Additive
    scaleSpace=BCS_BoneSpace
    translationMode=BMM_Additive
    translationSpace=BCS_BoneSpace
    alphaBoolBlend {blendOption=Linear}
    alphaScaleBiasClamp {bInterpResult=true, interpSpeedDecreasing=25, interpSpeedIncreasing=25}
    boneToModify {boneName=dyn_shorts_main_r}
   }
   n27 = "Transform (Modify) Bone - Bone: dyn_shorts_frtA_01_r"(ComponentPose=@n26, Alpha=1, Translation="0,0,0", Rotation="-5,-5,0.000", Scale="1,1,1") {
    rotationMode=BMM_Additive
    rotationSpace=BCS_BoneSpace
    scaleMode=BMM_Additive
    scaleSpace=BCS_BoneSpace
    translationMode=BMM_Additive
    translationSpace=BCS_BoneSpace
    alphaBoolBlend {blendOption=Linear}
    alphaScaleBiasClamp {bInterpResult=true, interpSpeedDecreasing=25, interpSpeedIncreasing=25}
    boneToModify {boneName=dyn_shorts_frtA_01_r}
    rotation {pitch=-5, yaw=-5}
   }
   n28 = "Transform (Modify) Bone - Bone: dyn_shorts_frt_01_r"(ComponentPose=@n27, Alpha=1, Translation="0,0,0", Rotation="0,-10,0.000", Scale="1,1,1") {
    rotationMode=BMM_Additive
    rotationSpace=BCS_BoneSpace
    scaleMode=BMM_Additive
    scaleSpace=BCS_BoneSpace
    translationMode=BMM_Additive
    translationSpace=BCS_BoneSpace
    alphaBoolBlend {blendOption=Linear}
    alphaScaleBiasClamp {bInterpResult=true, interpSpeedDecreasing=25, interpSpeedIncreasing=25}
    boneToModify {boneName=dyn_shorts_frt_01_r}
    rotation {yaw=-10}
   }
   n29 = "Transform (Modify) Bone - Bone: dyn_shorts_outA_01_r"(ComponentPose=@n28, Alpha=@n3, Translation="0,0,0", Rotation="-10,-10,0.000", Scale="0.9,0.9,0.9") {
    rotationMode=BMM_Additive
    rotationSpace=BCS_BoneSpace
    scaleMode=BMM_Additive
    scaleSpace=BCS_BoneSpace
    translationMode=BMM_Additive
    translationSpace=BCS_BoneSpace
    alphaBoolBlend {blendOption=Linear}
    alphaScaleBiasClamp {bInterpResult=true, interpSpeedDecreasing=25, interpSpeedIncreasing=25}
    boneToModify {boneName=dyn_shorts_outA_01_r}
    rotation {pitch=-10, yaw=-10}
    scale=<0.9 0.9 0.9>
   }
   n30 = "Transform (Modify) Bone - Bone: dyn_shorts_outB_01_r"(ComponentPose=@n29, Alpha=@n3, Translation="5,1,1", Rotation="-5,-30,0.000", Scale="1,1,1") {
    rotationMode=BMM_Additive
    rotationSpace=BCS_BoneSpace
    scaleMode=BMM_Additive
    scaleSpace=BCS_BoneSpace
    translationMode=BMM_Additive
    translationSpace=BCS_BoneSpace
    alphaBoolBlend {blendOption=Linear}
    alphaScaleBiasClamp {bInterpResult=true, interpSpeedDecreasing=25, interpSpeedIncreasing=25}
    boneToModify {boneName=dyn_shorts_outB_01_r}
    rotation {pitch=-5, yaw=-30}
    translation=<5 1 1>
   }
   n31 = "Transform (Modify) Bone - Bone: dyn_shorts_outB_02_r"(ComponentPose=@n30, Alpha=@n3, Translation="5,1,1", Rotation="-5,0,0.000", Scale="1,1,1") {
    rotationMode=BMM_Additive
    rotationSpace=BCS_BoneSpace
    scaleMode=BMM_Additive
    scaleSpace=BCS_BoneSpace
    translationMode=BMM_Additive
    translationSpace=BCS_BoneSpace
    alphaBoolBlend {blendOption=Linear}
    alphaScaleBiasClamp {bInterpResult=true, interpSpeedDecreasing=25, interpSpeedIncreasing=25}
    boneToModify {boneName=dyn_shorts_outB_02_r}
    rotation {pitch=-5}
    translation=<5 1 1>
   }
   n32 = "Transform (Modify) Bone - Bone: dyn_shorts_bkA_01_r"(ComponentPose=@n31, Alpha=@n3, Translation="12,5,0", Rotation="-3,-30,0.000", Scale="1,1,1") {
    rotationMode=BMM_Additive
    rotationSpace=BCS_BoneSpace
    scaleMode=BMM_Additive
    scaleSpace=BCS_BoneSpace
    translationMode=BMM_Additive
    translationSpace=BCS_BoneSpace
    alphaBoolBlend {blendOption=Linear}
    alphaScaleBiasClamp {bInterpResult=true, interpSpeedDecreasing=25, interpSpeedIncreasing=25}
    boneToModify {boneName=dyn_shorts_bkA_01_r}
    rotation {pitch=-3, yaw=-30}
    translation=<12 5>
   }
   n33 = "Transform (Modify) Bone - Bone: dyn_shorts_bkA_02_r"(ComponentPose=@n32, Alpha=@n3, Translation="0,0,0", Rotation="-10,-20,0.000", Scale="1,1,1") {
    rotationMode=BMM_Additive
    rotationSpace=BCS_BoneSpace
    scaleMode=BMM_Additive
    scaleSpace=BCS_BoneSpace
    translationMode=BMM_Additive
    translationSpace=BCS_BoneSpace
    alphaBoolBlend {blendOption=Linear}
    alphaScaleBiasClamp {bInterpResult=true, interpSpeedDecreasing=25, interpSpeedIncreasing=25}
    boneToModify {boneName=dyn_shorts_bkA_02_r}
    rotation {pitch=-10, yaw=-20}
   }
   n34 = "Transform (Modify) Bone - Bone: dyn_shorts_bkB_01_r"(ComponentPose=@n33, Alpha=@n3, Translation="14,5,2", Rotation="-20,-25,-20", Scale="1,1,1") {
    rotationMode=BMM_Additive
    rotationSpace=BCS_BoneSpace
    scaleMode=BMM_Additive
    scaleSpace=BCS_BoneSpace
    translationMode=BMM_Additive
    translationSpace=BCS_BoneSpace
    alphaBoolBlend {blendOption=Linear}
    alphaScaleBiasClamp {bInterpResult=true, interpSpeedDecreasing=25, interpSpeedIncreasing=25}
    boneToModify {boneName=dyn_shorts_bkB_01_r}
    rotation=<rot -20 -25 -20>
    translation=<14 5 2>
   }
   n35 = "Transform (Modify) Bone - Bone: dyn_rearA_01_r"(ComponentPose=@n34, Alpha=@n3, Translation="12,0,2", Rotation="-5,-15,-15", Scale="1,1,1") {
    rotationMode=BMM_Additive
    rotationSpace=BCS_BoneSpace
    scaleMode=BMM_Additive
    scaleSpace=BCS_BoneSpace
    translationMode=BMM_Additive
    translationSpace=BCS_BoneSpace
    alphaBoolBlend {blendOption=Linear}
    alphaScaleBiasClamp {bInterpResult=true, interpSpeedDecreasing=25, interpSpeedIncreasing=25}
    boneToModify {boneName=dyn_rearA_01_r}
    rotation=<rot -5 -15 -15>
    translation {x=12, z=2}
   }
   n36 = RigidBody(ComponentPose=@n35) {
    bClampLinearTranslationLimitToRefPose=true
    bOverrideWorldGravity=true
    overridePhysicsAsset=&M_MED_DJ_Remix_Physics
    simulationSpace=BaseBoneSpace
    alphaBoolBlend {blendOption=Linear}
    baseBoneRef {boneName=spine_05}
    componentAppliedLinearAccClamp=<20000 20000 1300>
    componentLinearAccScale=<1 1 1>
    componentLinearVelScale=<1 1 1>
    overrideWorldGravity {z=-980}
   }
   n37 = "Transform (Modify) Bone - Bone: dyn_parent_collar_r"(ComponentPose=@n36, Alpha=Get arm_up_r() -> (arm_up_r:double), Translation="3,0,0", Rotation="-23.243,-15.9,6.75", Scale="0.8,0.8,0.8") {
    rotationMode=BMM_Additive
    rotationSpace=BCS_BoneSpace
    scaleMode=BMM_Additive
    scaleSpace=BCS_BoneSpace
    translationMode=BMM_Additive
    translationSpace=BCS_BoneSpace
    alphaBoolBlend {blendOption=Linear}
    alphaScaleBiasClamp {bInterpResult=true, interpSpeedDecreasing=15, interpSpeedIncreasing=15}
    boneToModify {boneName=dyn_parent_collar_r}
    rotation=<rot -23.243 -15.9 6.75>
    scale=<0.8 0.8 0.8>
    translation {x=3}
   }
   n38 = "Transform (Modify) Bone - Bone: dyn_parent_collar_l"(ComponentPose=@n37, Alpha=Get arm_up_l() -> (arm_up_l:double), Translation="-3,0,0", Rotation="25,0.000,0.000", Scale="1,1,1") {
    rotationMode=BMM_Additive
    rotationSpace=BCS_BoneSpace
    translationMode=BMM_Additive
    translationSpace=BCS_BoneSpace
    alphaBoolBlend {blendOption=Linear}
    alphaScaleBiasClamp {bInterpResult=true, interpSpeedDecreasing=15, interpSpeedIncreasing=15}
    boneToModify {boneName=dyn_parent_collar_l}
    rotation {pitch=25}
    translation {x=-3}
   }
   n39 = "Transform (Modify) Bone - Bone: dyn_rapel_r"(
    ComponentPose=@n38
    Alpha=Get arm_up_low_r() -> (arm_up_low_r:double)
    Translation="1.5,-3,0"
    Rotation="0.000,-30,0.000"
    Scale="0.8,0.8,0.8"
   ) {
    rotationMode=BMM_Additive
    rotationSpace=BCS_BoneSpace
    scaleMode=BMM_Additive
    scaleSpace=BCS_BoneSpace
    translationMode=BMM_Additive
    translationSpace=BCS_BoneSpace
    alphaBoolBlend {blendOption=Linear}
    alphaScaleBiasClamp {bInterpResult=true, interpSpeedDecreasing=15, interpSpeedIncreasing=15}
    boneToModify {boneName=dyn_rapel_r}
    rotation {yaw=-30}
    scale=<0.8 0.8 0.8>
    translation=<1.5 -3>
   }
   n40 = "Transform (Modify) Bone - Bone: dyn_lapel_l"(ComponentPose=@n39, Alpha=Get arm_up_low_l() -> (arm_up_low_l:double), Translation="-1,0,0", Rotation="0,-30,0.000", Scale="0.9,0.9,0.9") {
    rotationMode=BMM_Additive
    rotationSpace=BCS_BoneSpace
    translationMode=BMM_Additive
    translationSpace=BCS_BoneSpace
    alphaBoolBlend {blendOption=Linear}
    alphaScaleBiasClamp {bInterpResult=true, interpSpeedDecreasing=15, interpSpeedIncreasing=15}
    boneToModify {boneName=dyn_lapel_l}
    rotation {yaw=-30}
    scale=<0.9 0.9 0.9>
    translation {x=-1}
   }
   n41 = "Transform (Modify) Bone - Bone: dyn_collar_01_bck"(ComponentPose=@n40, Alpha=1, Translation="0,0,0", Rotation="0,-35,0", Scale="1,1,1") {
    rotationMode=BMM_Additive
    rotationSpace=BCS_BoneSpace
    scaleMode=BMM_Additive
    translationMode=BMM_Additive
    translationSpace=BCS_BoneSpace
    alphaBoolBlend {blendOption=Linear}
    alphaScaleBiasClamp {interpSpeedDecreasing=15, interpSpeedIncreasing=15}
    boneToModify {boneName=dyn_collar_01_bck}
    rotation {yaw=-35}
   }
   n42 = "Constraint - dyn_parent_collar_bck (R)"(ComponentPose=@n41, Alpha=1, ConstraintWeights_0=1) {
    alphaBoolBlend {blendOption=Linear}
    boneToModify {boneName=dyn_parent_collar_bck}
    constraintSetup [{transformType=Rotation, targetBone {boneName=head}}]
    constraintWeights [1]
   }
   n43 = "Constraint - dyn_shorts_frt_01_l (PP)"!(
    ComponentPose=Local To Component(LocalPose=Use cached pose 'Anim_Cache'())
    Alpha=1
    ConstraintWeights_0=0.0635
    ConstraintWeights_1:float=0.9365
   ) -> (Pose) {
    alphaBoolBlend {blendOption=Linear}
    boneToModify {boneName=dyn_shorts_frt_01_l}
    constraintSetup [{targetBone {boneName=pelvis}}, {targetBone {boneName=thigh_twist_01_l}}]
    constraintWeights [1,1]
   }
   n44 = "Constraint - dyn_shorts_frt_02_l (PP)"!(ComponentPose=@n43, Alpha=1, ConstraintWeights_0=0.8695, ConstraintWeights_1:float=0.1305) -> (Pose) {
    alphaBoolBlend {blendOption=Linear}
    boneToModify {boneName=dyn_shorts_frt_02_l}
    constraintSetup [{targetBone {boneName=thigh_twist_01_l}}, {targetBone {boneName=thigh_l}}]
    constraintWeights [1,1]
   }
   n45 = "Constraint - dyn_shorts_frt_03_l (PPP)"!(ComponentPose=@n44, Alpha=1, ConstraintWeights_0=0.0136, ConstraintWeights_1:float=0.4131, ConstraintWeights_2:float=0.5733) -> (Pose) {
    alphaBoolBlend {blendOption=Linear}
    boneToModify {boneName=dyn_shorts_frt_03_l}
    constraintSetup [{targetBone {boneName=calf_l}}, {targetBone {boneName=thigh_twist_01_l}}, {targetBone {boneName=thigh_l}}]
    constraintWeights [1,1,1]
   }
   n46 = "Constraint - dyn_shorts_outA_01_l (PP)"!(ComponentPose=@n45, Alpha=1, ConstraintWeights_0=0.0835, ConstraintWeights_1:float=0.9165) -> (Pose) {
    alphaBoolBlend {blendOption=Linear}
    boneToModify {boneName=dyn_shorts_outA_01_l}
    constraintSetup [{targetBone {boneName=pelvis}}, {targetBone {boneName=thigh_twist_01_l}}]
    constraintWeights [1,1]
   }
   n47 = "Constraint - dyn_shorts_outA_02_l (PP)"!(ComponentPose=@n46, Alpha=1, ConstraintWeights_0=0.8996, ConstraintWeights_1:float=0.1004) -> (Pose) {
    alphaBoolBlend {blendOption=Linear}
    boneToModify {boneName=dyn_shorts_outA_02_l}
    constraintSetup [{targetBone {boneName=thigh_twist_01_l}}, {targetBone {boneName=thigh_l}}]
    constraintWeights [1,1]
   }
   n48 = "Constraint - dyn_shorts_outB_01_l (PPP)"!(ComponentPose=@n47, Alpha=1, ConstraintWeights_0=0.256, ConstraintWeights_1:float=0.7212, ConstraintWeights_2:float=0.0227) -> (Pose) {
    alphaBoolBlend {blendOption=Linear}
    boneToModify {boneName=dyn_shorts_outB_01_l}
    constraintSetup [{targetBone {boneName=pelvis}}, {targetBone {boneName=thigh_twist_01_l}}, {targetBone {boneName=thigh_l}}]
    constraintWeights [1,1,1]
   }
   n49 = "Constraint - dyn_shorts_outB_02_l (PPP)"!(ComponentPose=@n48, Alpha=1, ConstraintWeights_0=0.0284, ConstraintWeights_1:float=0.8547, ConstraintWeights_2:float=0.1168) -> (Pose) {
    alphaBoolBlend {blendOption=Linear}
    boneToModify {boneName=dyn_shorts_outB_02_l}
    constraintSetup [{targetBone {boneName=pelvis}}, {targetBone {boneName=thigh_twist_01_l}}, {targetBone {boneName=thigh_l}}]
    constraintWeights [1,1,1]
   }
   n50 = "Constraint - dyn_shorts_bkA_01_l (PP)"!(ComponentPose=@n49, Alpha=1, ConstraintWeights_0=0.5239, ConstraintWeights_1:float=0.4761) -> (Pose) {
    alphaBoolBlend {blendOption=Linear}
    boneToModify {boneName=dyn_shorts_bkA_01_l}
    constraintSetup [{targetBone {boneName=pelvis}}, {targetBone {boneName=thigh_twist_01_l}}]
    constraintWeights [1,1]
   }
   n51 = "Constraint - dyn_shorts_bkA_02_l (PPP)"!(ComponentPose=@n50, Alpha=1, ConstraintWeights_0=0.1083, ConstraintWeights_1:float=0.7995, ConstraintWeights_2:float=0.0922) -> (Pose) {
    alphaBoolBlend {blendOption=Linear}
    boneToModify {boneName=dyn_shorts_bkA_02_l}
    constraintSetup [{targetBone {boneName=pelvis}}, {targetBone {boneName=thigh_twist_01_l}}, {targetBone {boneName=thigh_l}}]
    constraintWeights [1,1,1]
   }
   n52 = "Constraint - dyn_shorts_bkA_03_l (PPP)"!(ComponentPose=@n51, Alpha=1, ConstraintWeights_0=0.0151, ConstraintWeights_1:float=0.4672, ConstraintWeights_2:float=0.5177) -> (Pose) {
    alphaBoolBlend {blendOption=Linear}
    boneToModify {boneName=dyn_shorts_bkA_03_l}
    constraintSetup [{targetBone {boneName=calf_l}}, {targetBone {boneName=thigh_twist_01_l}}, {targetBone {boneName=thigh_l}}]
    constraintWeights [1,1,1]
   }
   n53 = "Constraint - dyn_shorts_bkB_01_l (PPP)"!(ComponentPose=@n52, Alpha=1, ConstraintWeights_0=0.6926, ConstraintWeights_1:float=0.2834, ConstraintWeights_2:float=0.0241) -> (Pose) {
    alphaBoolBlend {blendOption=Linear}
    boneToModify {boneName=dyn_shorts_bkB_01_l}
    constraintSetup [{targetBone {boneName=pelvis}}, {targetBone {boneName=thigh_twist_01_l}}, {targetBone {boneName=thigh_twist_01_r}}]
    constraintWeights [1,1,1]
   }
   n54 = "Constraint - dyn_shorts_bkB_02_l (PPP)"!(ComponentPose=@n53, Alpha=1, ConstraintWeights_0=0.1152, ConstraintWeights_1:float=0.7834, ConstraintWeights_2:float=0.1014) -> (Pose) {
    alphaBoolBlend {blendOption=Linear}
    boneToModify {boneName=dyn_shorts_bkB_02_l}
    constraintSetup [{targetBone {boneName=pelvis}}, {targetBone {boneName=thigh_twist_01_l}}, {targetBone {boneName=thigh_l}}]
    constraintWeights [1,1,1]
   }
   n55 = "Constraint - dyn_shorts_bkB_03_l (PPPP)"!(
    ComponentPose=@n54
    Alpha=1
    ConstraintWeights_0=0.0197
    ConstraintWeights_1:float=0.5922
    ConstraintWeights_2:float=0.3768
    ConstraintWeights_3:float=0.0113
   ) -> (Pose) {
    alphaBoolBlend {blendOption=Linear}
    boneToModify {boneName=dyn_shorts_bkB_03_l}
    constraintSetup [{targetBone {boneName=pelvis}}, {targetBone {boneName=thigh_twist_01_l}}, {targetBone {boneName=thigh_l}}, {targetBone {boneName=calf_l}}]
    constraintWeights [1,1,1,1]
   }
   n56 = "Constraint - dyn_rearA_01_l (PPP)"!(ComponentPose=@n55, Alpha=1, ConstraintWeights_0=0.7459, ConstraintWeights_1:float=0.1577, ConstraintWeights_2:float=0.0964) -> (Pose) {
    alphaBoolBlend {blendOption=Linear}
    boneToModify {boneName=dyn_rearA_01_l}
    constraintSetup [{targetBone {boneName=pelvis}}, {targetBone {boneName=thigh_twist_01_l}}, {targetBone {boneName=thigh_twist_01_r}}]
    constraintWeights [1,1,1]
   }
   n57 = "Constraint - dyn_rearA_02_l (PPPP)"!(
    ComponentPose=@n56
    Alpha=1
    ConstraintWeights_0=0.2319
    ConstraintWeights_1:float=0.6936
    ConstraintWeights_2:float=0.0114
    ConstraintWeights_3:float=0.0631
   ) -> (Pose) {
    alphaBoolBlend {blendOption=Linear}
    boneToModify {boneName=dyn_rearA_02_l}
    constraintSetup [
     {targetBone {boneName=pelvis}}
     {targetBone {boneName=thigh_twist_01_l}}
     {targetBone {boneName=thigh_twist_01_r}}
     {targetBone {boneName=thigh_l}}
    ]
    constraintWeights [1,1,1,1]
   }
   n58 = "Constraint - dyn_rearA_03_l (PPPP)"!(
    ComponentPose=@n57
    Alpha=1
    ConstraintWeights_0=0.0257
    ConstraintWeights_1:float=0.5662
    ConstraintWeights_2:float=0.3927
    ConstraintWeights_3:float=0.0154
   ) -> (Pose) {
    alphaBoolBlend {blendOption=Linear}
    boneToModify {boneName=dyn_rearA_03_l}
    constraintSetup [{targetBone {boneName=pelvis}}, {targetBone {boneName=thigh_twist_01_l}}, {targetBone {boneName=thigh_l}}, {targetBone {boneName=calf_l}}]
    constraintWeights [1,1,1,1]
   }
   n59 = "Constraint - dyn_shorts_frt_01_r (P)"(ComponentPose=@n58, Alpha=1, ConstraintWeights_0=1) {
    alphaBoolBlend {blendOption=Linear}
    boneToModify {boneName=dyn_shorts_frt_01_r}
    constraintSetup [{targetBone {boneName=thigh_r}}]
    constraintWeights [1]
   }
   n60 = "Constraint - dyn_shorts_frt_02_r (P)"(ComponentPose=@n59, Alpha=1, ConstraintWeights_0=1) {
    alphaBoolBlend {blendOption=Linear}
    boneToModify {boneName=dyn_shorts_frt_02_r}
    constraintSetup [{targetBone {boneName=thigh_r}}]
    constraintWeights [1]
   }
   n61 = "Constraint - dyn_shorts_frt_03_r (P)"(ComponentPose=@n60, Alpha=1, ConstraintWeights_0=1) {
    alphaBoolBlend {blendOption=Linear}
    boneToModify {boneName=dyn_shorts_frt_03_r}
    constraintSetup [{targetBone {boneName=thigh_r}}]
    constraintWeights [1]
   }
   n62 = "Constraint - dyn_shorts_outA_01_r (PP)"!(ComponentPose=@n61, Alpha=1, ConstraintWeights_0=1, ConstraintWeights_1:float=0) -> (Pose) {
    alphaBoolBlend {blendOption=Linear}
    boneToModify {boneName=dyn_shorts_outA_01_r}
    constraintSetup [{targetBone {boneName=thigh_r}}, {targetBone {boneName=pelvis}}]
    constraintWeights [1,1]
   }
   n63 = "Constraint - dyn_shorts_outA_02_r (P)"(ComponentPose=@n62, Alpha=1, ConstraintWeights_0=1) {
    alphaBoolBlend {blendOption=Linear}
    boneToModify {boneName=dyn_shorts_outA_02_r}
    constraintSetup [{targetBone {boneName=thigh_r}}]
    constraintWeights [1]
   }
   n64 = "Constraint - dyn_shorts_outB_01_r (PP)"!(ComponentPose=@n63, Alpha=1, ConstraintWeights_0=1, ConstraintWeights_1:float=0) -> (Pose) {
    alphaBoolBlend {blendOption=Linear}
    boneToModify {boneName=dyn_shorts_outB_01_r}
    constraintSetup [{targetBone {boneName=thigh_r}}, {targetBone {boneName=pelvis}}]
    constraintWeights [1,1]
   }
   n65 = "Constraint - dyn_shorts_outB_02_r (P)"(ComponentPose=@n64, Alpha=1, ConstraintWeights_0=1) {
    alphaBoolBlend {blendOption=Linear}
    boneToModify {boneName=dyn_shorts_outB_02_r}
    constraintSetup [{targetBone {boneName=thigh_r}}]
    constraintWeights [1]
   }
   n66 = "Constraint - dyn_shorts_bkA_01_r (PP)"!(ComponentPose=@n65, Alpha=1, ConstraintWeights_0=0.5, ConstraintWeights_1:float=0.5) -> (Pose) {
    alphaBoolBlend {blendOption=Linear}
    boneToModify {boneName=dyn_shorts_bkA_01_r}
    constraintSetup [{targetBone {boneName=pelvis}}, {targetBone {boneName=thigh_r}}]
    constraintWeights [1,1]
   }
   n67 = "Constraint - dyn_shorts_bkA_02_r (P)"(ComponentPose=@n66, Alpha=1, ConstraintWeights_0=1) {
    alphaBoolBlend {blendOption=Linear}
    boneToModify {boneName=dyn_shorts_bkA_02_r}
    constraintSetup [{targetBone {boneName=thigh_r}}]
    constraintWeights [1]
   }
   n68 = "Constraint - dyn_shorts_bkA_03_r (P)"(ComponentPose=@n67, Alpha=1, ConstraintWeights_0=1) {
    alphaBoolBlend {blendOption=Linear}
    boneToModify {boneName=dyn_shorts_bkA_03_r}
    constraintSetup [{targetBone {boneName=thigh_r}}]
    constraintWeights [1]
   }
   n69 = "Constraint - dyn_shorts_bkB_01_r (PP)"!(ComponentPose=@n68, Alpha=1, ConstraintWeights_0=0.6, ConstraintWeights_1:float=0.4) -> (Pose) {
    alphaBoolBlend {blendOption=Linear}
    boneToModify {boneName=dyn_shorts_bkB_01_r}
    constraintSetup [{targetBone {boneName=pelvis}}, {targetBone {boneName=thigh_l}}]
    constraintWeights [1,1]
   }
   n70 = "Constraint - dyn_shorts_bkB_02_r (P)"(ComponentPose=@n69, Alpha=1, ConstraintWeights_0=1) {
    alphaBoolBlend {blendOption=Linear}
    boneToModify {boneName=dyn_shorts_bkB_02_r}
    constraintSetup [{targetBone {boneName=thigh_r}}]
    constraintWeights [1]
   }
   n71 = "Constraint - dyn_shorts_bkB_03_r (P)"(ComponentPose=@n70, Alpha=1, ConstraintWeights_0=1) {
    alphaBoolBlend {blendOption=Linear}
    boneToModify {boneName=dyn_shorts_bkB_03_r}
    constraintSetup [{targetBone {boneName=thigh_r}}]
    constraintWeights [1]
   }
   n72 = "Constraint - dyn_rearA_01_r (PP)"!(ComponentPose=@n71, Alpha=1, ConstraintWeights_0=0.7, ConstraintWeights_1:float=0.3) -> (Pose) {
    alphaBoolBlend {blendOption=Linear}
    boneToModify {boneName=dyn_rearA_01_r}
    constraintSetup [{targetBone {boneName=pelvis}}, {targetBone {boneName=thigh_l}}]
    constraintWeights [1,1]
   }
   n73 = "Constraint - dyn_rearA_02_r (P)"(ComponentPose=@n72, Alpha=1, ConstraintWeights_0=1) {
    alphaBoolBlend {blendOption=Linear}
    boneToModify {boneName=dyn_rearA_02_r}
    constraintSetup [{targetBone {boneName=thigh_r}}]
    constraintWeights [1]
   }
   n74 = "Constraint - dyn_rearA_03_r (P)"(ComponentPose=@n73, Alpha=1, ConstraintWeights_0=1) {
    alphaBoolBlend {blendOption=Linear}
    boneToModify {boneName=dyn_rearA_03_r}
    constraintSetup [{targetBone {boneName=thigh_r}}]
    constraintWeights [1]
   }
   n75 = Control Rig(Source=Component To Local(ComponentPose=@n74)) {
    alphaBoolBlend {blendOption=Linear}
    controlRigAssetReference {blueprintRigClass=&/M_MED_DJ_Remix_ControlRig.M_MED_DJ_Remix_ControlRig_C}
   }
   Output Pose(Result=Blend Poses by bool(BlendPose_0=@n75, BlendPose_1=Component To Local(ComponentPose=@n42), bActiveValue=Get rigids_off() -> (rigids_off:bool)) {blendType=Linear}) {layerGroup=None}
   n76 = "Constraint - dyn_thigh_fix_l (P)"(ComponentPose=Local To Component(LocalPose=Copy Pose From Mesh() {bUseAttachedParent=true}), Alpha=1, ConstraintWeights_0=1) {
    alphaBoolBlend {blendOption=Linear}
    boneToModify {boneName=dyn_thigh_fix_l}
    constraintSetup [{targetBone {boneName=thigh_twist_01_l}}]
    constraintWeights [1]
   }
   n77 = "Constraint - dyn_thigh_fix_r (P)"(ComponentPose=@n76, Alpha=1, ConstraintWeights_0=1) {
    alphaBoolBlend {blendOption=Linear}
    boneToModify {boneName=dyn_thigh_fix_r}
    constraintSetup [{targetBone {boneName=thigh_twist_01_r}}]
    constraintWeights [1]
   }
   n78 = "Transform (Modify) Bone - Bone: dyn_thigh_fix_l"(ComponentPose=@n77, Alpha=Get thigh_up_l() -> (thigh_up_l:double), Translation="3,4.5,-10.500", Rotation="0.000,0.000,-5", Scale="1,1,1") {
    rotationMode=BMM_Additive
    rotationSpace=BCS_BoneSpace
    scaleMode=BMM_Additive
    scaleSpace=BCS_BoneSpace
    translationMode=BMM_Additive
    translationSpace=BCS_BoneSpace
    alphaBoolBlend {blendOption=Linear}
    boneToModify {boneName=dyn_thigh_fix_l}
    rotation {roll=-5}
    translation=<3 4.5 -10.5>
   }
   n79 = "Transform (Modify) Bone - Bone: dyn_thigh_fix_r"(ComponentPose=@n78, Alpha=Get thigh_up_r() -> (thigh_up_r:double), Translation="-3,-4.5,10", Rotation="0.000,0.000,5", Scale="1,1,1") {
    rotationMode=BMM_Additive
    rotationSpace=BCS_BoneSpace
    scaleMode=BMM_Additive
    scaleSpace=BCS_BoneSpace
    translationMode=BMM_Additive
    translationSpace=BCS_BoneSpace
    alphaBoolBlend {blendOption=Linear}
    boneToModify {boneName=dyn_thigh_fix_r}
    rotation {roll=5}
    translation=<-3 -4.5 10>
   }
   Save cached pose 'Anim_Cache'(Pose=Component To Local(ComponentPose=@n79)) {cacheName=Anim_Cache}
  ]}
 EventGraph {nodes [
   Event BlueprintUpdateAnimation()
   > n1 = Sequence!(execute) -> (then_0, then_1, then_2:Exec, then_3:Exec, then_4:Exec, then_5:Exec, then_6:Exec, then_7:Exec, then_8:Exec)
   > n2 = Branch(Condition="float >= float"(A=Get PawnSpeed2D(), B=100) {is=title K2Node_CallFunction})
   > Set is_moving(is_moving:bool=true) -> (Output_Get:bool)
   Set is_moving(execute=@n2.else, is_moving=false)
   n3 = Branch(execute=@n1.then_3, Condition=Get bIsSkyDiving())
   n4 = Get SkydivingExternalForceInput()
   > Set skydiving_force(skydiving_force:Vector=vector * float(
     A=Break Fort Anim Input Skydiving External Force(FortAnimInput_SkydivingExternalForce=@n4) -> (
      bUseSkydivingVectorForce=~
      SkydivingVectorForce=~
      HeadToPelvisDirection=~
      FloatingMultiplier=~
      FloatingAdditive=~
      DivingMultiplier=~
      DivingAdditive=~
      ParachutingMultiplier=~
      ParachutingAdditive=~
     ).SkydivingVectorForce
     B=4
    )) -> (Output_Get:Vector)
   Set skydiving_force(execute=@n3.else, skydiving_force="0, 0, 0")
   n5 = BreakTransform(InTransform=GetDeltaTransformFromRefPose(self=GetOwnerSkeletalMeshComponent(), BoneName=calf_l, BaseName=thigh_l))
   Set calf_rot_back_l(execute=@n1.then_2, calf_rot_back_l:double=MapRangeClamped(Value=BreakRotator(InRot=@n5.Rotation).Yaw, InRangeA=15, InRangeB=115)) -> (Output_Get:double)
   > Branch(Condition=Get is_moving() -> (is_moving:bool))
   > Set calf_rot_back_l(calf_rot_back_l=float * float(A=Get calf_rot_back_l(), B=0.7))
   n6 = BreakTransform(InTransform=GetDeltaTransformFromRefPose(self=GetOwnerSkeletalMeshComponent(), BoneName=thigh_l, BaseName=pelvis))
   Set thigh_up_l(execute=@n1.then_8, thigh_up_l:double=MapRangeClamped(Value=BreakRotator(InRot=@n6.Rotation).Yaw, InRangeA=0, InRangeB=-85)) -> (Output_Get:double)
   n7 = Get CurrentLOD()
   n8 = Get bRigidBodyEnabled()
   Set rigids_off(execute=@n1.then_4, rigids_off:bool=OR Boolean(A=NOT Boolean(A=@n8), B="integer > integer"(A=@n7, B=1) {is=title K2Node_CallFunction})) -> (Output_Get:bool)
   > Set is_crouching(is_crouching:bool=Get bIsCrouching()) -> (Output_Get:bool)
   n9 = BreakTransform(InTransform=GetDeltaTransformFromRefPose(self=GetOwnerSkeletalMeshComponent(), BoneName=thigh_r, BaseName=pelvis))
   Set thigh_up_r(execute=@n1.then_7, thigh_up_r:double=MapRangeClamped(Value=BreakRotator(InRot=@n9.Rotation).Yaw, InRangeA=0, InRangeB=-85)) -> (Output_Get:double)
   n10 = GetDeltaTransformFromRefPose(self=GetOwnerSkeletalMeshComponent(), BoneName=upperarm_l, BaseName=spine_05)
   n11 = "Absolute (Float)"(A=BreakRotator(InRot=BreakTransform(InTransform=@n10).Rotation).Yaw)
   Set arm_up_l(execute=@n1.then_6, arm_up_l:double=MapRangeClamped(Value=@n11, InRangeA=50, InRangeB=105)) -> (Output_Get:double)
   > Set arm_up_low_l(arm_up_low_l:double=MapRangeClamped(Value=@n11, InRangeA=25, InRangeB=55)) -> (Output_Get:double)
   n12 = GetDeltaTransformFromRefPose(self=GetOwnerSkeletalMeshComponent(), BoneName=upperarm_r, BaseName=spine_05)
   n13 = "Absolute (Float)"(A=BreakRotator(InRot=BreakTransform(InTransform=@n12).Rotation).Yaw)
   Set arm_up_r(execute=@n1.then_5, arm_up_r:double=MapRangeClamped(Value=@n13, InRangeA=50, InRangeB=138)) -> (Output_Get:double)
   > Set arm_up_low_r(arm_up_low_r:double=MapRangeClamped(Value=@n13, InRangeA=10, InRangeB=70)) -> (Output_Get:double)
   n14 = BreakTransform(InTransform=GetDeltaTransformFromRefPose(self=GetOwnerSkeletalMeshComponent(), BoneName=calf_r, BaseName=thigh_r))
   Set calf_rot_back_r(execute=@n1.then_1, calf_rot_back_r:double=MapRangeClamped(Value=BreakRotator(InRot=@n14.Rotation).Yaw, InRangeA=15, InRangeB=115)) -> (Output_Get:double)
   > Branch(Condition=Get is_moving())
   > Set calf_rot_back_r(calf_rot_back_r=float * float(A=Get calf_rot_back_r(), B=0.65))
   Event OnParachuteOpened()
   > n15 = Delay(Duration=1.5)
   ResetDynamics(execute=[@n15.then, @n16.then, @n20.then, @n19.then], InTeleportType=TeleportPhysics)
   > ResetDynamics(InTeleportType=ResetPhysics)
   n16 = Event OnBeginSkydiving()
   Event OnEndTargeting() {is=title K2Node_Event}
   > Set targeting_alpha(targeting_alpha:double=0) -> (Output_Get:double)
   Event OnBeginTargeting() {is=title K2Node_Event}
   > n17 = Set targeting_alpha(targeting_alpha=1)
   n18 = Get bIsCrouching()
   n19 = Branch(execute=[@n21.then, @n17.then], Condition=AND Boolean(A=Get bIsTargeting(), B=NOT Boolean(A=@n18)))
   n20 = Event OnEndSkydiving()
   n21 = Event OnEndCrouching()
  ]}
]
variables [
 arm_up_l {type=number}
 arm_up_low_l {type=number}
 arm_up_low_r {type=number}
 arm_up_r {type=number}
 calf_rot_back_l {type=number}
 calf_rot_back_r {type=number}
 is_crouching {type=boolean}
 is_moving {type=boolean}
 rigids_off {type=boolean}
 Shorts_Alpha {type=number, default=0.6}
 skydiving_force {type=Vector}
 targeting_alpha {type=number}
 thigh_up_l {type=number}
 thigh_up_r {type=number}
]
