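# Physics asset for the Tier 2 DJ Remix medium male body cosmetic, covering 15 dynamic frills on each leg.
# Each frill capsule joins its own small parent box with a wide swing limit and tight twist limit.
# Parent boxes link through free-rotation constraints to a per-side frills root, then to the thigh and pelvis.
# Large thigh capsules give the legs volume alongside the frills.
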PhysicsAsset /BRCosmetics/Characters/Player/Male/Medium/Bodies/M_MED_DJ_Remix/Meshes/M_MED_DJ_Remix_Tier2_Physics
bodies {
 dyn_frill_01_l [{shapeType=Capsule, length=2.5, radius=0.6, center=<4 0 0>, rotation=<rot 90 0 0>}]
 dyn_frill_01_r [{shapeType=Capsule, length=2.5, radius=0.6, center=<4 0 0>, rotation=<rot 90 0 0>}]
 dyn_frill_02_l [{shapeType=Capsule, length=2.5, radius=0.6, center=<4 0 0>, rotation=<rot 90 0 0>}]
 dyn_frill_02_r [{shapeType=Capsule, length=2.5, radius=0.6, center=<4 0 0>, rotation=<rot 90 0 0>}]
 dyn_frill_03_l [{shapeType=Capsule, length=2.5, radius=0.6, center=<4 0 0>, rotation=<rot 90 0 0>}]
 dyn_frill_03_r [{shapeType=Capsule, length=2.5, radius=0.6, center=<4 0 0>, rotation=<rot 90 0 0>}]
 dyn_frill_04_l [{shapeType=Capsule, length=2.5, radius=0.6, center=<4 0 0>, rotation=<rot 90 0 0>}]
 dyn_frill_04_r [{shapeType=Capsule, length=2.5, radius=0.6, center=<4 0 0>, rotation=<rot 90 0 0>}]
 dyn_frill_05_l [{shapeType=Capsule, length=2.5, radius=0.6, center=<4 0 0>, rotation=<rot 90 0 0>}]
 dyn_frill_05_r [{shapeType=Capsule, length=2.5, radius=0.6, center=<4 0 0>, rotation=<rot 90 0 0>}]
 dyn_frill_06_l [{shapeType=Capsule, length=2.5, radius=0.6, center=<4 0 0>, rotation=<rot 90 0 0>}]
 dyn_frill_06_r [{shapeType=Capsule, length=2.5, radius=0.6, center=<4 0 0>, rotation=<rot 90 0 0>}]
 dyn_frill_07_l [{shapeType=Capsule, length=2.5, radius=0.6, center=<4 0 0>, rotation=<rot 90 0 0>}]
 dyn_frill_07_r [{shapeType=Capsule, length=2.5, radius=0.6, center=<4 0 0>, rotation=<rot 90 0 0>}]
 dyn_frill_08_l [{shapeType=Capsule, length=2.5, radius=0.6, center=<4 0 0>, rotation=<rot 90 0 0>}]
 dyn_frill_08_r [{shapeType=Capsule, length=2.5, radius=0.6, center=<4 0 0>, rotation=<rot 90 0 0>}]
 dyn_frill_09_l [{shapeType=Capsule, length=2.5, radius=0.6, center=<4 0 0>, rotation=<rot 90 0 0>}]
 dyn_frill_09_r [{shapeType=Capsule, length=2.5, radius=0.6, center=<4 0 0>, rotation=<rot 90 0 0>}]
 dyn_frill_10_l [{shapeType=Capsule, length=2.5, radius=0.6, center=<4 0 0>, rotation=<rot 90 0 0>}]
 dyn_frill_10_r [{shapeType=Capsule, length=2.5, radius=0.6, center=<4 0 0>, rotation=<rot 90 0 0>}]
 dyn_frill_11_l [{shapeType=Capsule, length=2.5, radius=0.6, center=<4 0 0>, rotation=<rot 90 0 0>}]
 dyn_frill_11_r [{shapeType=Capsule, length=2.5, radius=0.6, center=<4 0 0>, rotation=<rot 90 0 0>}]
 dyn_frill_12_l [{shapeType=Capsule, length=2.5, radius=0.6, center=<4 0 0>, rotation=<rot 90 0 0>}]
 dyn_frill_12_r [{shapeType=Capsule, length=2.5, radius=0.6, center=<4 0 0>, rotation=<rot 90 0 0>}]
 dyn_frill_13_l [{shapeType=Capsule, length=2.5, radius=0.6, center=<4 0 0>, rotation=<rot 90 0 0>}]
 dyn_frill_13_r [{shapeType=Capsule, length=2.5, radius=0.6, center=<4 0 0>, rotation=<rot 90 0 0>}]
 dyn_frill_14_l [{shapeType=Capsule, length=2.5, radius=0.6, center=<4 0 0>, rotation=<rot 90 0 0>}]
 dyn_frill_14_r [{shapeType=Capsule, length=2.5, radius=0.6, center=<4 0 0>, rotation=<rot 90 0 0>}]
 dyn_frill_15_l [{shapeType=Capsule, length=2.5, radius=0.6, center=<4 0 0>, rotation=<rot 90 0 0>}]
 dyn_frill_15_r [{shapeType=Capsule, length=2.5, radius=0.6, center=<4 0 0>, rotation=<rot 90 0 0>}]
 dyn_parent_frill_01_l [{shapeType=Box, extentX=1.01, extentY=1.01, extentZ=1.01, rotation=<rot 0 -90 0>}]
 dyn_parent_frill_01_r [{shapeType=Box, extentX=1.01, extentY=1.01, extentZ=1.01, rotation=<rot 0 -90 0>}]
 dyn_parent_frill_02_l [{shapeType=Box, extentX=1.01, extentY=1.01, extentZ=1.01, rotation=<rot 0 -90 0>}]
 dyn_parent_frill_02_r [{shapeType=Box, extentX=1.01, extentY=1.01, extentZ=1.01, rotation=<rot 0 -90 0>}]
 dyn_parent_frill_03_l [{shapeType=Box, extentX=1.01, extentY=1.01, extentZ=1.01, rotation=<rot 0 -90 0>}]
 dyn_parent_frill_03_r [{shapeType=Box, extentX=1.01, extentY=1.01, extentZ=1.01, rotation=<rot 0 -90 0>}]
 dyn_parent_frill_04_l [{shapeType=Box, extentX=1.01, extentY=1.01, extentZ=1.01, rotation=<rot 0 -90 0>}]
 dyn_parent_frill_04_r [{shapeType=Box, extentX=1.01, extentY=1.01, extentZ=1.01, rotation=<rot 0 -90 0>}]
 dyn_parent_frill_05_l [{shapeType=Box, extentX=1.01, extentY=1.01, extentZ=1.01, rotation=<rot 0 -90 0>}]
 dyn_parent_frill_05_r [{shapeType=Box, extentX=1.01, extentY=1.01, extentZ=1.01, rotation=<rot 0 -90 0>}]
 dyn_parent_frill_06_l [{shapeType=Box, extentX=1.01, extentY=1.01, extentZ=1.01, rotation=<rot 0 -90 0>}]
 dyn_parent_frill_06_r [{shapeType=Box, extentX=1.01, extentY=1.01, extentZ=1.01, rotation=<rot 0 -90 0>}]
 dyn_parent_frill_07_l [{shapeType=Box, extentX=1.01, extentY=1.01, extentZ=1.01, rotation=<rot 0 -90 0>}]
 dyn_parent_frill_07_r [{shapeType=Box, extentX=1.01, extentY=1.01, extentZ=1.01, rotation=<rot 0 -90 0>}]
 dyn_parent_frill_08_l [{shapeType=Box, extentX=1.01, extentY=1.01, extentZ=1.01, rotation=<rot 0 -90 0>}]
 dyn_parent_frill_08_r [{shapeType=Box, extentX=1.01, extentY=1.01, extentZ=1.01, rotation=<rot 0 -90 0>}]
 dyn_parent_frill_09_l [{shapeType=Box, extentX=1.01, extentY=1.01, extentZ=1.01, rotation=<rot 0 -90 0>}]
 dyn_parent_frill_09_r [{shapeType=Box, extentX=1.01, extentY=1.01, extentZ=1.01, rotation=<rot 0 -90 0>}]
 dyn_parent_frill_10_l [{shapeType=Box, extentX=1.01, extentY=1.01, extentZ=1.01, rotation=<rot 0 -90 0>}]
 dyn_parent_frill_10_r [{shapeType=Box, extentX=1.01, extentY=1.01, extentZ=1.01, rotation=<rot 0 -90 0>}]
 dyn_parent_frill_11_l [{shapeType=Box, extentX=1.01, extentY=1.01, extentZ=1.01, rotation=<rot 0 -90 0>}]
 dyn_parent_frill_11_r [{shapeType=Box, extentX=1.01, extentY=1.01, extentZ=1.01, rotation=<rot 0 -90 0>}]
 dyn_parent_frill_12_l [{shapeType=Box, extentX=1.01, extentY=1.01, extentZ=1.01, rotation=<rot 0 -90 0>}]
 dyn_parent_frill_12_r [{shapeType=Box, extentX=1.01, extentY=1.01, extentZ=1.01, rotation=<rot 0 -90 0>}]
 dyn_parent_frill_13_l [{shapeType=Box, extentX=1.01, extentY=1.01, extentZ=1.01, rotation=<rot 0 -90 0>}]
 dyn_parent_frill_13_r [{shapeType=Box, extentX=1.01, extentY=1.01, extentZ=1.01, rotation=<rot 0 -90 0>}]
 dyn_parent_frill_14_l [{shapeType=Box, extentX=1.01, extentY=1.01, extentZ=1.01, rotation=<rot 0 -90 0>}]
 dyn_parent_frill_14_r [{shapeType=Box, extentX=1.01, extentY=1.01, extentZ=1.01, rotation=<rot 0 -90 0>}]
 dyn_parent_frill_15_l [{shapeType=Box, extentX=1.01, extentY=1.01, extentZ=1.01, rotation=<rot 0 -90 0>}]
 dyn_parent_frill_15_r [{shapeType=Box, extentX=1.01, extentY=1.01, extentZ=1.01, rotation=<rot 0 -90 0>}]
 dyn_parent_frills_l [{shapeType=Box, extentX=1.01, extentY=1.01, extentZ=1.01, rotation=<rot 0 -90 0>}]
 dyn_parent_frills_r [{shapeType=Box, extentX=1.01, extentY=1.01, extentZ=1.01, rotation=<rot 0 -90 0>}]
 pelvis [{shapeType=Box, extentX=1, extentY=1, extentZ=1}]
 thigh_l [{shapeType=Capsule, length=8, radius=7.75, center=<-10.5 0.65 -0.22>, rotation=<rot -7.0728 -89.318 -88.1221>}]
 thigh_r [{shapeType=Capsule, length=8, radius=7.75, center=<10.5 -0.65 -0.22>, rotation=<rot -7.0728 -89.318 -88.1221>}]
}
constraints [
 {bone1Name=dyn_frill_01_l, bone2Name=dyn_parent_frill_01_l, swing1LimitDegrees=60, swing2LimitDegrees=60, twistLimitDegrees=5}
 {bone1Name=dyn_frill_02_l, bone2Name=dyn_parent_frill_02_l, swing1LimitDegrees=60, swing2LimitDegrees=60, twistLimitDegrees=5}
 {bone1Name=dyn_frill_03_l, bone2Name=dyn_parent_frill_03_l, swing1LimitDegrees=60, swing2LimitDegrees=60, twistLimitDegrees=5}
 {bone1Name=dyn_frill_04_l, bone2Name=dyn_parent_frill_04_l, swing1LimitDegrees=60, swing2LimitDegrees=60, twistLimitDegrees=5}
 {bone1Name=dyn_frill_05_l, bone2Name=dyn_parent_frill_05_l, swing1LimitDegrees=60, swing2LimitDegrees=60, twistLimitDegrees=5}
 {bone1Name=dyn_frill_06_l, bone2Name=dyn_parent_frill_06_l, swing1LimitDegrees=60, swing2LimitDegrees=60, twistLimitDegrees=5}
 {bone1Name=dyn_frill_07_l, bone2Name=dyn_parent_frill_07_l, swing1LimitDegrees=60, swing2LimitDegrees=60, twistLimitDegrees=5}
 {bone1Name=dyn_frill_08_l, bone2Name=dyn_parent_frill_08_l, swing1LimitDegrees=60, swing2LimitDegrees=60, twistLimitDegrees=5}
 {bone1Name=dyn_frill_09_l, bone2Name=dyn_parent_frill_09_l, swing1LimitDegrees=60, swing2LimitDegrees=60, twistLimitDegrees=5}
 {bone1Name=dyn_frill_10_l, bone2Name=dyn_parent_frill_10_l, swing1LimitDegrees=60, swing2LimitDegrees=60, twistLimitDegrees=5}
 {bone1Name=dyn_frill_11_l, bone2Name=dyn_parent_frill_11_l, swing1LimitDegrees=60, swing2LimitDegrees=60, twistLimitDegrees=5}
 {bone1Name=dyn_frill_12_l, bone2Name=dyn_parent_frill_12_l, swing1LimitDegrees=60, swing2LimitDegrees=60, twistLimitDegrees=5}
 {bone1Name=dyn_frill_13_l, bone2Name=dyn_parent_frill_13_l, swing1LimitDegrees=60, swing2LimitDegrees=60, twistLimitDegrees=5}
 {bone1Name=dyn_frill_14_l, bone2Name=dyn_parent_frill_14_l, swing1LimitDegrees=60, swing2LimitDegrees=60, twistLimitDegrees=5}
 {bone1Name=dyn_frill_15_l, bone2Name=dyn_parent_frill_15_l, swing1LimitDegrees=60, swing2LimitDegrees=60, twistLimitDegrees=5}
 {bone1Name=dyn_frill_01_r, bone2Name=dyn_parent_frill_01_r, swing1LimitDegrees=60, swing2LimitDegrees=60, twistLimitDegrees=5}
 {bone1Name=dyn_frill_02_r, bone2Name=dyn_parent_frill_02_r, swing1LimitDegrees=60, swing2LimitDegrees=60, twistLimitDegrees=5}
 {bone1Name=dyn_frill_03_r, bone2Name=dyn_parent_frill_03_r, swing1LimitDegrees=60, swing2LimitDegrees=60, twistLimitDegrees=5}
 {bone1Name=dyn_frill_04_r, bone2Name=dyn_parent_frill_04_r, swing1LimitDegrees=60, swing2LimitDegrees=60, twistLimitDegrees=5}
 {bone1Name=dyn_frill_05_r, bone2Name=dyn_parent_frill_05_r, swing1LimitDegrees=60, swing2LimitDegrees=60, twistLimitDegrees=5}
 {bone1Name=dyn_frill_06_r, bone2Name=dyn_parent_frill_06_r, swing1LimitDegrees=60, swing2LimitDegrees=60, twistLimitDegrees=5}
 {bone1Name=dyn_frill_07_r, bone2Name=dyn_parent_frill_07_r, swing1LimitDegrees=60, swing2LimitDegrees=60, twistLimitDegrees=5}
 {bone1Name=dyn_frill_08_r, bone2Name=dyn_parent_frill_08_r, swing1LimitDegrees=60, swing2LimitDegrees=60, twistLimitDegrees=5}
 {bone1Name=dyn_frill_09_r, bone2Name=dyn_parent_frill_09_r, swing1LimitDegrees=60, swing2LimitDegrees=60, twistLimitDegrees=5}
 {bone1Name=dyn_frill_10_r, bone2Name=dyn_parent_frill_10_r, swing1LimitDegrees=60, swing2LimitDegrees=60, twistLimitDegrees=5}
 {bone1Name=dyn_frill_11_r, bone2Name=dyn_parent_frill_11_r, swing1LimitDegrees=60, swing2LimitDegrees=60, twistLimitDegrees=5}
 {bone1Name=dyn_frill_12_r, bone2Name=dyn_parent_frill_12_r, swing1LimitDegrees=60, swing2LimitDegrees=60, twistLimitDegrees=5}
 {bone1Name=dyn_frill_13_r, bone2Name=dyn_parent_frill_13_r, swing1LimitDegrees=60, swing2LimitDegrees=60, twistLimitDegrees=5}
 {bone1Name=dyn_frill_14_r, bone2Name=dyn_parent_frill_14_r, swing1LimitDegrees=60, swing2LimitDegrees=60, twistLimitDegrees=5}
 {bone1Name=dyn_frill_15_r, bone2Name=dyn_parent_frill_15_r, swing1LimitDegrees=60, swing2LimitDegrees=60, twistLimitDegrees=5}
 {
  bone1Name=dyn_parent_frill_01_l
  bone2Name=dyn_parent_frills_l
  swing1LimitDegrees=45
  swing1Motion=Free
  swing2LimitDegrees=45
  swing2Motion=Free
  twistLimitDegrees=45
  twistMotion=Free
 }
 {
  bone1Name=dyn_parent_frill_02_l
  bone2Name=dyn_parent_frills_l
  swing1LimitDegrees=45
  swing1Motion=Free
  swing2LimitDegrees=45
  swing2Motion=Free
  twistLimitDegrees=45
  twistMotion=Free
 }
 {
  bone1Name=dyn_parent_frill_03_l
  bone2Name=dyn_parent_frills_l
  swing1LimitDegrees=45
  swing1Motion=Free
  swing2LimitDegrees=45
  swing2Motion=Free
  twistLimitDegrees=45
  twistMotion=Free
 }
 {
  bone1Name=dyn_parent_frill_04_l
  bone2Name=dyn_parent_frills_l
  swing1LimitDegrees=45
  swing1Motion=Free
  swing2LimitDegrees=45
  swing2Motion=Free
  twistLimitDegrees=45
  twistMotion=Free
 }
 {
  bone1Name=dyn_parent_frill_05_l
  bone2Name=dyn_parent_frills_l
  swing1LimitDegrees=45
  swing1Motion=Free
  swing2LimitDegrees=45
  swing2Motion=Free
  twistLimitDegrees=45
  twistMotion=Free
 }
 {
  bone1Name=dyn_parent_frill_06_l
  bone2Name=dyn_parent_frills_l
  swing1LimitDegrees=45
  swing1Motion=Free
  swing2LimitDegrees=45
  swing2Motion=Free
  twistLimitDegrees=45
  twistMotion=Free
 }
 {
  bone1Name=dyn_parent_frill_07_l
  bone2Name=dyn_parent_frills_l
  swing1LimitDegrees=45
  swing1Motion=Free
  swing2LimitDegrees=45
  swing2Motion=Free
  twistLimitDegrees=45
  twistMotion=Free
 }
 {
  bone1Name=dyn_parent_frill_08_l
  bone2Name=dyn_parent_frills_l
  swing1LimitDegrees=45
  swing1Motion=Free
  swing2LimitDegrees=45
  swing2Motion=Free
  twistLimitDegrees=45
  twistMotion=Free
 }
 {
  bone1Name=dyn_parent_frill_09_l
  bone2Name=dyn_parent_frills_l
  swing1LimitDegrees=45
  swing1Motion=Free
  swing2LimitDegrees=45
  swing2Motion=Free
  twistLimitDegrees=45
  twistMotion=Free
 }
 {
  bone1Name=dyn_parent_frill_10_l
  bone2Name=dyn_parent_frills_l
  swing1LimitDegrees=45
  swing1Motion=Free
  swing2LimitDegrees=45
  swing2Motion=Free
  twistLimitDegrees=45
  twistMotion=Free
 }
 {
  bone1Name=dyn_parent_frill_11_l
  bone2Name=dyn_parent_frills_l
  swing1LimitDegrees=45
  swing1Motion=Free
  swing2LimitDegrees=45
  swing2Motion=Free
  twistLimitDegrees=45
  twistMotion=Free
 }
 {
  bone1Name=dyn_parent_frill_12_l
  bone2Name=dyn_parent_frills_l
  swing1LimitDegrees=45
  swing1Motion=Free
  swing2LimitDegrees=45
  swing2Motion=Free
  twistLimitDegrees=45
  twistMotion=Free
 }
 {
  bone1Name=dyn_parent_frill_13_l
  bone2Name=dyn_parent_frills_l
  swing1LimitDegrees=45
  swing1Motion=Free
  swing2LimitDegrees=45
  swing2Motion=Free
  twistLimitDegrees=45
  twistMotion=Free
 }
 {
  bone1Name=dyn_parent_frill_14_l
  bone2Name=dyn_parent_frills_l
  swing1LimitDegrees=45
  swing1Motion=Free
  swing2LimitDegrees=45
  swing2Motion=Free
  twistLimitDegrees=45
  twistMotion=Free
 }
 {
  bone1Name=dyn_parent_frill_15_l
  bone2Name=dyn_parent_frills_l
  swing1LimitDegrees=45
  swing1Motion=Free
  swing2LimitDegrees=45
  swing2Motion=Free
  twistLimitDegrees=45
  twistMotion=Free
 }
 {
  bone1Name=dyn_parent_frill_01_r
  bone2Name=dyn_parent_frills_r
  swing1LimitDegrees=45
  swing1Motion=Free
  swing2LimitDegrees=45
  swing2Motion=Free
  twistLimitDegrees=45
  twistMotion=Free
 }
 {
  bone1Name=dyn_parent_frill_02_r
  bone2Name=dyn_parent_frills_r
  swing1LimitDegrees=45
  swing1Motion=Free
  swing2LimitDegrees=45
  swing2Motion=Free
  twistLimitDegrees=45
  twistMotion=Free
 }
 {
  bone1Name=dyn_parent_frill_03_r
  bone2Name=dyn_parent_frills_r
  swing1LimitDegrees=45
  swing1Motion=Free
  swing2LimitDegrees=45
  swing2Motion=Free
  twistLimitDegrees=45
  twistMotion=Free
 }
 {
  bone1Name=dyn_parent_frill_04_r
  bone2Name=dyn_parent_frills_r
  swing1LimitDegrees=45
  swing1Motion=Free
  swing2LimitDegrees=45
  swing2Motion=Free
  twistLimitDegrees=45
  twistMotion=Free
 }
 {
  bone1Name=dyn_parent_frill_05_r
  bone2Name=dyn_parent_frills_r
  swing1LimitDegrees=45
  swing1Motion=Free
  swing2LimitDegrees=45
  swing2Motion=Free
  twistLimitDegrees=45
  twistMotion=Free
 }
 {
  bone1Name=dyn_parent_frill_06_r
  bone2Name=dyn_parent_frills_r
  swing1LimitDegrees=45
  swing1Motion=Free
  swing2LimitDegrees=45
  swing2Motion=Free
  twistLimitDegrees=45
  twistMotion=Free
 }
 {
  bone1Name=dyn_parent_frill_07_r
  bone2Name=dyn_parent_frills_r
  swing1LimitDegrees=45
  swing1Motion=Free
  swing2LimitDegrees=45
  swing2Motion=Free
  twistLimitDegrees=45
  twistMotion=Free
 }
 {
  bone1Name=dyn_parent_frill_08_r
  bone2Name=dyn_parent_frills_r
  swing1LimitDegrees=45
  swing1Motion=Free
  swing2LimitDegrees=45
  swing2Motion=Free
  twistLimitDegrees=45
  twistMotion=Free
 }
 {
  bone1Name=dyn_parent_frill_09_r
  bone2Name=dyn_parent_frills_r
  swing1LimitDegrees=45
  swing1Motion=Free
  swing2LimitDegrees=45
  swing2Motion=Free
  twistLimitDegrees=45
  twistMotion=Free
 }
 {
  bone1Name=dyn_parent_frill_10_r
  bone2Name=dyn_parent_frills_r
  swing1LimitDegrees=45
  swing1Motion=Free
  swing2LimitDegrees=45
  swing2Motion=Free
  twistLimitDegrees=45
  twistMotion=Free
 }
 {
  bone1Name=dyn_parent_frill_11_r
  bone2Name=dyn_parent_frills_r
  swing1LimitDegrees=45
  swing1Motion=Free
  swing2LimitDegrees=45
  swing2Motion=Free
  twistLimitDegrees=45
  twistMotion=Free
 }
 {
  bone1Name=dyn_parent_frill_12_r
  bone2Name=dyn_parent_frills_r
  swing1LimitDegrees=45
  swing1Motion=Free
  swing2LimitDegrees=45
  swing2Motion=Free
  twistLimitDegrees=45
  twistMotion=Free
 }
 {
  bone1Name=dyn_parent_frill_13_r
  bone2Name=dyn_parent_frills_r
  swing1LimitDegrees=45
  swing1Motion=Free
  swing2LimitDegrees=45
  swing2Motion=Free
  twistLimitDegrees=45
  twistMotion=Free
 }
 {
  bone1Name=dyn_parent_frill_14_r
  bone2Name=dyn_parent_frills_r
  swing1LimitDegrees=45
  swing1Motion=Free
  swing2LimitDegrees=45
  swing2Motion=Free
  twistLimitDegrees=45
  twistMotion=Free
 }
 {
  bone1Name=dyn_parent_frill_15_r
  bone2Name=dyn_parent_frills_r
  swing1LimitDegrees=45
  swing1Motion=Free
  swing2LimitDegrees=45
  swing2Motion=Free
  twistLimitDegrees=45
  twistMotion=Free
 }
 {
  bone1Name=thigh_l
  bone2Name=pelvis
  swing1LimitDegrees=45
  swing1Motion=Free
  swing2LimitDegrees=45
  swing2Motion=Free
  twistLimitDegrees=45
  twistMotion=Free
 }
 {
  bone1Name=thigh_r
  bone2Name=pelvis
  swing1LimitDegrees=45
  swing1Motion=Free
  swing2LimitDegrees=45
  swing2Motion=Free
  twistLimitDegrees=45
  twistMotion=Free
 }
 {
  bone1Name=dyn_parent_frills_l
  bone2Name=thigh_l
  swing1LimitDegrees=45
  swing1Motion=Free
  swing2LimitDegrees=45
  swing2Motion=Free
  twistLimitDegrees=45
  twistMotion=Free
 }
 {
  bone1Name=dyn_parent_frills_r
  bone2Name=thigh_r
  swing1LimitDegrees=45
  swing1Motion=Free
  swing2LimitDegrees=45
  swing2Motion=Free
  twistLimitDegrees=45
  twistMotion=Free
 }
]
