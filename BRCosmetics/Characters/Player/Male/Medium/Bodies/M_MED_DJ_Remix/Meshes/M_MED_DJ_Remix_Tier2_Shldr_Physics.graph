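# Physics asset for the Tier 2 shoulder piece of the male medium DJ Remix player cosmetic, setting up dangling frills on both shoulders.
# Capsules on upper spine, clavicles and upper arms form the body shapes, linked by free-rotation constraints.
# Each side has seventeen frills, each a small sphere parent body hung freely from the clavicle.
# A capsule frill body joins each parent with wide swing and tight twist limits.
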PhysicsAsset /BRCosmetics/Characters/Player/Male/Medium/Bodies/M_MED_DJ_Remix/Meshes/M_MED_DJ_Remix_Tier2_Shldr_Physics
bodies {
 clavicle_l [{shapeType=Capsule, length=3.25, radius=7.25, center=<13.2244 -1.1702 -0.5965>, rotation=<rot 82.8856 12.4843 -0.9661>}]
 clavicle_r [{shapeType=Capsule, length=3.25, radius=7.25, center=<-13.2244 -1.1702 0.5965>, rotation=<rot 82.8856 12.4843 -0.9661>}]
 dyn_parent_shldr_frillA_01_l [{shapeType=Sphere, radius=0.505}]
 dyn_parent_shldr_frillA_01_r [{shapeType=Sphere, radius=0.505}]
 dyn_parent_shldr_frillB_01_l [{shapeType=Sphere, radius=0.505}]
 dyn_parent_shldr_frillB_01_r [{shapeType=Sphere, radius=0.505}]
 dyn_parent_shldr_frillC_01_l [{shapeType=Sphere, radius=0.505}]
 dyn_parent_shldr_frillC_01_r [{shapeType=Sphere, radius=0.505}]
 dyn_parent_shldr_frillD_01_l [{shapeType=Sphere, radius=0.505}]
 dyn_parent_shldr_frillD_01_r [{shapeType=Sphere, radius=0.505}]
 dyn_parent_shldr_frillE_01_l [{shapeType=Sphere, radius=0.505}]
 dyn_parent_shldr_frillE_01_r [{shapeType=Sphere, radius=0.505}]
 dyn_parent_shldr_frillF_01_l [{shapeType=Sphere, radius=0.505}]
 dyn_parent_shldr_frillF_01_r [{shapeType=Sphere, radius=0.505}]
 dyn_parent_shldr_frillG_01_l [{shapeType=Sphere, radius=0.505}]
 dyn_parent_shldr_frillG_01_r [{shapeType=Sphere, radius=0.505}]
 dyn_parent_shldr_frillH_01_l [{shapeType=Sphere, radius=0.505}]
 dyn_parent_shldr_frillH_01_r [{shapeType=Sphere, radius=0.505}]
 dyn_parent_shldr_frillI_01_l [{shapeType=Sphere, radius=0.505}]
 dyn_parent_shldr_frillI_01_r [{shapeType=Sphere, radius=0.505}]
 dyn_parent_shldr_frillJ_01_l [{shapeType=Sphere, radius=0.505}]
 dyn_parent_shldr_frillJ_01_r [{shapeType=Sphere, radius=0.505}]
 dyn_parent_shldr_frillK_01_l [{shapeType=Sphere, radius=0.505}]
 dyn_parent_shldr_frillK_01_r [{shapeType=Sphere, radius=0.505}]
 dyn_parent_shldr_frillL_01_l [{shapeType=Sphere, radius=0.505}]
 dyn_parent_shldr_frillL_01_r [{shapeType=Sphere, radius=0.505}]
 dyn_parent_shldr_frillM_01_l [{shapeType=Sphere, radius=0.505}]
 dyn_parent_shldr_frillM_01_r [{shapeType=Sphere, radius=0.505}]
 dyn_parent_shldr_frillN_01_l [{shapeType=Sphere, radius=0.505}]
 dyn_parent_shldr_frillN_01_r [{shapeType=Sphere, radius=0.505}]
 dyn_parent_shldr_frillO_01_l [{shapeType=Sphere, radius=0.505}]
 dyn_parent_shldr_frillO_01_r [{shapeType=Sphere, radius=0.505}]
 dyn_parent_shldr_frillP_01_l [{shapeType=Sphere, radius=0.505}]
 dyn_parent_shldr_frillP_01_r [{shapeType=Sphere, radius=0.505}]
 dyn_parent_shldr_frillQ_01_l [{shapeType=Sphere, radius=0.505}]
 dyn_parent_shldr_frillQ_01_r [{shapeType=Sphere, radius=0.505}]
 dyn_shldr_frillA_01_l [{shapeType=Capsule, length=1.2, radius=0.45, center=<4 0 0>, rotation=<rot 90 0 0>}]
 dyn_shldr_frillA_01_r [{shapeType=Capsule, length=1.2, radius=0.45, center=<-4 0 0>, rotation=<rot 90 0 0>}]
 dyn_shldr_frillB_01_l [{shapeType=Capsule, length=1.2, radius=0.45, center=<4 0 0>, rotation=<rot 90 0 0>}]
 dyn_shldr_frillB_01_r [{shapeType=Capsule, length=1.2, radius=0.45, center=<-4 0 0>, rotation=<rot 90 0 0>}]
 dyn_shldr_frillC_01_l [{shapeType=Capsule, length=1.2, radius=0.45, center=<4 0 0>, rotation=<rot 90 0 0>}]
 dyn_shldr_frillC_01_r [{shapeType=Capsule, length=1.2, radius=0.45, center=<-4 0 0>, rotation=<rot 90 0 0>}]
 dyn_shldr_frillD_01_l [{shapeType=Capsule, length=1.2, radius=0.45, center=<4 0 0>, rotation=<rot 90 0 0>}]
 dyn_shldr_frillD_01_r [{shapeType=Capsule, length=1.2, radius=0.45, center=<-4 0 0>, rotation=<rot 90 0 0>}]
 dyn_shldr_frillE_01_l [{shapeType=Capsule, length=1.2, radius=0.45, center=<4 0 0>, rotation=<rot 90 0 0>}]
 dyn_shldr_frillE_01_r [{shapeType=Capsule, length=1.2, radius=0.45, center=<-4 0 0>, rotation=<rot 90 0 0>}]
 dyn_shldr_frillF_01_l [{shapeType=Capsule, length=1.2, radius=0.45, center=<4 0 0>, rotation=<rot 90 0 0>}]
 dyn_shldr_frillF_01_r [{shapeType=Capsule, length=1.2, radius=0.45, center=<-4 0 0>, rotation=<rot 90 0 0>}]
 dyn_shldr_frillG_01_l [{shapeType=Capsule, length=1.2, radius=0.45, center=<4 0 0>, rotation=<rot 90 0 0>}]
 dyn_shldr_frillG_01_r [{shapeType=Capsule, length=1.2, radius=0.45, center=<-4 0 0>, rotation=<rot 90 0 0>}]
 dyn_shldr_frillH_01_l [{shapeType=Capsule, length=1.2, radius=0.45, center=<4 0 0>, rotation=<rot 90 0 0>}]
 dyn_shldr_frillH_01_r [{shapeType=Capsule, length=1.2, radius=0.45, center=<-4 0 0>, rotation=<rot 90 0 0>}]
 dyn_shldr_frillI_01_l [{shapeType=Capsule, length=1.2, radius=0.45, center=<4 0 0>, rotation=<rot 90 0 0>}]
 dyn_shldr_frillI_01_r [{shapeType=Capsule, length=1.2, radius=0.45, center=<-4 0 0>, rotation=<rot 90 0 0>}]
 dyn_shldr_frillJ_01_l [{shapeType=Capsule, length=1.2, radius=0.45, center=<4 0 0>, rotation=<rot 90 0 0>}]
 dyn_shldr_frillJ_01_r [{shapeType=Capsule, length=1.2, radius=0.45, center=<-4 0 0>, rotation=<rot 90 0 0>}]
 dyn_shldr_frillK_01_l [{shapeType=Capsule, length=1.2, radius=0.45, center=<4 0 0>, rotation=<rot 90 0 0>}]
 dyn_shldr_frillK_01_r [{shapeType=Capsule, length=1.2, radius=0.45, center=<-4 0 0>, rotation=<rot 90 0 0>}]
 dyn_shldr_frillL_01_l [{shapeType=Capsule, length=1.2, radius=0.45, center=<4 0 0>, rotation=<rot 90 0 0>}]
 dyn_shldr_frillL_01_r [{shapeType=Capsule, length=1.2, radius=0.45, center=<-4 0 0>, rotation=<rot 90 0 0>}]
 dyn_shldr_frillM_01_l [{shapeType=Capsule, length=1.2, radius=0.45, center=<4 0 0>, rotation=<rot 90 0 0>}]
 dyn_shldr_frillM_01_r [{shapeType=Capsule, length=1.2, radius=0.45, center=<-3.8079 0 0>, rotation=<rot 90 0 0>}]
 dyn_shldr_frillN_01_l [{shapeType=Capsule, length=1.2, radius=0.45, center=<3.7308 0 0>, rotation=<rot 90 0 0>}]
 dyn_shldr_frillN_01_r [{shapeType=Capsule, length=1.2, radius=0.45, center=<-3.7468 0 0>, rotation=<rot 90 0 0>}]
 dyn_shldr_frillO_01_l [{shapeType=Capsule, length=1.9, radius=0.5, center=<3.25 0 0>, rotation=<rot 90 0 0>}]
 dyn_shldr_frillO_01_r [{shapeType=Capsule, length=2.2, radius=0.5, center=<-3.3108 0 0>, rotation=<rot 90 0 0>}]
 dyn_shldr_frillP_01_l [{shapeType=Capsule, length=1.95, radius=0.5, center=<3.3648 0.1289 0>, rotation=<rot 90 0 0>}]
 dyn_shldr_frillP_01_r [{shapeType=Capsule, length=2.2, radius=0.5, center=<-3.3573 -0.0001 0>, rotation=<rot 90 0 0>}]
 dyn_shldr_frillQ_01_l [{shapeType=Capsule, length=2.2, radius=0.5, center=<3.3364 0.1989 0>, rotation=<rot 90 0 0>}]
 dyn_shldr_frillQ_01_r [{shapeType=Capsule, length=2.2, radius=0.5, center=<-3.4214 0 0>, rotation=<rot 90 0 0>}]
 spine_05 [{shapeType=Capsule, length=24, radius=14, center=<-0.2512 4.2771 0>}]
 upperarm_l [{shapeType=Capsule, length=5.983, radius=4.125, center=<10.6321 -0.3923 -2.9279>, rotation=<rot -0.41 -98.3335 -92.6955>}]
 upperarm_r [{shapeType=Capsule, length=10.483, radius=5.625, center=<-9.3769 2.4095 0.2789>, rotation=<rot 0 -98.3079 -88.3529>}]
}
constraints [
 {bone1Name=dyn_shldr_frillA_01_l, bone2Name=dyn_parent_shldr_frillA_01_l, swing1LimitDegrees=70, swing2LimitDegrees=70, twistLimitDegrees=5}
 {bone1Name=dyn_shldr_frillB_01_l, bone2Name=dyn_parent_shldr_frillB_01_l, swing1LimitDegrees=70, swing2LimitDegrees=70, twistLimitDegrees=5}
 {bone1Name=dyn_shldr_frillC_01_l, bone2Name=dyn_parent_shldr_frillC_01_l, swing1LimitDegrees=70, swing2LimitDegrees=70, twistLimitDegrees=5}
 {bone1Name=dyn_shldr_frillD_01_l, bone2Name=dyn_parent_shldr_frillD_01_l, swing1LimitDegrees=70, swing2LimitDegrees=70, twistLimitDegrees=5}
 {bone1Name=dyn_shldr_frillE_01_l, bone2Name=dyn_parent_shldr_frillE_01_l, swing1LimitDegrees=70, swing2LimitDegrees=70, twistLimitDegrees=5}
 {bone1Name=dyn_shldr_frillF_01_l, bone2Name=dyn_parent_shldr_frillF_01_l, swing1LimitDegrees=70, swing2LimitDegrees=70, twistLimitDegrees=5}
 {bone1Name=dyn_shldr_frillG_01_l, bone2Name=dyn_parent_shldr_frillG_01_l, swing1LimitDegrees=70, swing2LimitDegrees=70, twistLimitDegrees=5}
 {bone1Name=dyn_shldr_frillH_01_l, bone2Name=dyn_parent_shldr_frillH_01_l, swing1LimitDegrees=70, swing2LimitDegrees=70, twistLimitDegrees=5}
 {bone1Name=dyn_shldr_frillI_01_l, bone2Name=dyn_parent_shldr_frillI_01_l, swing1LimitDegrees=70, swing2LimitDegrees=70, twistLimitDegrees=5}
 {bone1Name=dyn_shldr_frillJ_01_l, bone2Name=dyn_parent_shldr_frillJ_01_l, swing1LimitDegrees=70, swing2LimitDegrees=70, twistLimitDegrees=5}
 {bone1Name=dyn_shldr_frillK_01_l, bone2Name=dyn_parent_shldr_frillK_01_l, swing1LimitDegrees=70, swing2LimitDegrees=70, twistLimitDegrees=5}
 {bone1Name=dyn_shldr_frillL_01_l, bone2Name=dyn_parent_shldr_frillL_01_l, swing1LimitDegrees=70, swing2LimitDegrees=70, twistLimitDegrees=5}
 {bone1Name=dyn_shldr_frillM_01_l, bone2Name=dyn_parent_shldr_frillM_01_l, swing1LimitDegrees=70, swing2LimitDegrees=70, twistLimitDegrees=5}
 {bone1Name=dyn_shldr_frillN_01_l, bone2Name=dyn_parent_shldr_frillN_01_l, swing1LimitDegrees=70, swing2LimitDegrees=70, twistLimitDegrees=5}
 {bone1Name=dyn_shldr_frillO_01_l, bone2Name=dyn_parent_shldr_frillO_01_l, swing1LimitDegrees=70, swing2LimitDegrees=70, twistLimitDegrees=5}
 {bone1Name=dyn_shldr_frillP_01_l, bone2Name=dyn_parent_shldr_frillP_01_l, swing1LimitDegrees=70, swing2LimitDegrees=70, twistLimitDegrees=5}
 {bone1Name=dyn_shldr_frillQ_01_l, bone2Name=dyn_parent_shldr_frillQ_01_l, swing1LimitDegrees=70, swing2LimitDegrees=70, twistLimitDegrees=5}
 {bone1Name=dyn_shldr_frillA_01_r, bone2Name=dyn_parent_shldr_frillA_01_r, swing1LimitDegrees=70, swing2LimitDegrees=70, twistLimitDegrees=5}
 {bone1Name=dyn_shldr_frillB_01_r, bone2Name=dyn_parent_shldr_frillB_01_r, swing1LimitDegrees=70, swing2LimitDegrees=70, twistLimitDegrees=5}
 {bone1Name=dyn_shldr_frillC_01_r, bone2Name=dyn_parent_shldr_frillC_01_r, swing1LimitDegrees=70, swing2LimitDegrees=70, twistLimitDegrees=5}
 {bone1Name=dyn_shldr_frillD_01_r, bone2Name=dyn_parent_shldr_frillD_01_r, swing1LimitDegrees=70, swing2LimitDegrees=70, twistLimitDegrees=5}
 {bone1Name=dyn_shldr_frillE_01_r, bone2Name=dyn_parent_shldr_frillE_01_r, swing1LimitDegrees=70, swing2LimitDegrees=70, twistLimitDegrees=5}
 {bone1Name=dyn_shldr_frillF_01_r, bone2Name=dyn_parent_shldr_frillF_01_r, swing1LimitDegrees=70, swing2LimitDegrees=70, twistLimitDegrees=5}
 {bone1Name=dyn_shldr_frillG_01_r, bone2Name=dyn_parent_shldr_frillG_01_r, swing1LimitDegrees=70, swing2LimitDegrees=70, twistLimitDegrees=5}
 {bone1Name=dyn_shldr_frillH_01_r, bone2Name=dyn_parent_shldr_frillH_01_r, swing1LimitDegrees=70, swing2LimitDegrees=70, twistLimitDegrees=5}
 {bone1Name=dyn_shldr_frillI_01_r, bone2Name=dyn_parent_shldr_frillI_01_r, swing1LimitDegrees=70, swing2LimitDegrees=70, twistLimitDegrees=5}
 {bone1Name=dyn_shldr_frillJ_01_r, bone2Name=dyn_parent_shldr_frillJ_01_r, swing1LimitDegrees=70, swing2LimitDegrees=70, twistLimitDegrees=5}
 {bone1Name=dyn_shldr_frillK_01_r, bone2Name=dyn_parent_shldr_frillK_01_r, swing1LimitDegrees=70, swing2LimitDegrees=70, twistLimitDegrees=5}
 {bone1Name=dyn_shldr_frillL_01_r, bone2Name=dyn_parent_shldr_frillL_01_r, swing1LimitDegrees=70, swing2LimitDegrees=70, twistLimitDegrees=5}
 {bone1Name=dyn_shldr_frillM_01_r, bone2Name=dyn_parent_shldr_frillM_01_r, swing1LimitDegrees=70, swing2LimitDegrees=70, twistLimitDegrees=5}
 {bone1Name=dyn_shldr_frillN_01_r, bone2Name=dyn_parent_shldr_frillN_01_r, swing1LimitDegrees=70, swing2LimitDegrees=70, twistLimitDegrees=5}
 {bone1Name=dyn_shldr_frillO_01_r, bone2Name=dyn_parent_shldr_frillO_01_r, swing1LimitDegrees=70, swing2LimitDegrees=70, twistLimitDegrees=5}
 {bone1Name=dyn_shldr_frillP_01_r, bone2Name=dyn_parent_shldr_frillP_01_r, swing1LimitDegrees=70, swing2LimitDegrees=70, twistLimitDegrees=5}
 {bone1Name=dyn_shldr_frillQ_01_r, bone2Name=dyn_parent_shldr_frillQ_01_r, swing1LimitDegrees=70, swing2LimitDegrees=70, twistLimitDegrees=5}
 {
  bone1Name=upperarm_l
  bone2Name=clavicle_l
  swing1LimitDegrees=45
  swing1Motion=Free
  swing2LimitDegrees=45
  swing2Motion=Free
  twistLimitDegrees=45
  twistMotion=Free
 }
 {
  bone1Name=dyn_parent_shldr_frillA_01_l
  bone2Name=clavicle_l
  swing1LimitDegrees=45
  swing1Motion=Free
  swing2LimitDegrees=45
  swing2Motion=Free
  twistLimitDegrees=45
  twistMotion=Free
 }
 {
  bone1Name=dyn_parent_shldr_frillB_01_l
  bone2Name=clavicle_l
  swing1LimitDegrees=45
  swing1Motion=Free
  swing2LimitDegrees=45
  swing2Motion=Free
  twistLimitDegrees=45
  twistMotion=Free
 }
 {
  bone1Name=dyn_parent_shldr_frillC_01_l
  bone2Name=clavicle_l
  swing1LimitDegrees=45
  swing1Motion=Free
  swing2LimitDegrees=45
  swing2Motion=Free
  twistLimitDegrees=45
  twistMotion=Free
 }
 {
  bone1Name=dyn_parent_shldr_frillD_01_l
  bone2Name=clavicle_l
  swing1LimitDegrees=45
  swing1Motion=Free
  swing2LimitDegrees=45
  swing2Motion=Free
  twistLimitDegrees=45
  twistMotion=Free
 }
 {
  bone1Name=dyn_parent_shldr_frillE_01_l
  bone2Name=clavicle_l
  swing1LimitDegrees=45
  swing1Motion=Free
  swing2LimitDegrees=45
  swing2Motion=Free
  twistLimitDegrees=45
  twistMotion=Free
 }
 {
  bone1Name=dyn_parent_shldr_frillF_01_l
  bone2Name=clavicle_l
  swing1LimitDegrees=45
  swing1Motion=Free
  swing2LimitDegrees=45
  swing2Motion=Free
  twistLimitDegrees=45
  twistMotion=Free
 }
 {
  bone1Name=dyn_parent_shldr_frillG_01_l
  bone2Name=clavicle_l
  swing1LimitDegrees=45
  swing1Motion=Free
  swing2LimitDegrees=45
  swing2Motion=Free
  twistLimitDegrees=45
  twistMotion=Free
 }
 {
  bone1Name=dyn_parent_shldr_frillH_01_l
  bone2Name=clavicle_l
  swing1LimitDegrees=45
  swing1Motion=Free
  swing2LimitDegrees=45
  swing2Motion=Free
  twistLimitDegrees=45
  twistMotion=Free
 }
 {
  bone1Name=dyn_parent_shldr_frillI_01_l
  bone2Name=clavicle_l
  swing1LimitDegrees=45
  swing1Motion=Free
  swing2LimitDegrees=45
  swing2Motion=Free
  twistLimitDegrees=45
  twistMotion=Free
 }
 {
  bone1Name=dyn_parent_shldr_frillJ_01_l
  bone2Name=clavicle_l
  swing1LimitDegrees=45
  swing1Motion=Free
  swing2LimitDegrees=45
  swing2Motion=Free
  twistLimitDegrees=45
  twistMotion=Free
 }
 {
  bone1Name=dyn_parent_shldr_frillK_01_l
  bone2Name=clavicle_l
  swing1LimitDegrees=45
  swing1Motion=Free
  swing2LimitDegrees=45
  swing2Motion=Free
  twistLimitDegrees=45
  twistMotion=Free
 }
 {
  bone1Name=dyn_parent_shldr_frillL_01_l
  bone2Name=clavicle_l
  swing1LimitDegrees=45
  swing1Motion=Free
  swing2LimitDegrees=45
  swing2Motion=Free
  twistLimitDegrees=45
  twistMotion=Free
 }
 {
  bone1Name=dyn_parent_shldr_frillM_01_l
  bone2Name=clavicle_l
  swing1LimitDegrees=45
  swing1Motion=Free
  swing2LimitDegrees=45
  swing2Motion=Free
  twistLimitDegrees=45
  twistMotion=Free
 }
 {
  bone1Name=dyn_parent_shldr_frillN_01_l
  bone2Name=clavicle_l
  swing1LimitDegrees=45
  swing1Motion=Free
  swing2LimitDegrees=45
  swing2Motion=Free
  twistLimitDegrees=45
  twistMotion=Free
 }
 {
  bone1Name=dyn_parent_shldr_frillO_01_l
  bone2Name=clavicle_l
  swing1LimitDegrees=45
  swing1Motion=Free
  swing2LimitDegrees=45
  swing2Motion=Free
  twistLimitDegrees=45
  twistMotion=Free
 }
 {
  bone1Name=dyn_parent_shldr_frillP_01_l
  bone2Name=clavicle_l
  swing1LimitDegrees=45
  swing1Motion=Free
  swing2LimitDegrees=45
  swing2Motion=Free
  twistLimitDegrees=45
  twistMotion=Free
 }
 {
  bone1Name=dyn_parent_shldr_frillQ_01_l
  bone2Name=clavicle_l
  swing1LimitDegrees=45
  swing1Motion=Free
  swing2LimitDegrees=45
  swing2Motion=Free
  twistLimitDegrees=45
  twistMotion=Free
 }
 {
  bone1Name=upperarm_r
  bone2Name=clavicle_r
  swing1LimitDegrees=45
  swing1Motion=Free
  swing2LimitDegrees=45
  swing2Motion=Free
  twistLimitDegrees=45
  twistMotion=Free
 }
 {
  bone1Name=dyn_parent_shldr_frillA_01_r
  bone2Name=clavicle_r
  swing1LimitDegrees=45
  swing1Motion=Free
  swing2LimitDegrees=45
  swing2Motion=Free
  twistLimitDegrees=45
  twistMotion=Free
 }
 {
  bone1Name=dyn_parent_shldr_frillB_01_r
  bone2Name=clavicle_r
  swing1LimitDegrees=45
  swing1Motion=Free
  swing2LimitDegrees=45
  swing2Motion=Free
  twistLimitDegrees=45
  twistMotion=Free
 }
 {
  bone1Name=dyn_parent_shldr_frillC_01_r
  bone2Name=clavicle_r
  swing1LimitDegrees=45
  swing1Motion=Free
  swing2LimitDegrees=45
  swing2Motion=Free
  twistLimitDegrees=45
  twistMotion=Free
 }
 {
  bone1Name=dyn_parent_shldr_frillD_01_r
  bone2Name=clavicle_r
  swing1LimitDegrees=45
  swing1Motion=Free
  swing2LimitDegrees=45
  swing2Motion=Free
  twistLimitDegrees=45
  twistMotion=Free
 }
 {
  bone1Name=dyn_parent_shldr_frillE_01_r
  bone2Name=clavicle_r
  swing1LimitDegrees=45
  swing1Motion=Free
  swing2LimitDegrees=45
  swing2Motion=Free
  twistLimitDegrees=45
  twistMotion=Free
 }
 {
  bone1Name=dyn_parent_shldr_frillF_01_r
  bone2Name=clavicle_r
  swing1LimitDegrees=45
  swing1Motion=Free
  swing2LimitDegrees=45
  swing2Motion=Free
  twistLimitDegrees=45
  twistMotion=Free
 }
 {
  bone1Name=dyn_parent_shldr_frillG_01_r
  bone2Name=clavicle_r
  swing1LimitDegrees=45
  swing1Motion=Free
  swing2LimitDegrees=45
  swing2Motion=Free
  twistLimitDegrees=45
  twistMotion=Free
 }
 {
  bone1Name=dyn_parent_shldr_frillH_01_r
  bone2Name=clavicle_r
  swing1LimitDegrees=45
  swing1Motion=Free
  swing2LimitDegrees=45
  swing2Motion=Free
  twistLimitDegrees=45
  twistMotion=Free
 }
 {
  bone1Name=dyn_parent_shldr_frillI_01_r
  bone2Name=clavicle_r
  swing1LimitDegrees=45
  swing1Motion=Free
  swing2LimitDegrees=45
  swing2Motion=Free
  twistLimitDegrees=45
  twistMotion=Free
 }
 {
  bone1Name=dyn_parent_shldr_frillJ_01_r
  bone2Name=clavicle_r
  swing1LimitDegrees=45
  swing1Motion=Free
  swing2LimitDegrees=45
  swing2Motion=Free
  twistLimitDegrees=45
  twistMotion=Free
 }
 {
  bone1Name=dyn_parent_shldr_frillK_01_r
  bone2Name=clavicle_r
  swing1LimitDegrees=45
  swing1Motion=Free
  swing2LimitDegrees=45
  swing2Motion=Free
  twistLimitDegrees=45
  twistMotion=Free
 }
 {
  bone1Name=dyn_parent_shldr_frillL_01_r
  bone2Name=clavicle_r
  swing1LimitDegrees=45
  swing1Motion=Free
  swing2LimitDegrees=45
  swing2Motion=Free
  twistLimitDegrees=45
  twistMotion=Free
 }
 {
  bone1Name=dyn_parent_shldr_frillM_01_r
  bone2Name=clavicle_r
  swing1LimitDegrees=45
  swing1Motion=Free
  swing2LimitDegrees=45
  swing2Motion=Free
  twistLimitDegrees=45
  twistMotion=Free
 }
 {
  bone1Name=dyn_parent_shldr_frillN_01_r
  bone2Name=clavicle_r
  swing1LimitDegrees=45
  swing1Motion=Free
  swing2LimitDegrees=45
  swing2Motion=Free
  twistLimitDegrees=45
  twistMotion=Free
 }
 {
  bone1Name=dyn_parent_shldr_frillO_01_r
  bone2Name=clavicle_r
  swing1LimitDegrees=45
  swing1Motion=Free
  swing2LimitDegrees=45
  swing2Motion=Free
  twistLimitDegrees=45
  twistMotion=Free
 }
 {
  bone1Name=dyn_parent_shldr_frillP_01_r
  bone2Name=clavicle_r
  swing1LimitDegrees=45
  swing1Motion=Free
  swing2LimitDegrees=45
  swing2Motion=Free
  twistLimitDegrees=45
  twistMotion=Free
 }
 {
  bone1Name=dyn_parent_shldr_frillQ_01_r
  bone2Name=clavicle_r
  swing1LimitDegrees=45
  swing1Motion=Free
  swing2LimitDegrees=45
  swing2Motion=Free
  twistLimitDegrees=45
  twistMotion=Free
 }
 {
  bone1Name=clavicle_l
  bone2Name=spine_05
  swing1LimitDegrees=45
  swing1Motion=Free
  swing2LimitDegrees=45
  swing2Motion=Free
  twistLimitDegrees=45
  twistMotion=Free
 }
 {
  bone1Name=clavicle_r
  bone2Name=spine_05
  swing1LimitDegrees=45
  swing1Motion=Free
  swing2LimitDegrees=45
  swing2Motion=Free
  twistLimitDegrees=45
  twistMotion=Free
 }
]
