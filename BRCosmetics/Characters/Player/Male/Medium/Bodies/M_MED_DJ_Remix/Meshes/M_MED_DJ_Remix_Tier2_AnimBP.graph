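# Animation Blueprint for the DJ Remix Tier 2 male body cosmetic, driving secondary bones on outfit frills.
# It copies the attached parent mesh's pose, then constrains thigh-fix bones to thigh twists and offsets them by variable alphas.
# A long chain of weighted multi-target constraints then places hip and shoulder frill bones from pelvis, thighs, shorts, spine and shoulder pads.
# AnimDynamics, RigidBody and back-chain constraint nodes also appear, without visible input links.
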AnimBlueprint /BRCosmetics/Characters/Player/Male/Medium/Bodies/M_MED_DJ_Remix/Meshes/M_MED_DJ_Remix_Tier2_AnimBP
target_skeleton=&/M_MED_DJ_Remix_Skeleton
graphs [
 AnimGraph {nodes [
   Anim Dynamics - Bone: dyn_parent_chain_back(Alpha=1, ExternalForce="0,0,0") {
    angularDampingOverride=1
    angularSpringConstant=1500
    bAngularSpring=true
    bGravityOverrideInSimSpace=true
    bOverrideAngularDamping=true
    gravityScale=0.35
    simulationSpace=RootRelative
    alphaBoolBlend {blendOption=Linear}
    boundBone {boneName=dyn_parent_chain_back}
    physicsBodyDefinitions [{
      sphereCollisionRadius=0
      boundBone {boneName=dyn_parent_chain_back}
      boxExtents=<20 15 10>
      constraintSetup {twistAxis=AxisZ, angularLimitsMax=<45 20 0>, angularLimitsMin=<0 -20 0>, angularTarget=<0 0 0>}
      localJointOffset=<0 -0.2 -10>
     }]
   }
   Anim Dynamics - Bone: dyn_parent_chain_back(Alpha=1, ExternalForce="0,0,0") {
    angularDampingOverride=0.9999
    angularSpringConstant=200
    bGravityOverrideInSimSpace=true
    bOverrideAngularDamping=true
    gravityScale=0.35
    simulationSpace=RootRelative
    alphaBoolBlend {blendOption=Linear}
    boundBone {boneName=dyn_parent_chain_back}
    componentAppliedLinearAccClamp=<1200 1200 1350>
    componentLinearAccScale=<0.8 0.8 0.8>
    componentLinearVelScale=<0.8 0.8 0.8>
    physicsBodyDefinitions [{
      sphereCollisionRadius=0
      boundBone {boneName=dyn_parent_chain_back}
      boxExtents=<20 8 5>
      constraintSetup {twistAxis=AxisZ, angularLimitsMax=<60 20 0>, angularLimitsMin=<-15 -20 0>, angularTarget=<0 0 0>}
      localJointOffset=<0 0 -5>
     }]
   }
   "Constraint - dyn_chain_back (PPPPPP)"!(
    ComponentPose
    Alpha=1
    ConstraintWeights_0=0.0108
    ConstraintWeights_1:float=0.1311
    ConstraintWeights_2:float=0.022
    ConstraintWeights_3:float=0.7009
    ConstraintWeights_4:float=0.0162
    ConstraintWeights_5:float=0.1191
   ) -> (Pose) {
    alphaBoolBlend {blendOption=Linear}
    boneToModify {boneName=dyn_chain_back}
    constraintSetup [
     {targetBone {boneName=clavicle_l}}
     {targetBone {boneName=spine_03}}
     {targetBone {boneName=spine_04}}
     {targetBone {boneName=spine_05}}
     {targetBone {boneName=clavicle_r}}
     {targetBone {boneName=neck_01}}
    ]
    constraintWeights [1,1,1,1,1,1]
   }
   RigidBody() {
    bClampLinearTranslationLimitToRefPose=true
    bOverrideWorldGravity=true
    overridePhysicsAsset=&/M_MED_DJ_Remix_Physics
    simulationSpace=BaseBoneSpace
    alphaBoolBlend {blendOption=Linear}
    baseBoneRef {boneName=root}
    componentAppliedLinearAccClamp=<20000 20000 1300>
    componentLinearAccScale=<1 1 1>
    componentLinearVelScale=<1 1 1>
    overrideWorldGravity {z=-980}
   }
   n1 = "Constraint - dyn_thigh_fix_l (P)"(ComponentPose=Local To Component(LocalPose=Copy Pose From Mesh() {bUseAttachedParent=true}), Alpha=1, ConstraintWeights_0=1) {
    alphaBoolBlend {blendOption=Linear}
    boneToModify {boneName=dyn_thigh_fix_l}
    constraintSetup [{targetBone {boneName=thigh_twist_01_l}}]
    constraintWeights [1]
   }
   n2 = "Constraint - dyn_thigh_fix_r (P)"(ComponentPose=@n1, Alpha=1, ConstraintWeights_0=1) {
    alphaBoolBlend {blendOption=Linear}
    boneToModify {boneName=dyn_thigh_fix_r}
    constraintSetup [{targetBone {boneName=thigh_twist_01_r}}]
    constraintWeights [1]
   }
   n3 = "Transform (Modify) Bone - Bone: dyn_thigh_fix_l"(
    ComponentPose=@n2
    Alpha=Get ThighUpperLeftAlpha() -> (ThighUpperLeftAlpha:float)
    Translation="3.3,7.35,-21"
    Rotation="0.000,0.000,25"
    Scale="1,1,1"
   ) {
    rotationMode=BMM_Additive
    rotationSpace=BCS_BoneSpace
    scaleMode=BMM_Additive
    scaleSpace=BCS_BoneSpace
    translationMode=BMM_Additive
    translationSpace=BCS_BoneSpace
    alphaBoolBlend {blendOption=Linear}
    boneToModify {boneName=dyn_thigh_fix_l}
    rotation {roll=25}
    translation=<3.3 7.35 -21>
   }
   n4 = "Transform (Modify) Bone - Bone: dyn_thigh_fix_r"(
    ComponentPose=@n3
    Alpha=Get ThighUpperRightAlpha() -> (ThighUpperRightAlpha:float)
    Translation="-3,-7.5,21"
    Rotation="0.000,0.000,25"
    Scale="1,1,1"
   ) {
    rotationMode=BMM_Additive
    rotationSpace=BCS_BoneSpace
    scaleMode=BMM_Additive
    scaleSpace=BCS_BoneSpace
    translationMode=BMM_Additive
    translationSpace=BCS_BoneSpace
    alphaBoolBlend {blendOption=Linear}
    boneToModify {boneName=dyn_thigh_fix_r}
    rotation {roll=25}
    translation=<-3 -7.5 21>
   }
   n5 = "Constraint - dyn_parent_frill_01_r (PPPPP)"!(
    ComponentPose=@n4
    Alpha=1
    ConstraintWeights_0=0.4122
    ConstraintWeights_1:float=0.1202
    ConstraintWeights_2:float=0.0179
    ConstraintWeights_3:float=0.4388
    ConstraintWeights_4:float=0.011
   ) -> (Pose) {
    alphaBoolBlend {blendOption=Linear}
    boneToModify {boneName=dyn_parent_frill_01_r}
    constraintSetup [
     {targetBone {boneName=dyn_thigh_fix_r}}
     {targetBone {boneName=pelvis}}
     {targetBone {boneName=dyn_shorts_frt_02_r}}
     {targetBone {boneName=thigh_twist_01_r}}
     {targetBone {boneName=dyn_shorts_outA_02_r}}
    ]
    constraintWeights [1,1,1,1,1]
   }
   n6 = "Constraint - dyn_parent_frill_02_r (PPPPP)"!(
    ComponentPose=@n5
    Alpha=1
    ConstraintWeights_0=0.3553
    ConstraintWeights_1:float=0.1484
    ConstraintWeights_2:float=0.0251
    ConstraintWeights_3:float=0.4542
    ConstraintWeights_4:float=0.0171
   ) -> (Pose) {
    alphaBoolBlend {blendOption=Linear}
    boneToModify {boneName=dyn_parent_frill_02_r}
    constraintSetup [
     {targetBone {boneName=dyn_thigh_fix_r}}
     {targetBone {boneName=pelvis}}
     {targetBone {boneName=dyn_shorts_frt_02_r}}
     {targetBone {boneName=thigh_twist_01_r}}
     {targetBone {boneName=dyn_shorts_outA_02_r}}
    ]
    constraintWeights [1,1,1,1,1]
   }
   n7 = "Constraint - dyn_parent_frill_03_r (PPPPPPP)"!(
    ComponentPose=@n6
    Alpha=1
    ConstraintWeights_0=0.2729
    ConstraintWeights_1:float=0.0267
    ConstraintWeights_2:float=0.014
    ConstraintWeights_3:float=0.4861
    ConstraintWeights_4:float=0.0123
    ConstraintWeights_5:float=0.0365
    ConstraintWeights_6:float=0.1515
   ) -> (Pose) {
    alphaBoolBlend {blendOption=Linear}
    boneToModify {boneName=dyn_parent_frill_03_r}
    constraintSetup [
     {targetBone {boneName=dyn_thigh_fix_r}}
     {targetBone {boneName=dyn_shorts_outA_02_r}}
     {targetBone {boneName=dyn_shorts_frt_01_r}}
     {targetBone {boneName=thigh_twist_01_r}}
     {targetBone {boneName=dyn_shorts_outB_02_r}}
     {targetBone {boneName=dyn_shorts_frt_02_r}}
     {targetBone {boneName=pelvis}}
    ]
    constraintWeights [1,1,1,1,1,1,1]
   }
   n8 = "Constraint - dyn_parent_frill_04_r (PPPPPPPP)"!(
    ComponentPose=@n7
    Alpha=1
    ConstraintWeights_0=0.2147
    ConstraintWeights_1:float=0.0296
    ConstraintWeights_2:float=0.0119
    ConstraintWeights_3:float=0.495
    ConstraintWeights_4:float=0.0155
    ConstraintWeights_5:float=0.0109
    ConstraintWeights_6:float=0.0355
    ConstraintWeights_7:float=0.1868
   ) -> (Pose) {
    alphaBoolBlend {blendOption=Linear}
    boneToModify {boneName=dyn_parent_frill_04_r}
    constraintSetup [
     {targetBone {boneName=dyn_thigh_fix_r}}
     {targetBone {boneName=dyn_shorts_outA_02_r}}
     {targetBone {boneName=dyn_shorts_frt_01_r}}
     {targetBone {boneName=thigh_twist_01_r}}
     {targetBone {boneName=dyn_shorts_outB_02_r}}
     {targetBone {boneName=spine_01}}
     {targetBone {boneName=dyn_shorts_frt_02_r}}
     {targetBone {boneName=pelvis}}
    ]
    constraintWeights [1,1,1,1,1,1,1,1]
   }
   n9 = "Constraint - dyn_parent_frill_05_r (PPPPPPPP)"!(
    ComponentPose=@n8
    Alpha=1
    ConstraintWeights_0=0.1361
    ConstraintWeights_1:float=0.0316
    ConstraintWeights_2:float=0.5282
    ConstraintWeights_3:float=0.0189
    ConstraintWeights_4:float=0.0128
    ConstraintWeights_5:float=0.0319
    ConstraintWeights_6:float=0.0099
    ConstraintWeights_7:float=0.2306
   ) -> (Pose) {
    alphaBoolBlend {blendOption=Linear}
    boneToModify {boneName=dyn_parent_frill_05_r}
    constraintSetup [
     {targetBone {boneName=dyn_thigh_fix_r}}
     {targetBone {boneName=dyn_shorts_outA_02_r}}
     {targetBone {boneName=thigh_twist_01_r}}
     {targetBone {boneName=dyn_shorts_outB_02_r}}
     {targetBone {boneName=spine_01}}
     {targetBone {boneName=dyn_shorts_frt_02_r}}
     {targetBone {boneName=dyn_shorts_outB_01_r}}
     {targetBone {boneName=pelvis}}
    ]
    constraintWeights [1,1,1,1,1,1,1,1]
   }
   n10 = "Constraint - dyn_parent_frill_06_r (PPPPPPPP)"!(
    ComponentPose=@n9
    Alpha=1
    ConstraintWeights_0=0.0597
    ConstraintWeights_1:float=0.031
    ConstraintWeights_2:float=0.5531
    ConstraintWeights_3:float=0.0209
    ConstraintWeights_4:float=0.0138
    ConstraintWeights_5:float=0.0242
    ConstraintWeights_6:float=0.0133
    ConstraintWeights_7:float=0.2841
   ) -> (Pose) {
    alphaBoolBlend {blendOption=Linear}
    boneToModify {boneName=dyn_parent_frill_06_r}
    constraintSetup [
     {targetBone {boneName=dyn_thigh_fix_r}}
     {targetBone {boneName=dyn_shorts_outA_02_r}}
     {targetBone {boneName=thigh_twist_01_r}}
     {targetBone {boneName=dyn_shorts_outB_02_r}}
     {targetBone {boneName=spine_01}}
     {targetBone {boneName=dyn_shorts_frt_02_r}}
     {targetBone {boneName=dyn_shorts_outB_01_r}}
     {targetBone {boneName=pelvis}}
    ]
    constraintWeights [1,1,1,1,1,1,1,1]
   }
   n11 = "Constraint - dyn_parent_frill_07_r (PPP)"!(ComponentPose=@n10, Alpha=1, ConstraintWeights_0=0.0249, ConstraintWeights_1:float=0.2125, ConstraintWeights_2:float=0.7626) -> (Pose) {
    alphaBoolBlend {blendOption=Linear}
    boneToModify {boneName=dyn_parent_frill_07_r}
    constraintSetup [{targetBone {boneName=thigh_r}}, {targetBone {boneName=pelvis}}, {targetBone {boneName=thigh_twist_01_r}}]
    constraintWeights [1,1,1]
   }
   n12 = "Constraint - dyn_parent_frill_08_r (PPP)"!(ComponentPose=@n11, Alpha=1, ConstraintWeights_0=0.0243, ConstraintWeights_1:float=0.2738, ConstraintWeights_2:float=0.7019) -> (Pose) {
    alphaBoolBlend {blendOption=Linear}
    boneToModify {boneName=dyn_parent_frill_08_r}
    constraintSetup [{targetBone {boneName=thigh_r}}, {targetBone {boneName=pelvis}}, {targetBone {boneName=thigh_twist_01_r}}]
    constraintWeights [1,1,1]
   }
   n13 = "Constraint - dyn_parent_frill_09_r (PPP)"!(ComponentPose=@n12, Alpha=1, ConstraintWeights_0=0.0197, ConstraintWeights_1:float=0.3508, ConstraintWeights_2:float=0.6295) -> (Pose) {
    alphaBoolBlend {blendOption=Linear}
    boneToModify {boneName=dyn_parent_frill_09_r}
    constraintSetup [{targetBone {boneName=thigh_r}}, {targetBone {boneName=pelvis}}, {targetBone {boneName=thigh_twist_01_r}}]
    constraintWeights [1,1,1]
   }
   n14 = "Constraint - dyn_parent_frill_10_r (PPP)"!(ComponentPose=@n13, Alpha=1, ConstraintWeights_0=0.0157, ConstraintWeights_1:float=0.4239, ConstraintWeights_2:float=0.5604) -> (Pose) {
    alphaBoolBlend {blendOption=Linear}
    boneToModify {boneName=dyn_parent_frill_10_r}
    constraintSetup [{targetBone {boneName=thigh_r}}, {targetBone {boneName=pelvis}}, {targetBone {boneName=thigh_twist_01_r}}]
    constraintWeights [1,1,1]
   }
   n15 = "Constraint - dyn_parent_frill_11_r (PPP)"!(ComponentPose=@n14, Alpha=1, ConstraintWeights_0=0.0126, ConstraintWeights_1:float=0.4957, ConstraintWeights_2:float=0.4917) -> (Pose) {
    alphaBoolBlend {blendOption=Linear}
    boneToModify {boneName=dyn_parent_frill_11_r}
    constraintSetup [{targetBone {boneName=thigh_r}}, {targetBone {boneName=pelvis}}, {targetBone {boneName=thigh_twist_01_r}}]
    constraintWeights [1,1,1]
   }
   n16 = "Constraint - dyn_parent_frill_12_r (PP)"!(ComponentPose=@n15, Alpha=1, ConstraintWeights_0=0.5718, ConstraintWeights_1:float=0.4282) -> (Pose) {
    alphaBoolBlend {blendOption=Linear}
    boneToModify {boneName=dyn_parent_frill_12_r}
    constraintSetup [{targetBone {boneName=pelvis}}, {targetBone {boneName=thigh_twist_01_r}}]
    constraintWeights [1,1]
   }
   n17 = "Constraint - dyn_parent_frill_13_r (PPP)"!(ComponentPose=@n16, Alpha=1, ConstraintWeights_0=0.6339, ConstraintWeights_1:float=0.0173, ConstraintWeights_2:float=0.3488) -> (Pose) {
    alphaBoolBlend {blendOption=Linear}
    boneToModify {boneName=dyn_parent_frill_13_r}
    constraintSetup [{targetBone {boneName=pelvis}}, {targetBone {boneName=thigh_twist_01_l}}, {targetBone {boneName=thigh_twist_01_r}}]
    constraintWeights [1,1,1]
   }
   n18 = "Constraint - dyn_parent_frill_14_r (PPP)"!(ComponentPose=@n17, Alpha=1, ConstraintWeights_0=0.6932, ConstraintWeights_1:float=0.0292, ConstraintWeights_2:float=0.2776) -> (Pose) {
    alphaBoolBlend {blendOption=Linear}
    boneToModify {boneName=dyn_parent_frill_14_r}
    constraintSetup [{targetBone {boneName=pelvis}}, {targetBone {boneName=thigh_twist_01_l}}, {targetBone {boneName=thigh_twist_01_r}}]
    constraintWeights [1,1,1]
   }
   n19 = "Constraint - dyn_parent_frill_15_r (PPP)"!(ComponentPose=@n18, Alpha=1, ConstraintWeights_0=0.7483, ConstraintWeights_1:float=0.0464, ConstraintWeights_2:float=0.2053) -> (Pose) {
    alphaBoolBlend {blendOption=Linear}
    boneToModify {boneName=dyn_parent_frill_15_r}
    constraintSetup [{targetBone {boneName=pelvis}}, {targetBone {boneName=thigh_twist_01_l}}, {targetBone {boneName=thigh_twist_01_r}}]
    constraintWeights [1,1,1]
   }
   n20 = "Constraint - dyn_parent_frill_01_l (PPPPP)"!(
    ComponentPose=@n19
    Alpha=1
    ConstraintWeights_0=0.1067
    ConstraintWeights_1:float=0.4054
    ConstraintWeights_2:float=0.4594
    ConstraintWeights_3:float=0.0093
    ConstraintWeights_4:float=0.0192
   ) -> (Pose) {
    alphaBoolBlend {blendOption=Linear}
    boneToModify {boneName=dyn_parent_frill_01_l}
    constraintSetup [
     {targetBone {boneName=pelvis}}
     {targetBone {boneName=dyn_thigh_fix_l}}
     {targetBone {boneName=thigh_twist_01_l}}
     {targetBone {boneName=dyn_shorts_outA_02_l}}
     {targetBone {boneName=dyn_shorts_frt_02_l}}
    ]
    constraintWeights [1,1,1,1,1]
   }
   n21 = "Constraint - dyn_parent_frill_02_l (PPPPP)"!(
    ComponentPose=@n20
    Alpha=1
    ConstraintWeights_0=0.1352
    ConstraintWeights_1:float=0.3424
    ConstraintWeights_2:float=0.4777
    ConstraintWeights_3:float=0.018
    ConstraintWeights_4:float=0.0268
   ) -> (Pose) {
    alphaBoolBlend {blendOption=Linear}
    boneToModify {boneName=dyn_parent_frill_02_l}
    constraintSetup [
     {targetBone {boneName=pelvis}}
     {targetBone {boneName=dyn_thigh_fix_l}}
     {targetBone {boneName=thigh_twist_01_l}}
     {targetBone {boneName=dyn_shorts_outA_02_l}}
     {targetBone {boneName=dyn_shorts_frt_02_l}}
    ]
    constraintWeights [1,1,1,1,1]
   }
   n22 = "Constraint - dyn_parent_frill_03_l (PPPPPPP)"!(
    ComponentPose=@n21
    Alpha=1
    ConstraintWeights_0=0.0313
    ConstraintWeights_1:float=0.0037
    ConstraintWeights_2:float=0.4774
    ConstraintWeights_3:float=0.01
    ConstraintWeights_4:float=0.1441
    ConstraintWeights_5:float=0.3116
    ConstraintWeights_6:float=0.0219
   ) -> (Pose) {
    alphaBoolBlend {blendOption=Linear}
    boneToModify {boneName=dyn_parent_frill_03_l}
    constraintSetup [
     {targetBone {boneName=dyn_shorts_frt_02_l}}
     {targetBone {boneName=dyn_shorts_outB_02_l}}
     {targetBone {boneName=thigh_twist_01_l}}
     {targetBone {boneName=dyn_shorts_frt_01_l}}
     {targetBone {boneName=pelvis}}
     {targetBone {boneName=dyn_thigh_fix_l}}
     {targetBone {boneName=dyn_shorts_outA_02_l}}
    ]
    constraintWeights [1,1,1,1,1,1,1]
   }
   n23 = "Constraint - dyn_parent_frill_04_l (PPPPP)"!(
    ComponentPose=@n22
    Alpha=1
    ConstraintWeights_0=0.1677
    ConstraintWeights_1:float=0.2413
    ConstraintWeights_2:float=0.5496
    ConstraintWeights_3:float=0.0174
    ConstraintWeights_4:float=0.024
   ) -> (Pose) {
    alphaBoolBlend {blendOption=Linear}
    boneToModify {boneName=dyn_parent_frill_04_l}
    constraintSetup [
     {targetBone {boneName=pelvis}}
     {targetBone {boneName=dyn_thigh_fix_l}}
     {targetBone {boneName=thigh_twist_01_l}}
     {targetBone {boneName=dyn_shorts_outA_02_l}}
     {targetBone {boneName=dyn_shorts_frt_02_l}}
    ]
    constraintWeights [1,1,1,1,1]
   }
   n24 = "Constraint - dyn_parent_frill_05_l (PPPPP)"!(
    ComponentPose=@n23
    Alpha=1
    ConstraintWeights_0=0.1691
    ConstraintWeights_1:float=0.1858
    ConstraintWeights_2:float=0.6095
    ConstraintWeights_3:float=0.0162
    ConstraintWeights_4:float=0.0194
   ) -> (Pose) {
    alphaBoolBlend {blendOption=Linear}
    boneToModify {boneName=dyn_parent_frill_05_l}
    constraintSetup [
     {targetBone {boneName=pelvis}}
     {targetBone {boneName=dyn_thigh_fix_l}}
     {targetBone {boneName=thigh_twist_01_l}}
     {targetBone {boneName=dyn_shorts_outA_02_l}}
     {targetBone {boneName=dyn_shorts_frt_02_l}}
    ]
    constraintWeights [1,1,1,1,1]
   }
   n25 = "Constraint - dyn_parent_frill_06_l (PPPPPPPP)"!(
    ComponentPose=@n24
    Alpha=1
    ConstraintWeights_0=0.0132
    ConstraintWeights_1:float=0.0241
    ConstraintWeights_2:float=0.0209
    ConstraintWeights_3:float=0.0137
    ConstraintWeights_4:float=0.5535
    ConstraintWeights_5:float=0.2841
    ConstraintWeights_6:float=0.0596
    ConstraintWeights_7:float=0.0309
   ) -> (Pose) {
    alphaBoolBlend {blendOption=Linear}
    boneToModify {boneName=dyn_parent_frill_06_l}
    constraintSetup [
     {targetBone {boneName=dyn_shorts_outB_01_l}}
     {targetBone {boneName=dyn_shorts_frt_02_l}}
     {targetBone {boneName=dyn_shorts_outB_02_l}}
     {targetBone {boneName=spine_01}}
     {targetBone {boneName=thigh_twist_01_l}}
     {targetBone {boneName=pelvis}}
     {targetBone {boneName=dyn_thigh_fix_l}}
     {targetBone {boneName=dyn_shorts_outA_02_l}}
    ]
    constraintWeights [1,1,1,1,1,1,1,1]
   }
   n26 = "Constraint - dyn_parent_frill_07_l (PPP)"!(ComponentPose=@n25, Alpha=1, ConstraintWeights_0=0.2181, ConstraintWeights_1:float=0.7549, ConstraintWeights_2:float=0.027) -> (Pose) {
    alphaBoolBlend {blendOption=Linear}
    boneToModify {boneName=dyn_parent_frill_07_l}
    constraintSetup [{targetBone {boneName=pelvis}}, {targetBone {boneName=thigh_twist_01_l}}, {targetBone {boneName=thigh_l}}]
    constraintWeights [1,1,1]
   }
   n27 = "Constraint - dyn_parent_frill_08_l (PPP)"!(ComponentPose=@n26, Alpha=1, ConstraintWeights_0=0.2838, ConstraintWeights_1:float=0.6916, ConstraintWeights_2:float=0.0245) -> (Pose) {
    alphaBoolBlend {blendOption=Linear}
    boneToModify {boneName=dyn_parent_frill_08_l}
    constraintSetup [{targetBone {boneName=pelvis}}, {targetBone {boneName=thigh_twist_01_l}}, {targetBone {boneName=thigh_l}}]
    constraintWeights [1,1,1]
   }
   n28 = "Constraint - dyn_parent_frill_09_l (PPP)"!(ComponentPose=@n27, Alpha=1, ConstraintWeights_0=0.3633, ConstraintWeights_1:float=0.618, ConstraintWeights_2:float=0.0187) -> (Pose) {
    alphaBoolBlend {blendOption=Linear}
    boneToModify {boneName=dyn_parent_frill_09_l}
    constraintSetup [{targetBone {boneName=pelvis}}, {targetBone {boneName=thigh_twist_01_l}}, {targetBone {boneName=thigh_l}}]
    constraintWeights [1,1,1]
   }
   n29 = "Constraint - dyn_parent_frill_10_l (PPP)"!(ComponentPose=@n28, Alpha=1, ConstraintWeights_0=0.4433, ConstraintWeights_1:float=0.5436, ConstraintWeights_2:float=0.0132) -> (Pose) {
    alphaBoolBlend {blendOption=Linear}
    boneToModify {boneName=dyn_parent_frill_10_l}
    constraintSetup [{targetBone {boneName=pelvis}}, {targetBone {boneName=thigh_twist_01_l}}, {targetBone {boneName=thigh_l}}]
    constraintWeights [1,1,1]
   }
   n30 = "Constraint - dyn_parent_frill_11_l (PPP)"!(ComponentPose=@n29, Alpha=1, ConstraintWeights_0=0.505, ConstraintWeights_1:float=0.4844, ConstraintWeights_2:float=0.0107) -> (Pose) {
    alphaBoolBlend {blendOption=Linear}
    boneToModify {boneName=dyn_parent_frill_11_l}
    constraintSetup [{targetBone {boneName=pelvis}}, {targetBone {boneName=thigh_twist_01_l}}, {targetBone {boneName=thigh_l}}]
    constraintWeights [1,1,1]
   }
   n31 = "Constraint - dyn_parent_frill_12_l (PP)"!(ComponentPose=@n30, Alpha=1, ConstraintWeights_0=0.5707, ConstraintWeights_1:float=0.4293) -> (Pose) {
    alphaBoolBlend {blendOption=Linear}
    boneToModify {boneName=dyn_parent_frill_12_l}
    constraintSetup [{targetBone {boneName=pelvis}}, {targetBone {boneName=thigh_twist_01_l}}]
    constraintWeights [1,1]
   }
   n32 = "Constraint - dyn_parent_frill_13_l (PPP)"!(ComponentPose=@n31, Alpha=1, ConstraintWeights_0=0.6265, ConstraintWeights_1:float=0.3562, ConstraintWeights_2:float=0.0173) -> (Pose) {
    alphaBoolBlend {blendOption=Linear}
    boneToModify {boneName=dyn_parent_frill_13_l}
    constraintSetup [{targetBone {boneName=pelvis}}, {targetBone {boneName=thigh_twist_01_l}}, {targetBone {boneName=thigh_twist_01_r}}]
    constraintWeights [1,1,1]
   }
   n33 = "Constraint - dyn_parent_frill_14_l (PPP)"!(ComponentPose=@n32, Alpha=1, ConstraintWeights_0=0.6868, ConstraintWeights_1:float=0.2844, ConstraintWeights_2:float=0.0288) -> (Pose) {
    alphaBoolBlend {blendOption=Linear}
    boneToModify {boneName=dyn_parent_frill_14_l}
    constraintSetup [{targetBone {boneName=pelvis}}, {targetBone {boneName=thigh_twist_01_l}}, {targetBone {boneName=thigh_twist_01_r}}]
    constraintWeights [1,1,1]
   }
   n34 = "Constraint - dyn_parent_frill_15_l (PPP)"!(ComponentPose=@n33, Alpha=1, ConstraintWeights_0=0.7477, ConstraintWeights_1:float=0.2064, ConstraintWeights_2:float=0.0459) -> (Pose) {
    alphaBoolBlend {blendOption=Linear}
    boneToModify {boneName=dyn_parent_frill_15_l}
    constraintSetup [{targetBone {boneName=pelvis}}, {targetBone {boneName=thigh_twist_01_l}}, {targetBone {boneName=thigh_twist_01_r}}]
    constraintWeights [1,1,1]
   }
   n35 = "Constraint - dyn_parent_shldr_frillA_01_l (PPPPP)"!(
    ComponentPose=@n34
    Alpha=1
    ConstraintWeights_0=0.0626
    ConstraintWeights_1:float=0.0114
    ConstraintWeights_2:float=0.1011
    ConstraintWeights_3:float=0.6413
    ConstraintWeights_4:float=0.1836
   ) -> (Pose) {
    alphaBoolBlend {blendOption=Linear}
    boneToModify {boneName=dyn_parent_shldr_frillA_01_l}
    constraintSetup [
     {targetBone {boneName=spine_04}}
     {targetBone {boneName=spine_02}}
     {targetBone {boneName=spine_03}}
     {targetBone {boneName=shldr_pad_01_l}}
     {targetBone {boneName=spine_05}}
    ]
    constraintWeights [1,1,1,1,1]
   }
   n36 = "Constraint - dyn_parent_shldr_frillB_01_l (PPPPP)"!(
    ComponentPose=@n35
    Alpha=1
    ConstraintWeights_0=0.0602
    ConstraintWeights_1:float=0.0117
    ConstraintWeights_2:float=0.1005
    ConstraintWeights_3:float=0.6516
    ConstraintWeights_4:float=0.1761
   ) -> (Pose) {
    alphaBoolBlend {blendOption=Linear}
    boneToModify {boneName=dyn_parent_shldr_frillB_01_l}
    constraintSetup [
     {targetBone {boneName=spine_04}}
     {targetBone {boneName=spine_02}}
     {targetBone {boneName=spine_03}}
     {targetBone {boneName=shldr_pad_01_l}}
     {targetBone {boneName=spine_05}}
    ]
    constraintWeights [1,1,1,1,1]
   }
   n37 = "Constraint - dyn_parent_shldr_frillC_01_l (PPPPP)"!(
    ComponentPose=@n36
    Alpha=1
    ConstraintWeights_0=0.0578
    ConstraintWeights_1:float=0.0118
    ConstraintWeights_2:float=0.1001
    ConstraintWeights_3:float=0.6614
    ConstraintWeights_4:float=0.169
   ) -> (Pose) {
    alphaBoolBlend {blendOption=Linear}
    boneToModify {boneName=dyn_parent_shldr_frillC_01_l}
    constraintSetup [
     {targetBone {boneName=spine_04}}
     {targetBone {boneName=spine_02}}
     {targetBone {boneName=spine_03}}
     {targetBone {boneName=shldr_pad_01_l}}
     {targetBone {boneName=spine_05}}
    ]
    constraintWeights [1,1,1,1,1]
   }
   n38 = "Constraint - dyn_parent_shldr_frillD_01_l (PPPPP)"!(
    ComponentPose=@n37
    Alpha=1
    ConstraintWeights_0=0.0559
    ConstraintWeights_1:float=0.0119
    ConstraintWeights_2:float=0.0997
    ConstraintWeights_3:float=0.6691
    ConstraintWeights_4:float=0.1634
   ) -> (Pose) {
    alphaBoolBlend {blendOption=Linear}
    boneToModify {boneName=dyn_parent_shldr_frillD_01_l}
    constraintSetup [
     {targetBone {boneName=spine_04}}
     {targetBone {boneName=spine_02}}
     {targetBone {boneName=spine_03}}
     {targetBone {boneName=shldr_pad_01_l}}
     {targetBone {boneName=spine_05}}
    ]
    constraintWeights [1,1,1,1,1]
   }
   n39 = "Constraint - dyn_parent_shldr_frillE_01_l (PPPPP)"!(
    ComponentPose=@n38
    Alpha=1
    ConstraintWeights_0=0.0537
    ConstraintWeights_1:float=0.0116
    ConstraintWeights_2:float=0.1006
    ConstraintWeights_3:float=0.6748
    ConstraintWeights_4:float=0.1592
   ) -> (Pose) {
    alphaBoolBlend {blendOption=Linear}
    boneToModify {boneName=dyn_parent_shldr_frillE_01_l}
    constraintSetup [
     {targetBone {boneName=spine_04}}
     {targetBone {boneName=spine_02}}
     {targetBone {boneName=spine_03}}
     {targetBone {boneName=shldr_pad_01_l}}
     {targetBone {boneName=spine_05}}
    ]
    constraintWeights [1,1,1,1,1]
   }
   n40 = "Constraint - dyn_parent_shldr_frillF_01_l (PPPPP)"!(
    ComponentPose=@n39
    Alpha=1
    ConstraintWeights_0=0.0514
    ConstraintWeights_1:float=0.0107
    ConstraintWeights_2:float=0.1028
    ConstraintWeights_3:float=0.6779
    ConstraintWeights_4:float=0.1571
   ) -> (Pose) {
    alphaBoolBlend {blendOption=Linear}
    boneToModify {boneName=dyn_parent_shldr_frillF_01_l}
    constraintSetup [
     {targetBone {boneName=spine_04}}
     {targetBone {boneName=spine_02}}
     {targetBone {boneName=spine_03}}
     {targetBone {boneName=shldr_pad_01_l}}
     {targetBone {boneName=spine_05}}
    ]
    constraintWeights [1,1,1,1,1]
   }
   n41 = "Constraint - dyn_parent_shldr_frillG_01_l (PPPP)"!(
    ComponentPose=@n40
    Alpha=1
    ConstraintWeights_0=0.0498
    ConstraintWeights_1:float=0.1061
    ConstraintWeights_2:float=0.6872
    ConstraintWeights_3:float=0.1569
   ) -> (Pose) {
    alphaBoolBlend {blendOption=Linear}
    boneToModify {boneName=dyn_parent_shldr_frillG_01_l}
    constraintSetup [{targetBone {boneName=spine_04}}, {targetBone {boneName=spine_03}}, {targetBone {boneName=shldr_pad_01_l}}, {targetBone {boneName=spine_05}}]
    constraintWeights [1,1,1,1]
   }
   n42 = "Constraint - dyn_parent_shldr_frillH_01_l (PPPP)"!(
    ComponentPose=@n41
    Alpha=1
    ConstraintWeights_0=0.0475
    ConstraintWeights_1:float=0.1084
    ConstraintWeights_2:float=0.6889
    ConstraintWeights_3:float=0.1552
   ) -> (Pose) {
    alphaBoolBlend {blendOption=Linear}
    boneToModify {boneName=dyn_parent_shldr_frillH_01_l}
    constraintSetup [{targetBone {boneName=spine_04}}, {targetBone {boneName=spine_03}}, {targetBone {boneName=shldr_pad_01_l}}, {targetBone {boneName=spine_05}}]
    constraintWeights [1,1,1,1]
   }
   n43 = "Constraint - dyn_parent_shldr_frillI_01_l (PPPP)"!(
    ComponentPose=@n42
    Alpha=1
    ConstraintWeights_0=0.0453
    ConstraintWeights_1:float=0.1107
    ConstraintWeights_2:float=0.6906
    ConstraintWeights_3:float=0.1535
   ) -> (Pose) {
    alphaBoolBlend {blendOption=Linear}
    boneToModify {boneName=dyn_parent_shldr_frillI_01_l}
    constraintSetup [{targetBone {boneName=spine_04}}, {targetBone {boneName=spine_03}}, {targetBone {boneName=shldr_pad_01_l}}, {targetBone {boneName=spine_05}}]
    constraintWeights [1,1,1,1]
   }
   n44 = "Constraint - dyn_parent_shldr_frillJ_01_l (PPPP)"!(
    ComponentPose=@n43
    Alpha=1
    ConstraintWeights_0=0.0432
    ConstraintWeights_1:float=0.1129
    ConstraintWeights_2:float=0.6919
    ConstraintWeights_3:float=0.152
   ) -> (Pose) {
    alphaBoolBlend {blendOption=Linear}
    boneToModify {boneName=dyn_parent_shldr_frillJ_01_l}
    constraintSetup [{targetBone {boneName=spine_04}}, {targetBone {boneName=spine_03}}, {targetBone {boneName=shldr_pad_01_l}}, {targetBone {boneName=spine_05}}]
    constraintWeights [1,1,1,1]
   }
   n45 = "Constraint - dyn_parent_shldr_frillK_01_l (PPPP)"!(
    ComponentPose=@n44
    Alpha=1
    ConstraintWeights_0=0.0414
    ConstraintWeights_1:float=0.1149
    ConstraintWeights_2:float=0.6928
    ConstraintWeights_3:float=0.1509
   ) -> (Pose) {
    alphaBoolBlend {blendOption=Linear}
    boneToModify {boneName=dyn_parent_shldr_frillK_01_l}
    constraintSetup [{targetBone {boneName=spine_04}}, {targetBone {boneName=spine_03}}, {targetBone {boneName=shldr_pad_01_l}}, {targetBone {boneName=spine_05}}]
    constraintWeights [1,1,1,1]
   }
   n46 = "Constraint - dyn_parent_shldr_frillL_01_l (PPPP)"!(
    ComponentPose=@n45
    Alpha=1
    ConstraintWeights_0=0.041
    ConstraintWeights_1:float=0.1165
    ConstraintWeights_2:float=0.69
    ConstraintWeights_3:float=0.1525
   ) -> (Pose) {
    alphaBoolBlend {blendOption=Linear}
    boneToModify {boneName=dyn_parent_shldr_frillL_01_l}
    constraintSetup [{targetBone {boneName=spine_04}}, {targetBone {boneName=spine_03}}, {targetBone {boneName=shldr_pad_01_l}}, {targetBone {boneName=spine_05}}]
    constraintWeights [1,1,1,1]
   }
   n47 = "Constraint - dyn_parent_shldr_frillM_01_l (PPPP)"!(
    ComponentPose=@n46
    Alpha=1
    ConstraintWeights_0=0.0417
    ConstraintWeights_1:float=0.1183
    ConstraintWeights_2:float=0.6835
    ConstraintWeights_3:float=0.1565
   ) -> (Pose) {
    alphaBoolBlend {blendOption=Linear}
    boneToModify {boneName=dyn_parent_shldr_frillM_01_l}
    constraintSetup [{targetBone {boneName=spine_04}}, {targetBone {boneName=spine_03}}, {targetBone {boneName=shldr_pad_01_l}}, {targetBone {boneName=spine_05}}]
    constraintWeights [1,1,1,1]
   }
   n48 = "Constraint - dyn_parent_shldr_frillN_01_l (PPPP)"!(
    ComponentPose=@n47
    Alpha=1
    ConstraintWeights_0=0.0426
    ConstraintWeights_1:float=0.1205
    ConstraintWeights_2:float=0.6752
    ConstraintWeights_3:float=0.1617
   ) -> (Pose) {
    alphaBoolBlend {blendOption=Linear}
    boneToModify {boneName=dyn_parent_shldr_frillN_01_l}
    constraintSetup [{targetBone {boneName=spine_04}}, {targetBone {boneName=spine_03}}, {targetBone {boneName=shldr_pad_01_l}}, {targetBone {boneName=spine_05}}]
    constraintWeights [1,1,1,1]
   }
   n49 = "Constraint - dyn_parent_shldr_frillO_01_l (PPPP)"!(
    ComponentPose=@n48
    Alpha=1
    ConstraintWeights_0=0.0437
    ConstraintWeights_1:float=0.123
    ConstraintWeights_2:float=0.6655
    ConstraintWeights_3:float=0.1678
   ) -> (Pose) {
    alphaBoolBlend {blendOption=Linear}
    boneToModify {boneName=dyn_parent_shldr_frillO_01_l}
    constraintSetup [{targetBone {boneName=spine_04}}, {targetBone {boneName=spine_03}}, {targetBone {boneName=shldr_pad_01_l}}, {targetBone {boneName=spine_05}}]
    constraintWeights [1,1,1,1]
   }
   n50 = "Constraint - dyn_parent_shldr_frillP_01_l (PPPP)"!(
    ComponentPose=@n49
    Alpha=1
    ConstraintWeights_0=0.0447
    ConstraintWeights_1:float=0.1257
    ConstraintWeights_2:float=0.6556
    ConstraintWeights_3:float=0.174
   ) -> (Pose) {
    alphaBoolBlend {blendOption=Linear}
    boneToModify {boneName=dyn_parent_shldr_frillP_01_l}
    constraintSetup [{targetBone {boneName=spine_04}}, {targetBone {boneName=spine_03}}, {targetBone {boneName=shldr_pad_01_l}}, {targetBone {boneName=spine_05}}]
    constraintWeights [1,1,1,1]
   }
   n51 = "Constraint - dyn_parent_shldr_frillQ_01_l (PPPP)"!(
    ComponentPose=@n50
    Alpha=1
    ConstraintWeights_0=0.0456
    ConstraintWeights_1:float=0.1283
    ConstraintWeights_2:float=0.6462
    ConstraintWeights_3:float=0.1799
   ) -> (Pose) {
    alphaBoolBlend {blendOption=Linear}
    boneToModify {boneName=dyn_parent_shldr_frillQ_01_l}
    constraintSetup [{targetBone {boneName=spine_04}}, {targetBone {boneName=spine_03}}, {targetBone {boneName=shldr_pad_01_l}}, {targetBone {boneName=spine_05}}]
    constraintWeights [1,1,1,1]
   }
   n52 = "Constraint - dyn_parent_shldr_frillA_01_r (PPPP)"!(
    ComponentPose=@n51
    Alpha=1
    ConstraintWeights_0=0.6365
    ConstraintWeights_1:float=0.1024
    ConstraintWeights_2:float=0.0637
    ConstraintWeights_3:float=0.1973
   ) -> (Pose) {
    alphaBoolBlend {blendOption=Linear}
    boneToModify {boneName=dyn_parent_shldr_frillA_01_r}
    constraintSetup [{targetBone {boneName=shldr_pad_01_r}}, {targetBone {boneName=spine_03}}, {targetBone {boneName=spine_04}}, {targetBone {boneName=spine_05}}]
    constraintWeights [1,1,1,1]
   }
   n53 = "Constraint - dyn_parent_shldr_frillB_01_r (PPPP)"!(
    ComponentPose=@n52
    Alpha=1
    ConstraintWeights_0=0.6479
    ConstraintWeights_1:float=0.1017
    ConstraintWeights_2:float=0.0609
    ConstraintWeights_3:float=0.1895
   ) -> (Pose) {
    alphaBoolBlend {blendOption=Linear}
    boneToModify {boneName=dyn_parent_shldr_frillB_01_r}
    constraintSetup [{targetBone {boneName=shldr_pad_01_r}}, {targetBone {boneName=spine_03}}, {targetBone {boneName=spine_04}}, {targetBone {boneName=spine_05}}]
    constraintWeights [1,1,1,1]
   }
   n54 = "Constraint - dyn_parent_shldr_frillC_01_r (PPPP)"!(
    ComponentPose=@n53
    Alpha=1
    ConstraintWeights_0=0.6588
    ConstraintWeights_1:float=0.1012
    ConstraintWeights_2:float=0.058
    ConstraintWeights_3:float=0.1819
   ) -> (Pose) {
    alphaBoolBlend {blendOption=Linear}
    boneToModify {boneName=dyn_parent_shldr_frillC_01_r}
    constraintSetup [{targetBone {boneName=shldr_pad_01_r}}, {targetBone {boneName=spine_03}}, {targetBone {boneName=spine_04}}, {targetBone {boneName=spine_05}}]
    constraintWeights [1,1,1,1]
   }
   n55 = "Constraint - dyn_parent_shldr_frillD_01_r (PPPP)"!(
    ComponentPose=@n54
    Alpha=1
    ConstraintWeights_0=0.6674
    ConstraintWeights_1:float=0.1008
    ConstraintWeights_2:float=0.0558
    ConstraintWeights_3:float=0.176
   ) -> (Pose) {
    alphaBoolBlend {blendOption=Linear}
    boneToModify {boneName=dyn_parent_shldr_frillD_01_r}
    constraintSetup [{targetBone {boneName=shldr_pad_01_r}}, {targetBone {boneName=spine_03}}, {targetBone {boneName=spine_04}}, {targetBone {boneName=spine_05}}]
    constraintWeights [1,1,1,1]
   }
   n56 = "Constraint - dyn_parent_shldr_frillE_01_r (PPPP)"!(
    ComponentPose=@n55
    Alpha=1
    ConstraintWeights_0=0.6737
    ConstraintWeights_1:float=0.1018
    ConstraintWeights_2:float=0.0534
    ConstraintWeights_3:float=0.1712
   ) -> (Pose) {
    alphaBoolBlend {blendOption=Linear}
    boneToModify {boneName=dyn_parent_shldr_frillE_01_r}
    constraintSetup [{targetBone {boneName=shldr_pad_01_r}}, {targetBone {boneName=spine_03}}, {targetBone {boneName=spine_04}}, {targetBone {boneName=spine_05}}]
    constraintWeights [1,1,1,1]
   }
   n57 = "Constraint - dyn_parent_shldr_frillF_01_r (PPPP)"!(
    ComponentPose=@n56
    Alpha=1
    ConstraintWeights_0=0.6769
    ConstraintWeights_1:float=0.104
    ConstraintWeights_2:float=0.0509
    ConstraintWeights_3:float=0.1681
   ) -> (Pose) {
    alphaBoolBlend {blendOption=Linear}
    boneToModify {boneName=dyn_parent_shldr_frillF_01_r}
    constraintSetup [{targetBone {boneName=shldr_pad_01_r}}, {targetBone {boneName=spine_03}}, {targetBone {boneName=spine_04}}, {targetBone {boneName=spine_05}}]
    constraintWeights [1,1,1,1]
   }
   n58 = "Constraint - dyn_parent_shldr_frillG_01_r (PPPP)"!(
    ComponentPose=@n57
    Alpha=1
    ConstraintWeights_0=0.6795
    ConstraintWeights_1:float=0.1063
    ConstraintWeights_2:float=0.0487
    ConstraintWeights_3:float=0.1655
   ) -> (Pose) {
    alphaBoolBlend {blendOption=Linear}
    boneToModify {boneName=dyn_parent_shldr_frillG_01_r}
    constraintSetup [{targetBone {boneName=shldr_pad_01_r}}, {targetBone {boneName=spine_03}}, {targetBone {boneName=spine_04}}, {targetBone {boneName=spine_05}}]
    constraintWeights [1,1,1,1]
   }
   n59 = "Constraint - dyn_parent_shldr_frillH_01_r (PPPP)"!(
    ComponentPose=@n58
    Alpha=1
    ConstraintWeights_0=0.6818
    ConstraintWeights_1:float=0.1087
    ConstraintWeights_2:float=0.0465
    ConstraintWeights_3:float=0.163
   ) -> (Pose) {
    alphaBoolBlend {blendOption=Linear}
    boneToModify {boneName=dyn_parent_shldr_frillH_01_r}
    constraintSetup [{targetBone {boneName=shldr_pad_01_r}}, {targetBone {boneName=spine_03}}, {targetBone {boneName=spine_04}}, {targetBone {boneName=spine_05}}]
    constraintWeights [1,1,1,1]
   }
   n60 = "Constraint - dyn_parent_shldr_frillI_01_r (PPPP)"!(
    ComponentPose=@n59
    Alpha=1
    ConstraintWeights_0=0.684
    ConstraintWeights_1:float=0.1111
    ConstraintWeights_2:float=0.0443
    ConstraintWeights_3:float=0.1606
   ) -> (Pose) {
    alphaBoolBlend {blendOption=Linear}
    boneToModify {boneName=dyn_parent_shldr_frillI_01_r}
    constraintSetup [{targetBone {boneName=shldr_pad_01_r}}, {targetBone {boneName=spine_03}}, {targetBone {boneName=spine_04}}, {targetBone {boneName=spine_05}}]
    constraintWeights [1,1,1,1]
   }
   n61 = "Constraint - dyn_parent_shldr_frillJ_01_r (PPPP)"!(
    ComponentPose=@n60
    Alpha=1
    ConstraintWeights_0=0.6858
    ConstraintWeights_1:float=0.1133
    ConstraintWeights_2:float=0.0423
    ConstraintWeights_3:float=0.1585
   ) -> (Pose) {
    alphaBoolBlend {blendOption=Linear}
    boneToModify {boneName=dyn_parent_shldr_frillJ_01_r}
    constraintSetup [{targetBone {boneName=shldr_pad_01_r}}, {targetBone {boneName=spine_03}}, {targetBone {boneName=spine_04}}, {targetBone {boneName=spine_05}}]
    constraintWeights [1,1,1,1]
   }
   n62 = "Constraint - dyn_parent_shldr_frillK_01_r (PPPP)"!(
    ComponentPose=@n61
    Alpha=1
    ConstraintWeights_0=0.6871
    ConstraintWeights_1:float=0.1155
    ConstraintWeights_2:float=0.0406
    ConstraintWeights_3:float=0.1569
   ) -> (Pose) {
    alphaBoolBlend {blendOption=Linear}
    boneToModify {boneName=dyn_parent_shldr_frillK_01_r}
    constraintSetup [{targetBone {boneName=shldr_pad_01_r}}, {targetBone {boneName=spine_03}}, {targetBone {boneName=spine_04}}, {targetBone {boneName=spine_05}}]
    constraintWeights [1,1,1,1]
   }
   n63 = "Constraint - dyn_parent_shldr_frillL_01_r (PPPP)"!(
    ComponentPose=@n62
    Alpha=1
    ConstraintWeights_0=0.6843
    ConstraintWeights_1:float=0.1172
    ConstraintWeights_2:float=0.0403
    ConstraintWeights_3:float=0.1582
   ) -> (Pose) {
    alphaBoolBlend {blendOption=Linear}
    boneToModify {boneName=dyn_parent_shldr_frillL_01_r}
    constraintSetup [{targetBone {boneName=shldr_pad_01_r}}, {targetBone {boneName=spine_03}}, {targetBone {boneName=spine_04}}, {targetBone {boneName=spine_05}}]
    constraintWeights [1,1,1,1]
   }
   n64 = "Constraint - dyn_parent_shldr_frillM_01_r (PPPP)"!(
    ComponentPose=@n63
    Alpha=1
    ConstraintWeights_0=0.6775
    ConstraintWeights_1:float=0.1191
    ConstraintWeights_2:float=0.0411
    ConstraintWeights_3:float=0.1623
   ) -> (Pose) {
    alphaBoolBlend {blendOption=Linear}
    boneToModify {boneName=dyn_parent_shldr_frillM_01_r}
    constraintSetup [{targetBone {boneName=shldr_pad_01_r}}, {targetBone {boneName=spine_03}}, {targetBone {boneName=spine_04}}, {targetBone {boneName=spine_05}}]
    constraintWeights [1,1,1,1]
   }
   n65 = "Constraint - dyn_parent_shldr_frillN_01_r (PPPP)"!(
    ComponentPose=@n64
    Alpha=1
    ConstraintWeights_0=0.6687
    ConstraintWeights_1:float=0.1215
    ConstraintWeights_2:float=0.0422
    ConstraintWeights_3:float=0.1675
   ) -> (Pose) {
    alphaBoolBlend {blendOption=Linear}
    boneToModify {boneName=dyn_parent_shldr_frillN_01_r}
    constraintSetup [{targetBone {boneName=shldr_pad_01_r}}, {targetBone {boneName=spine_03}}, {targetBone {boneName=spine_04}}, {targetBone {boneName=spine_05}}]
    constraintWeights [1,1,1,1]
   }
   n66 = "Constraint - dyn_parent_shldr_frillO_01_r (PPPP)"!(
    ComponentPose=@n65
    Alpha=1
    ConstraintWeights_0=0.6585
    ConstraintWeights_1:float=0.1243
    ConstraintWeights_2:float=0.0435
    ConstraintWeights_3:float=0.1737
   ) -> (Pose) {
    alphaBoolBlend {blendOption=Linear}
    boneToModify {boneName=dyn_parent_shldr_frillO_01_r}
    constraintSetup [{targetBone {boneName=shldr_pad_01_r}}, {targetBone {boneName=spine_03}}, {targetBone {boneName=spine_04}}, {targetBone {boneName=spine_05}}]
    constraintWeights [1,1,1,1]
   }
   n67 = "Constraint - dyn_parent_shldr_frillP_01_r (PPPP)"!(
    ComponentPose=@n66
    Alpha=1
    ConstraintWeights_0=0.648
    ConstraintWeights_1:float=0.1272
    ConstraintWeights_2:float=0.0448
    ConstraintWeights_3:float=0.18
   ) -> (Pose) {
    alphaBoolBlend {blendOption=Linear}
    boneToModify {boneName=dyn_parent_shldr_frillP_01_r}
    constraintSetup [{targetBone {boneName=shldr_pad_01_r}}, {targetBone {boneName=spine_03}}, {targetBone {boneName=spine_04}}, {targetBone {boneName=spine_05}}]
    constraintWeights [1,1,1,1]
   }
   n68 = "Constraint - dyn_parent_shldr_frillQ_01_r (PPPP)"!(
    ComponentPose=@n67
    Alpha=1
    ConstraintWeights_0=0.638
    ConstraintWeights_1:float=0.1302
    ConstraintWeights_2:float=0.0459
    ConstraintWeights_3:float=0.1859
   ) -> (Pose) {
    alphaBoolBlend {blendOption=Linear}
    boneToModify {boneName=dyn_parent_shldr_frillQ_01_r}
    constraintSetup [{targetBone {boneName=shldr_pad_01_r}}, {targetBone {boneName=spine_03}}, {targetBone {boneName=spine_04}}, {targetBone {boneName=spine_05}}]
    constraintWeights [1,1,1,1]
   }
   n69 = RigidBody(ComponentPose=@n68, Alpha=Get target_alpha_shldr() -> (target_alpha_shldr:double)) {
    bClampLinearTranslationLimitToRefPose=true
    bOverrideWorldGravity=true
    overridePhysicsAsset=&/M_MED_DJ_Remix_Tier2_Shldr_Physics
    simulationSpace=BaseBoneSpace
    alphaBoolBlend {blendOption=Linear}
    baseBoneRef {boneName=root}
    componentAppliedLinearAccClamp=<1200 1200 1300>
    componentLinearAccScale=<0.3 0.3 0.5>
    componentLinearVelScale=<0.3 0.3 0.5>
    overrideWorldGravity {z=-850}
   }
   n70 = RigidBody(ComponentPose=@n69, Alpha=0.75) {
    bClampLinearTranslationLimitToRefPose=true
    bOverrideWorldGravity=true
    overridePhysicsAsset=&/M_MED_DJ_Remix_Tier2_Physics
    simulationSpace=BaseBoneSpace
    alphaBoolBlend {blendOption=Linear}
    baseBoneRef {boneName=root}
    componentAppliedLinearAccClamp=<1200 1200 1200>
    componentLinearAccScale=<0.8 0.8 0.6>
    componentLinearVelScale=<0.8 0.8 0.6>
    overrideWorldGravity {z=-900}
   }
   n71 = RigidBody(ComponentPose=@n70) {
    bClampLinearTranslationLimitToRefPose=true
    bOverrideWorldGravity=true
    overridePhysicsAsset=&M_MED_DJ_Remix_Physics
    simulationSpace=BaseBoneSpace
    alphaBoolBlend {blendOption=Linear}
    baseBoneRef {boneName=spine_05}
    componentAppliedLinearAccClamp=<20000 20000 20000>
    componentLinearAccScale=<1 1 1>
    componentLinearVelScale=<1 1 1>
    overrideWorldGravity {z=-980}
   }
   n72 = "Transform (Modify) Bone - Bone: dyn_parent_collar_r"(
    ComponentPose=@n71
    Alpha=Get ArmUpperRightAlpha() -> (ArmUpperRightAlpha:float)
    Translation="3,0,0"
    Rotation="-23.243,-15.9,6.75"
    Scale="0.8,0.8,0.8"
   ) {
    rotationMode=BMM_Additive
    rotationSpace=BCS_BoneSpace
    scaleMode=BMM_Additive
    scaleSpace=BCS_BoneSpace
    translationMode=BMM_Additive
    translationSpace=BCS_BoneSpace
    alphaBoolBlend {blendOption=Linear}
    alphaScaleBiasClamp {bInterpResult=true, interpSpeedDecreasing=15, interpSpeedIncreasing=15}
    boneToModify {boneName=dyn_parent_collar_r}
    rotation=<rot -23.243 -15.9 6.75>
    scale=<0.8 0.8 0.8>
    translation {x=3}
   }
   n73 = "Transform (Modify) Bone - Bone: dyn_parent_collar_l"(
    ComponentPose=@n72
    Alpha=Get ArmUpperLeftAlpha() -> (ArmUpperLeftAlpha:float)
    Translation="-3,0,0"
    Rotation="25,0.000,0.000"
    Scale="1,1,1"
   ) {
    rotationMode=BMM_Additive
    rotationSpace=BCS_BoneSpace
    translationMode=BMM_Additive
    translationSpace=BCS_BoneSpace
    alphaBoolBlend {blendOption=Linear}
    alphaScaleBiasClamp {bInterpResult=true, interpSpeedDecreasing=15, interpSpeedIncreasing=15}
    boneToModify {boneName=dyn_parent_collar_l}
    rotation {pitch=25}
    translation {x=-3}
   }
   n74 = "Transform (Modify) Bone - Bone: dyn_rapel_r"(
    ComponentPose=@n73
    Alpha=Get ArmLowerRightAlpha() -> (ArmLowerRightAlpha:float)
    Translation="1.5,-3,0"
    Rotation="0.000,-30,0.000"
    Scale="0.8,0.8,0.8"
   ) {
    rotationMode=BMM_Additive
    rotationSpace=BCS_BoneSpace
    scaleMode=BMM_Additive
    scaleSpace=BCS_BoneSpace
    translationMode=BMM_Additive
    translationSpace=BCS_BoneSpace
    alphaBoolBlend {blendOption=Linear}
    alphaScaleBiasClamp {bInterpResult=true, interpSpeedDecreasing=15, interpSpeedIncreasing=15}
    boneToModify {boneName=dyn_rapel_r}
    rotation {yaw=-30}
    scale=<0.8 0.8 0.8>
    translation=<1.5 -3>
   }
   n75 = "Transform (Modify) Bone - Bone: dyn_lapel_l"(
    ComponentPose=@n74
    Alpha=Get ArmLowerLeftAlpha() -> (ArmLowerLeftAlpha:float)
    Translation="-1,0,0"
    Rotation="0,-30,0.000"
    Scale="0.9,0.9,0.9"
   ) {
    rotationMode=BMM_Additive
    rotationSpace=BCS_BoneSpace
    translationMode=BMM_Additive
    translationSpace=BCS_BoneSpace
    alphaBoolBlend {blendOption=Linear}
    alphaScaleBiasClamp {bInterpResult=true, interpSpeedDecreasing=15, interpSpeedIncreasing=15}
    boneToModify {boneName=dyn_lapel_l}
    rotation {yaw=-30}
    scale=<0.9 0.9 0.9>
    translation {x=-1}
   }
   n76 = "Transform (Modify) Bone - Bone: dyn_collar_01_bck"(ComponentPose=@n75, Alpha=1, Translation="0,0,0", Rotation="0,-35,0", Scale="1,1,1") {
    rotationMode=BMM_Additive
    rotationSpace=BCS_BoneSpace
    scaleMode=BMM_Additive
    translationMode=BMM_Additive
    translationSpace=BCS_BoneSpace
    alphaBoolBlend {blendOption=Linear}
    alphaScaleBiasClamp {interpSpeedDecreasing=15, interpSpeedIncreasing=15}
    boneToModify {boneName=dyn_collar_01_bck}
    rotation {yaw=-35}
   }
   n77 = "Constraint - dyn_parent_collar_bck (R)"(ComponentPose=@n76, Alpha=1, ConstraintWeights_0=1) {
    alphaBoolBlend {blendOption=Linear}
    boneToModify {boneName=dyn_parent_collar_bck}
    constraintSetup [{transformType=Rotation, targetBone {boneName=head}}]
    constraintWeights [1]
   }
   n78 = "Transform (Modify) Bone - Bone: parent_shldr_pad_01_r"(ComponentPose=@n77, Alpha=Get ArmUpperRightAlpha(), Translation="0.300,-0.108,2.718", Rotation="-5.749,-16.307,-9.551", Scale="1,1,1") {
    rotationMode=BMM_Additive
    rotationSpace=BCS_BoneSpace
    scaleMode=BMM_Additive
    scaleSpace=BCS_BoneSpace
    translationMode=BMM_Additive
    translationSpace=BCS_BoneSpace
    alphaBoolBlend {blendOption=Linear}
    alphaScaleBiasClamp {bInterpResult=true, interpSpeedDecreasing=15, interpSpeedIncreasing=15}
    boneToModify {boneName=parent_shldr_pad_01_r}
    rotation=<rot -5.749 -16.3072 -9.5512>
    translation=<0.3 -0.108 2.718>
   }
   n79 = "Transform (Modify) Bone - Bone: parent_shldr_pad_01_l"(ComponentPose=@n78, Alpha=Get ArmUpperLeftAlpha(), Translation="-0.180,0.223,-1.967", Rotation="-3.682,-14.239,-4.977", Scale="1,1,1") {
    rotationMode=BMM_Additive
    rotationSpace=BCS_BoneSpace
    scaleMode=BMM_Additive
    scaleSpace=BCS_BoneSpace
    translationMode=BMM_Additive
    translationSpace=BCS_BoneSpace
    alphaBoolBlend {blendOption=Linear}
    alphaScaleBiasClamp {bInterpResult=true, interpSpeedDecreasing=15, interpSpeedIncreasing=15}
    boneToModify {boneName=parent_shldr_pad_01_l}
    rotation=<rot -3.682 -14.239 -4.977>
    translation=<-0.18 0.223 -1.9671>
   }
   Output Pose(Result=Component To Local(ComponentPose=@n79)) {layerGroup=None}
  ]}
 EventGraph {nodes [
   Event OnBeginTargeting() {is=title K2Node_Event}
   > Set target_alpha_shldr(target_alpha_shldr:double=0.3) -> (Output_Get:double)
   Event OnEndTargeting() {is=title K2Node_Event}
   > Set target_alpha_shldr(target_alpha_shldr=0.6)
  ]}
]
variables [target_alpha_shldr {type=number, default=0.6}]
