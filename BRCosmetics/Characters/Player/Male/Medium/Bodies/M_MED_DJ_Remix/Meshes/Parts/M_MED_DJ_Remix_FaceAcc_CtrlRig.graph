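# Control Rig for the DJ Remix character's face accessory that re-aims the dyn_cord bone as the body moves.
# It rotates the cord's X axis toward a fixed offset point in neck_01 space using a from-two-vectors quaternion, propagating to children.
# Debug axes draw at the cord and target.
# An additive cord bend weighted by head rotation relative to spine_05 is present but not wired into execution.
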ControlRigBlueprint /BRCosmetics/Characters/Player/Male/Medium/Bodies/M_MED_DJ_Remix/Meshes/Parts/M_MED_DJ_Remix_FaceAcc_CtrlRig
graphs [
 RigVMFunctionLibrary {links [], nodes []}
 RigVMModel {nodes [
   n1 = &RigUnit_BeginExecution()
   n2 = &RigUnit_GetBoneTransform(Bone=dyn_cord)
   n3 = /Script/RigVM.RigVMFunction_MathQuaternionGetAxis(Axis=X, Quaternion=@n2.Transform.Rotation) {is=struct RigVMUnitNode, template="Axis::Execute(in Quaternion,in Axis,out Result)"}
   n4 = /Script/RigVM.RigVMFunction_MathTransformTransformVector(
    Transform=&RigUnit_GetBoneTransform(Bone=neck_01).Transform
    Location=/Script/RigVM.RigVMFunction_MathVectorAbs(Value=(X=0,Y=1,Z=9)) {is=struct RigVMUnitNode, template="Absolute::Execute(in Value,out Result)"}.Result
   ) {is=struct RigVMUnitNode}
   n5 = /Script/RigVM.RigVMFunction_MathVectorSub(B=@n2.Transform.Translation, A=@n4.Result) {is=struct RigVMUnitNode, template="Subtract::Execute(in A,in B,out Result)"}
   /Script/ControlRig.RigUnit_SetBoneRotation(
    Bone=dyn_cord
    Space=GlobalSpace
    Weight=1
    bPropagateToChildren=True
    ExecutePin=@n1.ExecutePin
    Rotation=/Script/RigVM.RigVMFunction_MathQuaternionMul(
     B=@n2.Transform.Rotation
     A=/Script/RigVM.RigVMFunction_MathQuaternionFromTwoVectors(A=@n3.Result, B=@n5.Result) {is=struct RigVMUnitNode, template="FromTwoVectors::Execute(in A,in B,out Result)"}.Result
    ) {is=struct RigVMUnitNode, template="Multiply::Execute(in A,in B,out Result)"}.Result
   ) {is=struct RigVMUnitNode}
   > /Script/ControlRig.RigUnit_DebugTransformMutable(
    Color=(R=1,G=0,B=0,A=1)
    DebugDrawSettings=(DepthPriority=SDPG_Foreground,Lifetime=-1)
    Mode=Axes
    Scale=10
    Space=dyn_cord
    Thickness=1
    Transform=(Rotation=(X=0,Y=0,Z=0,W=1),Translation=(X=0,Y=0,Z=0),Scale3D=(X=1,Y=1,Z=1))
    WorldOffset=(Rotation=(X=0,Y=0,Z=0,W=1),Translation=(X=0,Y=0,Z=0),Scale3D=(X=1,Y=1,Z=1))
    bEnabled=True
   ) {is=struct RigVMUnitNode}
   > &RigUnit_DebugTransformMutable(
    Color=(R=1,G=0,B=0,A=1)
    DebugDrawSettings=(DepthPriority=SDPG_Foreground,Lifetime=-1)
    Mode=Axes
    Scale=55
    Space=None
    Thickness=0
    Transform=(Rotation=(X=0,Y=0,Z=0,W=1),Translation=(X=0,Y=0,Z=0),Scale3D=(X=1,Y=1,Z=1))
    WorldOffset=(Rotation=(X=0,Y=0,Z=0,W=1),Translation=(X=0,Y=0,Z=0),Scale3D=(X=1,Y=1,Z=1))
    bEnabled=True
    Transform.Translation=@n4.Result
   )
   &RigUnit_BeginExecution()
   &RigUnit_DebugTransformMutable(
    Color=(R=1,G=0,B=0,A=1)
    DebugDrawSettings=(DepthPriority=SDPG_Foreground,Lifetime=-1)
    ExecutePin=()
    Mode=Axes
    Scale=50
    Space=dyn_cord
    Thickness=0
    Transform=(Rotation=(X=0,Y=0,Z=0,W=1),Translation=(X=0,Y=0,Z=0),Scale3D=(X=1,Y=1,Z=1))
    WorldOffset=(Rotation=(X=0,Y=0,Z=0,W=1),Translation=(X=0,Y=0,Z=0),Scale3D=(X=1,Y=1,Z=1))
    bEnabled=True
   )
   n6 = &RigUnit_GetRelativeBoneTransform(Bone=head, Space=spine_05)
   /Script/ControlRig.RigUnit_ModifyBoneTransforms(
    BoneToModify=((Bone="dyn_cord",Transform=(Rotation=(X=0,Y=0,Z=0,W=1),Translation=(X=0,Y=0,Z=0),Scale3D=(X=1,Y=1,Z=1))))
    ExecutePin=()
    Mode=AdditiveLocal
    WeightMaximum=1
    WeightMinimum=0
    Weight=&RigVMFunction_MathFloatRemap(
     SourceMaximum=-40
     SourceMinimum=0
     TargetMinimum=0
     bClamp=true
     Value=&RigVMFunction_MathQuaternionToEuler(RotationOrder=ZYX, Value=@n6.Transform.Rotation).Result.Z
    ).Result
    BoneToModify.0.Transform.Rotation=&RigVMFunction_MathQuaternionFromEuler(Euler=(X=0,Y=0,Z=-35)).Result
   ) {is=struct RigVMUnitNode}
  ]}
]
hierarchy [
 root {loc=<0 0 0>, rot=<0 0 0 1>}
 pelvis {parent=root, loc=<0 -0.7071 93.1191>, rot=<0.0127 0.707 -0.0127 -0.707>}
 spine_01 {parent=pelvis, loc=<6.7494 -0.6388 0>, rot=<0 0 0.034 -0.9994>}
 spine_02 {parent=spine_01, loc=<7.2518 -0.035 0>, rot=<0 0 0.0238 -0.9997>}
 spine_03 {parent=spine_02, loc=<6.0104 0 0>, rot=<0 0 0.0149 -0.9999>}
 spine_04 {parent=spine_03, loc=<8.1024 0 0>, rot=<0 0 -0.0132 -0.9999>}
 spine_05 {parent=spine_04, loc=<10.0146 0 0>, rot=<0 0 -0.0015 -1>}
 clavicle_r {parent=spine_05, loc=<4.3575 -0.1466 2.3223>, rot=<0.7289 0.0465 -0.6829 0.009>}
 neck_01 {parent=spine_05, loc=<11.3138 0.1529 0>, rot=<0 0 -0.2003 -0.9797>}
 neck_02 {parent=neck_01, loc=<4.5627 0 0>, rot=<0 0 0.0504 -0.9987>}
 head {parent=neck_02, loc=<5.3914 0 0>, rot=<0 0 0.115 -0.9934>}
 dyn_cord {parent=head, loc=<3.8715 1.0752 10.3067>, rot=<0.0375 -0.995 0.0012 -0.0927>}
]
