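# Physics asset for the hair and face accessory mesh of the medium male DJ Remix player cosmetic, giving capsule bodies to the head and dynamic hair bones.
# Five two-bone hair chains (A, Q, EE, DD, M) branch from a shared hair base, which is constrained to the head.
# Joints share uniform angular limit values, but the base-to-head, DD tip and M root joints set swing and twist free.
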PhysicsAsset /BRCosmetics/Characters/Player/Male/Medium/Bodies/M_MED_DJ_Remix/Meshes/Parts/M_MED_DJ_Remix_Hair_FaceAcc_Physics
bodies {
 dyn_hairA_01 [{shapeType=Capsule, length=1.5, radius=1.3, center=<1 0 0>, rotation=<rot -90 0 0>}]
 dyn_hairA_02 [{shapeType=Capsule, length=0.75, radius=0.8, center=<1.0476 0 0>, rotation=<rot -90 0 0>}]
 dyn_hairDD_01 [{shapeType=Capsule, length=1.75, radius=1.05, center=<1 0 0>, rotation=<rot -90 0 0>}]
 dyn_hairDD_02 [{shapeType=Capsule, length=1.8, radius=0.75, center=<1.25 0 0>, rotation=<rot -90 0 0>}]
 dyn_hairEE_01 [{shapeType=Capsule, length=1.5, radius=1.3, center=<1 0 0>, rotation=<rot -90 0 0>}]
 dyn_hairEE_02 [{shapeType=Capsule, length=1.25, radius=0.8, center=<1.6359 0 0>, rotation=<rot -90 0 0>}]
 dyn_hairM_01 [{shapeType=Capsule, length=0.6906, radius=0.6497, center=<0.3521 0.1187 0.0287>, rotation=<rot 61.6054 -83.3067 5.8941>}]
 dyn_hairM_02 [{shapeType=Capsule, length=0.75, radius=0.8, center=<1.0476 0 0>, rotation=<rot -90 0 0>}]
 dyn_hairQ_01 [{shapeType=Capsule, length=1.5, radius=1.3, center=<1 0 0>, rotation=<rot -90 0 0>}]
 dyn_hairQ_02 [{shapeType=Capsule, length=0.35, radius=1.05, center=<1 0 0>, rotation=<rot -90 0 0>}]
 dyn_hair_base [{shapeType=Capsule, length=1.0063, radius=0.8778, center=<-0.0024 -0.4112 6.5081>, rotation=<rot 0 -162.1614 -89.0566>}]
 head [{shapeType=Capsule, length=4.2854, radius=6.9437, center=<3.761 4.2812 -0.0322>, rotation=<rot 0 61.7431 -89.6235>}]
}
constraints [
 {
  bone1Name=dyn_hair_base
  bone2Name=head
  swing1LimitDegrees=45
  swing1Motion=Free
  swing2LimitDegrees=45
  swing2Motion=Free
  twistLimitDegrees=45
  twistMotion=Free
 }
 {bone1Name=dyn_hairA_01, bone2Name=dyn_hair_base, swing1LimitDegrees=45, swing2LimitDegrees=45, twistLimitDegrees=45}
 {bone1Name=dyn_hairA_02, bone2Name=dyn_hairA_01, swing1LimitDegrees=45, swing2LimitDegrees=45, twistLimitDegrees=45}
 {bone1Name=dyn_hairQ_01, bone2Name=dyn_hair_base, swing1LimitDegrees=45, swing2LimitDegrees=45, twistLimitDegrees=45}
 {bone1Name=dyn_hairQ_02, bone2Name=dyn_hairQ_01, swing1LimitDegrees=45, swing2LimitDegrees=45, twistLimitDegrees=45}
 {bone1Name=dyn_hairEE_01, bone2Name=dyn_hair_base, swing1LimitDegrees=45, swing2LimitDegrees=45, twistLimitDegrees=45}
 {bone1Name=dyn_hairEE_02, bone2Name=dyn_hairEE_01, swing1LimitDegrees=45, swing2LimitDegrees=45, twistLimitDegrees=45}
 {bone1Name=dyn_hairDD_01, bone2Name=dyn_hair_base, swing1LimitDegrees=45, swing2LimitDegrees=45, twistLimitDegrees=45}
 {
  bone1Name=dyn_hairDD_02
  bone2Name=dyn_hairDD_01
  swing1LimitDegrees=45
  swing1Motion=Free
  swing2LimitDegrees=45
  swing2Motion=Free
  twistLimitDegrees=45
  twistMotion=Free
 }
 {
  bone1Name=dyn_hairM_01
  bone2Name=dyn_hair_base
  swing1LimitDegrees=45
  swing1Motion=Free
  swing2LimitDegrees=45
  swing2Motion=Free
  twistLimitDegrees=45
  twistMotion=Free
 }
 {bone1Name=dyn_hairM_02, bone2Name=dyn_hairM_01, swing1LimitDegrees=45, swing2LimitDegrees=45, twistLimitDegrees=45}
]
