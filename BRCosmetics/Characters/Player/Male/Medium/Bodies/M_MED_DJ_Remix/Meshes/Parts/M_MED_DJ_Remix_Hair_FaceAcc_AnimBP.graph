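# Post-copy secondary-motion AnimBP that adds hair sway to the male DJ Remix character's hair and face-accessory part.
# Copies pose and curves from the attached parent, then runs a long serial chain of single-bone AnimDynamics nodes on dyn_hair bones.
# Each uses angular springs, limits, reduced gravity and its own fixed alpha; dyn_hairK_02 is simulated twice.
# A RigidBody branch with a physics asset is disconnected from the output, and the EventGraph is empty stubs.
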AnimBlueprint /BRCosmetics/Characters/Player/Male/Medium/Bodies/M_MED_DJ_Remix/Meshes/Parts/M_MED_DJ_Remix_Hair_FaceAcc_AnimBP
target_skeleton=&/M_MED_DJ_Remix_Hair_FaceAcc_Skeleton
graphs [
 AnimGraph {nodes [
   n1 = RigidBody(ComponentPose=Local To Component()) {
    bClampLinearTranslationLimitToRefPose=true
    bOverrideWorldGravity=true
    overridePhysicsAsset=&/M_MED_DJ_Remix_Hair_FaceAcc_Physics
    simulationSpace=BaseBoneSpace
    alphaBoolBlend {blendOption=Linear}
    baseBoneRef {boneName=root}
    componentAppliedLinearAccClamp=<1111 1111 1200>
    componentLinearAccScale=<1 1 1>
    componentLinearVelScale=<1 1 1>
    overrideWorldGravity {z=-980}
   }
   Component To Local(ComponentPose=@n1)
   n2 = Anim Dynamics - Bone: dyn_hairB_02(
    ComponentPose=Local To Component(LocalPose=Copy Pose From Mesh() {bCopyCurves=true, bUseAttachedParent=true})
    Alpha=1
    ExternalForce="0,0,0"
   ) {
    angularDampingOverride=0.995
    angularSpringConstant=100
    bAngularSpring=true
    bGravityOverrideInSimSpace=true
    bOverrideAngularDamping=true
    gravityScale=0.35
    simulationSpace=RootRelative
    alphaBoolBlend {blendOption=Linear}
    boundBone {boneName=dyn_hairB_02}
    componentAppliedLinearAccClamp=<3000 3000 1300>
    componentLinearAccScale=<0.6 0.6 0.6>
    componentLinearVelScale=<0.6 0.6 0.6>
    physicsBodyDefinitions [{
      sphereCollisionRadius=0
      boundBone {boneName=dyn_hairB_02}
      boxExtents=<12 16 15>
      constraintSetup {angularLimitsMax=<0 40 40>, angularLimitsMin=<0 -40 -15>, angularTarget=<0 0 0>}
      localJointOffset=<5 0 0>
     }]
   }
   n3 = Anim Dynamics - Bone: dyn_hairU_02(ComponentPose=@n2, Alpha=0.5, ExternalForce="0,0,0") {
    angularDampingOverride=0.995
    angularSpringConstant=100
    bAngularSpring=true
    bGravityOverrideInSimSpace=true
    bOverrideAngularDamping=true
    gravityScale=0.31
    simulationSpace=RootRelative
    alphaBoolBlend {blendOption=Linear}
    boundBone {boneName=dyn_hairU_02}
    componentAppliedLinearAccClamp=<3000 3000 1300>
    componentLinearAccScale=<0.6 0.6 0.6>
    componentLinearVelScale=<0.6 0.6 0.6>
    physicsBodyDefinitions [{
      sphereCollisionRadius=0
      boundBone {boneName=dyn_hairU_02}
      boxExtents=<12 12 22>
      constraintSetup {angularLimitsMax=<0 20 40>, angularLimitsMin=<0 -18 -5>, angularTarget=<0 0 0>}
      localJointOffset=<4 0 -2>
     }]
   }
   n4 = Anim Dynamics - Bone: dyn_hairII_02(ComponentPose=@n3, Alpha=0.5, ExternalForce="0,0,0") {
    angularDampingOverride=0.995
    angularSpringConstant=100
    bAngularSpring=true
    bGravityOverrideInSimSpace=true
    bOverrideAngularDamping=true
    gravityScale=0.35
    simulationSpace=RootRelative
    alphaBoolBlend {blendOption=Linear}
    boundBone {boneName=dyn_hairII_02}
    componentAppliedLinearAccClamp=<3000 3000 1300>
    componentLinearAccScale=<0.6 0.6 0.6>
    componentLinearVelScale=<0.6 0.6 0.6>
    physicsBodyDefinitions [{
      sphereCollisionRadius=0
      boundBone {boneName=dyn_hairII_02}
      boxExtents=<16 28 28>
      constraintSetup {angularLimitsMax=<0 40 40>, angularLimitsMin=<0 -40 -5>, angularTarget=<0 0 0>}
      localJointOffset=<9 0 0>
     }]
   }
   n5 = Anim Dynamics - Bone: dyn_hairP_02(ComponentPose=@n4, Alpha=0.6, ExternalForce="0,0,0") {
    angularDampingOverride=0.995
    angularSpringConstant=100
    bAngularSpring=true
    bGravityOverrideInSimSpace=true
    bOverrideAngularDamping=true
    gravityScale=0.35
    simulationSpace=RootRelative
    alphaBoolBlend {blendOption=Linear}
    boundBone {boneName=dyn_hairP_02}
    componentAppliedLinearAccClamp=<3000 3000 1300>
    componentLinearAccScale=<0.6 0.6 0.6>
    componentLinearVelScale=<0.6 0.6 0.6>
    physicsBodyDefinitions [{
      sphereCollisionRadius=0
      boundBone {boneName=dyn_hairP_02}
      boxExtents=<18 23 22>
      constraintSetup {angularLimitsMax=<0 40 40>, angularLimitsMin=<0 -40 -5>, angularTarget=<0 0 0>}
      localJointOffset=<6 0 0>
     }]
   }
   n6 = Anim Dynamics - Bone: dyn_hairM_02(ComponentPose=@n5, Alpha=0.45, ExternalForce="0,0,0") {
    angularDampingOverride=0.995
    angularSpringConstant=100
    bAngularSpring=true
    bGravityOverrideInSimSpace=true
    bOverrideAngularDamping=true
    gravityScale=0.35
    simulationSpace=RootRelative
    alphaBoolBlend {blendOption=Linear}
    boundBone {boneName=dyn_hairM_02}
    componentAppliedLinearAccClamp=<3000 3000 1300>
    componentLinearAccScale=<0.6 0.6 0.6>
    componentLinearVelScale=<0.6 0.6 0.6>
    physicsBodyDefinitions [{
      sphereCollisionRadius=0
      boundBone {boneName=dyn_hairM_02}
      boxExtents=<12 18 18>
      constraintSetup {angularLimitsMax=<0 20 40>, angularLimitsMin=<0 -18 -5>, angularTarget=<0 0 0>}
      localJointOffset=<6 0 0>
     }]
   }
   n7 = Anim Dynamics - Bone: dyn_hairO_01(ComponentPose=@n6, Alpha=0.3, ExternalForce="0,0,0") {
    angularDampingOverride=0.995
    angularSpringConstant=100
    bAngularSpring=true
    bGravityOverrideInSimSpace=true
    bOverrideAngularDamping=true
    gravityScale=0.35
    simulationSpace=RootRelative
    alphaBoolBlend {blendOption=Linear}
    boundBone {boneName=dyn_hairO_01}
    componentAppliedLinearAccClamp=<3000 3000 1300>
    componentLinearAccScale=<0.6 0.6 0.6>
    componentLinearVelScale=<0.6 0.6 0.6>
    physicsBodyDefinitions [{
      sphereCollisionRadius=0
      boundBone {boneName=dyn_hairO_01}
      boxExtents=<5 15 28>
      constraintSetup {angularLimitsMax=<0 40 40>, angularLimitsMin=<0 -40 -5>, angularTarget=<1 0 0>}
      localJointOffset=<-0.5 -0.2 0.1>
     }]
   }
   n8 = Anim Dynamics - Bone: dyn_hairO_02(ComponentPose=@n7, Alpha=0.7, ExternalForce="0,0,0") {
    angularDampingOverride=0.995
    angularSpringConstant=100
    bAngularSpring=true
    bGravityOverrideInSimSpace=true
    bOverrideAngularDamping=true
    gravityScale=0.35
    simulationSpace=RootRelative
    alphaBoolBlend {blendOption=Linear}
    boundBone {boneName=dyn_hairO_02}
    componentAppliedLinearAccClamp=<3000 3000 1300>
    componentLinearAccScale=<0.6 0.6 0.6>
    componentLinearVelScale=<0.6 0.6 0.6>
    physicsBodyDefinitions [{
      sphereCollisionRadius=0
      boundBone {boneName=dyn_hairO_02}
      boxExtents=<7 18 18>
      constraintSetup {angularLimitsMax=<0 40 40>, angularLimitsMin=<0 -40 -5>, angularTarget=<0 0 0>}
      localJointOffset=<6 0 0>
     }]
   }
   n9 = Anim Dynamics - Bone: dyn_hairJJ_02(ComponentPose=@n8, Alpha=0.55, ExternalForce="0,0,0") {
    angularDampingOverride=0.995
    angularSpringConstant=100
    bAngularSpring=true
    bGravityOverrideInSimSpace=true
    bOverrideAngularDamping=true
    gravityScale=0.31
    simulationSpace=RootRelative
    alphaBoolBlend {blendOption=Linear}
    boundBone {boneName=dyn_hairJJ_02}
    componentAppliedLinearAccClamp=<3000 3000 1300>
    componentLinearAccScale=<0.6 0.6 0.6>
    componentLinearVelScale=<0.6 0.6 0.6>
    physicsBodyDefinitions [{
      sphereCollisionRadius=0
      boundBone {boneName=dyn_hairJJ_02}
      boxExtents=<8 12 18>
      constraintSetup {angularLimitsMax=<0 60 40>, angularLimitsMin=<0 -18 -5>, angularTarget=<0 0 0>}
      localJointOffset=<5 0 0>
     }]
   }
   n10 = Anim Dynamics - Bone: dyn_hairS_02(ComponentPose=@n9, Alpha=0.5, ExternalForce="0,0,0") {
    angularDampingOverride=0.995
    angularSpringConstant=100
    bAngularSpring=true
    bGravityOverrideInSimSpace=true
    bOverrideAngularDamping=true
    gravityScale=0.3
    simulationSpace=RootRelative
    alphaBoolBlend {blendOption=Linear}
    boundBone {boneName=dyn_hairS_02}
    componentAppliedLinearAccClamp=<3000 3000 1300>
    componentLinearAccScale=<0.6 0.6 0.6>
    componentLinearVelScale=<0.6 0.6 0.6>
    physicsBodyDefinitions [{
      sphereCollisionRadius=0
      boundBone {boneName=dyn_hairS_02}
      boxExtents=<12 15 25>
      constraintSetup {angularLimitsMax=<0 80 40>, angularLimitsMin=<0 -80 -5>, angularTarget=<0 0 0>}
      localJointOffset=<5 0 -0.25>
     }]
   }
   n11 = Anim Dynamics - Bone: dyn_hairD_02(ComponentPose=@n10, Alpha=0.6, ExternalForce="0,0,0") {
    angularDampingOverride=0.995
    angularSpringConstant=100
    bAngularSpring=true
    bGravityOverrideInSimSpace=true
    bOverrideAngularDamping=true
    gravityScale=0.35
    simulationSpace=RootRelative
    alphaBoolBlend {blendOption=Linear}
    boundBone {boneName=dyn_hairD_02}
    componentAppliedLinearAccClamp=<3000 3000 1300>
    componentLinearAccScale=<0.6 0.6 0.6>
    componentLinearVelScale=<0.6 0.6 0.6>
    physicsBodyDefinitions [{
      sphereCollisionRadius=0
      boundBone {boneName=dyn_hairD_02}
      boxExtents=<18 23 22>
      constraintSetup {angularLimitsMax=<0 40 40>, angularLimitsMin=<0 -40 -5>, angularTarget=<0 0 0>}
      localJointOffset=<6 0 0>
     }]
   }
   n12 = Anim Dynamics - Bone: dyn_hairHH_01(ComponentPose=@n11, Alpha=0.25, ExternalForce="0,0,0") {
    angularDampingOverride=0.995
    angularSpringConstant=100
    bAngularSpring=true
    bGravityOverrideInSimSpace=true
    bOverrideAngularDamping=true
    gravityScale=0.3
    simulationSpace=RootRelative
    alphaBoolBlend {blendOption=Linear}
    boundBone {boneName=dyn_hairHH_01}
    componentAppliedLinearAccClamp=<3000 3000 1300>
    componentLinearAccScale=<0.6 0.6 0.6>
    componentLinearVelScale=<0.6 0.6 0.6>
    physicsBodyDefinitions [{
      sphereCollisionRadius=0
      boundBone {boneName=dyn_hairHH_01}
      boxExtents=<5 18 10>
      constraintSetup {angularLimitsMax=<0 45 40>, angularLimitsMin=<0 -25 -5>, angularTarget=<0 0 0>}
      localJointOffset=<-1 0 0>
     }]
   }
   n13 = Anim Dynamics - Bone: dyn_hairHH_02(ComponentPose=@n12, Alpha=0.6, ExternalForce="0,0,0") {
    angularDampingOverride=0.995
    angularSpringConstant=100
    bAngularSpring=true
    bGravityOverrideInSimSpace=true
    bOverrideAngularDamping=true
    gravityScale=0.3
    simulationSpace=RootRelative
    alphaBoolBlend {blendOption=Linear}
    boundBone {boneName=dyn_hairHH_02}
    componentAppliedLinearAccClamp=<3000 3000 1300>
    componentLinearAccScale=<0.6 0.6 0.6>
    componentLinearVelScale=<0.6 0.6 0.6>
    physicsBodyDefinitions [{
      sphereCollisionRadius=0
      boundBone {boneName=dyn_hairHH_02}
      boxExtents=<15 18 15>
      constraintSetup {angularLimitsMax=<0 45 40>, angularLimitsMin=<0 -25 -25>, angularTarget=<0 0 0>}
      localJointOffset=<1 0 0.675>
     }]
   }
   n14 = Anim Dynamics - Bone: dyn_hairN_01(ComponentPose=@n13, Alpha=0.8, ExternalForce="0,0,0") {
    angularDampingOverride=0.995
    angularSpringConstant=100
    bAngularSpring=true
    bGravityOverrideInSimSpace=true
    bOverrideAngularDamping=true
    gravityScale=0.31
    simulationSpace=RootRelative
    alphaBoolBlend {blendOption=Linear}
    boundBone {boneName=dyn_hairN_01}
    componentAppliedLinearAccClamp=<3000 3000 1300>
    componentLinearAccScale=<0.6 0.6 0.6>
    componentLinearVelScale=<0.6 0.6 0.6>
    physicsBodyDefinitions [{
      sphereCollisionRadius=0
      boundBone {boneName=dyn_hairN_01}
      boxExtents=<8 15 15>
      constraintSetup {angularLimitsMax=<0 1 40>, angularLimitsMin=<0 -45 -15>, angularTarget=<0 0 0>}
      localJointOffset=<3 0 0>
     }]
   }
   n15 = Anim Dynamics - Bone: dyn_hairR_01(ComponentPose=@n14, Alpha=0.7, ExternalForce="0,0,0") {
    angularDampingOverride=0.995
    angularSpringConstant=100
    bAngularSpring=true
    bGravityOverrideInSimSpace=true
    bOverrideAngularDamping=true
    gravityScale=0.35
    simulationSpace=RootRelative
    alphaBoolBlend {blendOption=Linear}
    boundBone {boneName=dyn_hairR_01}
    componentAppliedLinearAccClamp=<3000 3000 1300>
    componentLinearAccScale=<0.6 0.6 0.6>
    componentLinearVelScale=<0.6 0.6 0.6>
    physicsBodyDefinitions [{
      sphereCollisionRadius=0
      boundBone {boneName=dyn_hairR_01}
      boxExtents=<18 25 18>
      constraintSetup {angularLimitsMax=<0 40 40>, angularLimitsMin=<0 -40 -5>, angularTarget=<0 0 0>}
      localJointOffset=<6 0 0>
     }]
   }
   n16 = Anim Dynamics - Bone: dyn_hairI_02(ComponentPose=@n15, Alpha=0.5, ExternalForce="0,0,0") {
    angularDampingOverride=0.995
    angularSpringConstant=100
    bAngularSpring=true
    bGravityOverrideInSimSpace=true
    bOverrideAngularDamping=true
    gravityScale=0.34
    simulationSpace=RootRelative
    alphaBoolBlend {blendOption=Linear}
    boundBone {boneName=dyn_hairI_02}
    componentAppliedLinearAccClamp=<3000 3000 1300>
    componentLinearAccScale=<0.6 0.6 0.6>
    componentLinearVelScale=<0.6 0.6 0.6>
    physicsBodyDefinitions [{
      sphereCollisionRadius=0
      boundBone {boneName=dyn_hairI_02}
      boxExtents=<15 23 23>
      constraintSetup {angularLimitsMax=<0 40 40>, angularLimitsMin=<0 -40 -15>, angularTarget=<0 0 0>}
      localJointOffset=<9 0 0>
     }]
   }
   n17 = Anim Dynamics - Bone: dyn_hairY_01(ComponentPose=@n16, Alpha=0.85, ExternalForce="0,0,0") {
    angularDampingOverride=0.995
    angularSpringConstant=100
    bAngularSpring=true
    bGravityOverrideInSimSpace=true
    bOverrideAngularDamping=true
    gravityScale=0.32
    simulationSpace=RootRelative
    alphaBoolBlend {blendOption=Linear}
    boundBone {boneName=dyn_hairY_01}
    componentAppliedLinearAccClamp=<3000 3000 1300>
    componentLinearAccScale=<0.6 0.6 0.6>
    componentLinearVelScale=<0.6 0.6 0.6>
    physicsBodyDefinitions [{
      sphereCollisionRadius=0
      boundBone {boneName=dyn_hairY_01}
      boxExtents=<16 22 20>
      constraintSetup {angularLimitsMax=<0 40 40>, angularLimitsMin=<0 -40 -15>, angularTarget=<0 0 0>}
      localJointOffset=<8 0 0>
     }]
   }
   n18 = Anim Dynamics - Bone: dyn_hairH_02(ComponentPose=@n17, Alpha=0.5, ExternalForce="0,0,0") {
    angularDampingOverride=0.995
    angularSpringConstant=100
    bAngularSpring=true
    bGravityOverrideInSimSpace=true
    bOverrideAngularDamping=true
    gravityScale=0.32
    simulationSpace=RootRelative
    alphaBoolBlend {blendOption=Linear}
    boundBone {boneName=dyn_hairH_02}
    componentAppliedLinearAccClamp=<3000 3000 1300>
    componentLinearAccScale=<0.6 0.6 0.6>
    componentLinearVelScale=<0.6 0.6 0.6>
    physicsBodyDefinitions [{
      sphereCollisionRadius=0
      boundBone {boneName=dyn_hairH_02}
      boxExtents=<18 25 25>
      constraintSetup {angularLimitsMax=<0 40 40>, angularLimitsMin=<0 -40 -5>, angularTarget=<0 0 0>}
      localJointOffset=<6 0 0>
     }]
   }
   n19 = Anim Dynamics - Bone: dyn_hairK_02(ComponentPose=@n18, Alpha=0.6, ExternalForce="0,0,0") {
    angularDampingOverride=0.995
    angularSpringConstant=100
    bAngularSpring=true
    bGravityOverrideInSimSpace=true
    bOverrideAngularDamping=true
    gravityScale=0.35
    simulationSpace=RootRelative
    alphaBoolBlend {blendOption=Linear}
    boundBone {boneName=dyn_hairK_02}
    componentAppliedLinearAccClamp=<3000 3000 1300>
    componentLinearAccScale=<0.6 0.6 0.6>
    componentLinearVelScale=<0.6 0.6 0.6>
    physicsBodyDefinitions [{
      sphereCollisionRadius=0
      boundBone {boneName=dyn_hairK_02}
      boxExtents=<12 18 28>
      constraintSetup {angularLimitsMax=<0 40 40>, angularLimitsMin=<0 -40 -5>, angularTarget=<0 0 0>}
      localJointOffset=<6 0 0>
     }]
   }
   n20 = Anim Dynamics - Bone: dyn_hairT_01(ComponentPose=@n19, Alpha=1, ExternalForce="0,0,0") {
    angularDampingOverride=0.995
    angularSpringConstant=100
    bAngularSpring=true
    bGravityOverrideInSimSpace=true
    bOverrideAngularDamping=true
    gravityScale=0.35
    simulationSpace=RootRelative
    alphaBoolBlend {blendOption=Linear}
    boundBone {boneName=dyn_hairT_01}
    componentAppliedLinearAccClamp=<3000 3000 1300>
    componentLinearAccScale=<0.6 0.6 0.6>
    componentLinearVelScale=<0.6 0.6 0.6>
    physicsBodyDefinitions [{
      sphereCollisionRadius=0
      boundBone {boneName=dyn_hairT_01}
      boxExtents=<10 15 15>
      constraintSetup {angularLimitsMax=<0 40 40>, angularLimitsMin=<0 -40 -15>, angularTarget=<0 0 0>}
      localJointOffset=<3 0 0>
     }]
   }
   n21 = Anim Dynamics - Bone: dyn_hairA_02(ComponentPose=@n20, Alpha=0.8, ExternalForce="0,0,0") {
    angularDampingOverride=0.995
    angularSpringConstant=100
    bAngularSpring=true
    bGravityOverrideInSimSpace=true
    bOverrideAngularDamping=true
    gravityScale=0.35
    simulationSpace=RootRelative
    alphaBoolBlend {blendOption=Linear}
    boundBone {boneName=dyn_hairA_02}
    componentAppliedLinearAccClamp=<3000 3000 1300>
    componentLinearAccScale=<0.6 0.6 0.6>
    componentLinearVelScale=<0.6 0.6 0.6>
    physicsBodyDefinitions [{
      sphereCollisionRadius=0
      boundBone {boneName=dyn_hairA_02}
      boxExtents=<8 25 25>
      constraintSetup {angularLimitsMax=<0 40 40>, angularLimitsMin=<0 -40 -15>, angularTarget=<0 0 0>}
      localJointOffset=<3 0 0>
     }]
   }
   n22 = Anim Dynamics - Bone: dyn_hairQ_01(ComponentPose=@n21, Alpha=0.25, ExternalForce="0,0,0") {
    angularDampingOverride=0.995
    angularSpringConstant=100
    bAngularSpring=true
    bGravityOverrideInSimSpace=true
    bOverrideAngularDamping=true
    gravityScale=0.25
    simulationSpace=RootRelative
    alphaBoolBlend {blendOption=Linear}
    boundBone {boneName=dyn_hairQ_01}
    componentAppliedLinearAccClamp=<3000 3000 1300>
    componentLinearAccScale=<0.6 0.6 0.6>
    componentLinearVelScale=<0.6 0.6 0.6>
    physicsBodyDefinitions [{
      sphereCollisionRadius=0
      boundBone {boneName=dyn_hairQ_01}
      boxExtents=<15 35 35>
      constraintSetup {angularLimitsMax=<0 45 40>, angularLimitsMin=<0 -25 -5>, angularTarget=<0 0 0>}
      localJointOffset=<5 0 0>
     }]
   }
   n23 = Anim Dynamics - Bone: dyn_hairQ_02(ComponentPose=@n22, Alpha=0.7, ExternalForce="0,0,0") {
    angularDampingOverride=0.995
    angularSpringConstant=100
    bAngularSpring=true
    bGravityOverrideInSimSpace=true
    bOverrideAngularDamping=true
    gravityScale=0.32
    simulationSpace=RootRelative
    alphaBoolBlend {blendOption=Linear}
    boundBone {boneName=dyn_hairQ_02}
    componentAppliedLinearAccClamp=<3000 3000 1300>
    componentLinearAccScale=<0.6 0.6 0.6>
    componentLinearVelScale=<0.6 0.6 0.6>
    physicsBodyDefinitions [{
      sphereCollisionRadius=0
      boundBone {boneName=dyn_hairQ_02}
      boxExtents=<10 22 16>
      constraintSetup {angularLimitsMax=<0 40 40>, angularLimitsMin=<0 -40 -5>, angularTarget=<0 0 0>}
      localJointOffset=<6 0 0>
     }]
   }
   n24 = Anim Dynamics - Bone: dyn_hairK_02(ComponentPose=@n23, Alpha=0.6, ExternalForce="0,0,0") {
    angularDampingOverride=0.995
    angularSpringConstant=100
    bAngularSpring=true
    bGravityOverrideInSimSpace=true
    bOverrideAngularDamping=true
    gravityScale=0.35
    simulationSpace=RootRelative
    alphaBoolBlend {blendOption=Linear}
    boundBone {boneName=dyn_hairK_02}
    componentAppliedLinearAccClamp=<3000 3000 1300>
    componentLinearAccScale=<0.6 0.6 0.6>
    componentLinearVelScale=<0.6 0.6 0.6>
    physicsBodyDefinitions [{
      sphereCollisionRadius=0
      boundBone {boneName=dyn_hairK_02}
      boxExtents=<12 18 28>
      constraintSetup {angularLimitsMax=<0 40 40>, angularLimitsMin=<0 -40 -5>, angularTarget=<0 0 0>}
      localJointOffset=<6 0 0>
     }]
   }
   n25 = Anim Dynamics - Bone: dyn_hairEE_02(ComponentPose=@n24, Alpha=1, ExternalForce="0,0,0") {
    angularDampingOverride=0.995
    angularSpringConstant=100
    bAngularSpring=true
    bGravityOverrideInSimSpace=true
    bOverrideAngularDamping=true
    gravityScale=0.32
    simulationSpace=RootRelative
    alphaBoolBlend {blendOption=Linear}
    boundBone {boneName=dyn_hairEE_02}
    componentAppliedLinearAccClamp=<3000 3000 1300>
    componentLinearAccScale=<0.6 0.6 0.6>
    componentLinearVelScale=<0.6 0.6 0.6>
    physicsBodyDefinitions [{
      sphereCollisionRadius=0
      boundBone {boneName=dyn_hairEE_02}
      boxExtents=<10 19 20>
      constraintSetup {angularLimitsMax=<0 40 40>, angularLimitsMin=<0 -40 -15>, angularTarget=<0 0 0>}
      localJointOffset=<3 0 0>
     }]
   }
   n26 = Anim Dynamics - Bone: dyn_hairDD_02(ComponentPose=@n25, Alpha=0.5, ExternalForce="0,0,0") {
    angularDampingOverride=0.995
    angularSpringConstant=100
    bAngularSpring=true
    bGravityOverrideInSimSpace=true
    bOverrideAngularDamping=true
    gravityScale=0.32
    simulationSpace=RootRelative
    alphaBoolBlend {blendOption=Linear}
    boundBone {boneName=dyn_hairDD_02}
    componentAppliedLinearAccClamp=<3000 3000 1300>
    componentLinearAccScale=<0.6 0.6 0.6>
    componentLinearVelScale=<0.6 0.6 0.6>
    physicsBodyDefinitions [{
      sphereCollisionRadius=0
      boundBone {boneName=dyn_hairDD_02}
      boxExtents=<18 25 25>
      constraintSetup {angularLimitsMax=<0 40 40>, angularLimitsMin=<0 -40 -5>, angularTarget=<0 0 0>}
      localJointOffset=<6 0 0>
     }]
   }
   n27 = Anim Dynamics - Bone: dyn_hairL_02(ComponentPose=@n26, Alpha=0.6, ExternalForce="0,0,0") {
    angularDampingOverride=0.995
    angularSpringConstant=100
    bAngularSpring=true
    bGravityOverrideInSimSpace=true
    bOverrideAngularDamping=true
    gravityScale=0.35
    simulationSpace=RootRelative
    alphaBoolBlend {blendOption=Linear}
    boundBone {boneName=dyn_hairL_02}
    componentAppliedLinearAccClamp=<3000 3000 1300>
    componentLinearAccScale=<0.6 0.6 0.6>
    componentLinearVelScale=<0.6 0.6 0.6>
    physicsBodyDefinitions [{
      sphereCollisionRadius=0
      boundBone {boneName=dyn_hairL_02}
      boxExtents=<12 18 28>
      constraintSetup {angularLimitsMax=<0 40 40>, angularLimitsMin=<0 -40 -5>, angularTarget=<0 0 0>}
      localJointOffset=<6 0 0>
     }]
   }
   n28 = Anim Dynamics - Bone: dyn_hairGG_02(ComponentPose=@n27, Alpha=1, ExternalForce="0,0,0") {
    angularDampingOverride=0.995
    angularSpringConstant=100
    bAngularSpring=true
    bGravityOverrideInSimSpace=true
    bOverrideAngularDamping=true
    gravityScale=0.35
    simulationSpace=RootRelative
    alphaBoolBlend {blendOption=Linear}
    boundBone {boneName=dyn_hairGG_02}
    componentAppliedLinearAccClamp=<3000 3000 1300>
    componentLinearAccScale=<0.6 0.6 0.6>
    componentLinearVelScale=<0.6 0.6 0.6>
    physicsBodyDefinitions [{
      sphereCollisionRadius=0
      boundBone {boneName=dyn_hairGG_02}
      boxExtents=<10 15 15>
      constraintSetup {angularLimitsMax=<0 40 40>, angularLimitsMin=<0 -40 -15>, angularTarget=<0 0 0>}
      localJointOffset=<3 0 0>
     }]
   }
   n29 = Anim Dynamics - Bone: dyn_hairV_01(ComponentPose=@n28, Alpha=0.8, ExternalForce="0,0,0") {
    angularDampingOverride=0.995
    angularSpringConstant=100
    bAngularSpring=true
    bGravityOverrideInSimSpace=true
    bOverrideAngularDamping=true
    gravityScale=0.35
    simulationSpace=RootRelative
    alphaBoolBlend {blendOption=Linear}
    boundBone {boneName=dyn_hairV_01}
    componentAppliedLinearAccClamp=<3000 3000 1300>
    componentLinearAccScale=<0.6 0.6 0.6>
    componentLinearVelScale=<0.6 0.6 0.6>
    physicsBodyDefinitions [{
      sphereCollisionRadius=0
      boundBone {boneName=dyn_hairV_01}
      boxExtents=<8 25 25>
      constraintSetup {angularLimitsMax=<0 40 40>, angularLimitsMin=<0 -40 -15>, angularTarget=<0 0 0>}
      localJointOffset=<3 0 0>
     }]
   }
   n30 = Anim Dynamics - Bone: dyn_hairF_02(ComponentPose=@n29, Alpha=0.7, ExternalForce="0,0,0") {
    angularDampingOverride=0.995
    angularSpringConstant=100
    bAngularSpring=true
    bGravityOverrideInSimSpace=true
    bOverrideAngularDamping=true
    gravityScale=0.32
    simulationSpace=RootRelative
    alphaBoolBlend {blendOption=Linear}
    boundBone {boneName=dyn_hairF_02}
    componentAppliedLinearAccClamp=<3000 3000 1300>
    componentLinearAccScale=<0.6 0.6 0.6>
    componentLinearVelScale=<0.6 0.6 0.6>
    physicsBodyDefinitions [{
      sphereCollisionRadius=0
      boundBone {boneName=dyn_hairF_02}
      boxExtents=<10 22 16>
      constraintSetup {angularLimitsMax=<0 40 40>, angularLimitsMin=<0 -40 -5>, angularTarget=<0 0 0>}
      localJointOffset=<6 0 0>
     }]
   }
   n31 = Anim Dynamics - Bone: dyn_hairC_02(ComponentPose=@n30, Alpha=0.6, ExternalForce="0,0,0") {
    angularDampingOverride=0.995
    angularSpringConstant=100
    bAngularSpring=true
    bGravityOverrideInSimSpace=true
    bOverrideAngularDamping=true
    gravityScale=0.31
    simulationSpace=RootRelative
    alphaBoolBlend {blendOption=Linear}
    boundBone {boneName=dyn_hairC_02}
    componentAppliedLinearAccClamp=<3000 3000 1300>
    componentLinearAccScale=<0.6 0.6 0.6>
    componentLinearVelScale=<0.6 0.6 0.6>
    physicsBodyDefinitions [{
      sphereCollisionRadius=0
      boundBone {boneName=dyn_hairC_02}
      boxExtents=<12 25 28>
      constraintSetup {angularLimitsMax=<0 40 40>, angularLimitsMin=<0 -40 -5>, angularTarget=<0 0 0>}
      localJointOffset=<6 0 0>
     }]
   }
   n32 = Anim Dynamics - Bone: dyn_hairJ_01(ComponentPose=@n31, Alpha=0.6, ExternalForce="0,0,0") {
    angularDampingOverride=0.995
    angularSpringConstant=100
    bAngularSpring=true
    bGravityOverrideInSimSpace=true
    bOverrideAngularDamping=true
    gravityScale=0.35
    simulationSpace=RootRelative
    alphaBoolBlend {blendOption=Linear}
    boundBone {boneName=dyn_hairJ_01}
    componentAppliedLinearAccClamp=<3000 3000 1300>
    componentLinearAccScale=<0.6 0.6 0.6>
    componentLinearVelScale=<0.6 0.6 0.6>
    physicsBodyDefinitions [{
      sphereCollisionRadius=0
      boundBone {boneName=dyn_hairJ_01}
      boxExtents=<18 25 30>
      constraintSetup {angularLimitsMax=<0 40 40>, angularLimitsMin=<0 -40 -5>, angularTarget=<0 0 0>}
      localJointOffset=<6 0 0>
     }]
   }
   n33 = Anim Dynamics - Bone: dyn_hairZ_02(ComponentPose=@n32, Alpha=0.5, ExternalForce="0,0,0") {
    angularDampingOverride=0.995
    angularSpringConstant=100
    bAngularSpring=true
    bGravityOverrideInSimSpace=true
    bOverrideAngularDamping=true
    gravityScale=0.34
    simulationSpace=RootRelative
    alphaBoolBlend {blendOption=Linear}
    boundBone {boneName=dyn_hairZ_02}
    componentAppliedLinearAccClamp=<3000 3000 1300>
    componentLinearAccScale=<0.6 0.6 0.6>
    componentLinearVelScale=<0.6 0.6 0.6>
    physicsBodyDefinitions [{
      sphereCollisionRadius=0
      boundBone {boneName=dyn_hairZ_02}
      boxExtents=<15 23 23>
      constraintSetup {angularLimitsMax=<0 40 40>, angularLimitsMin=<0 -40 -15>, angularTarget=<0 0 0>}
      localJointOffset=<9 0 0>
     }]
   }
   n34 = Anim Dynamics - Bone: dyn_hairAA_01(ComponentPose=@n33, Alpha=0.85, ExternalForce="0,0,0") {
    angularDampingOverride=0.995
    angularSpringConstant=100
    bAngularSpring=true
    bGravityOverrideInSimSpace=true
    bOverrideAngularDamping=true
    gravityScale=0.32
    simulationSpace=RootRelative
    alphaBoolBlend {blendOption=Linear}
    boundBone {boneName=dyn_hairAA_01}
    componentAppliedLinearAccClamp=<3000 3000 1300>
    componentLinearAccScale=<0.6 0.6 0.6>
    componentLinearVelScale=<0.6 0.6 0.6>
    physicsBodyDefinitions [{
      sphereCollisionRadius=0
      boundBone {boneName=dyn_hairAA_01}
      boxExtents=<16 22 20>
      constraintSetup {angularLimitsMax=<0 40 40>, angularLimitsMin=<0 -40 -15>, angularTarget=<0 0 0>}
      localJointOffset=<8 0 0>
     }]
   }
   n35 = Anim Dynamics - Bone: dyn_hairG_01(ComponentPose=@n34, Alpha=0.85, ExternalForce="0,0,0") {
    angularDampingOverride=0.995
    angularSpringConstant=100
    bAngularSpring=true
    bGravityOverrideInSimSpace=true
    bOverrideAngularDamping=true
    gravityScale=0.34
    simulationSpace=RootRelative
    alphaBoolBlend {blendOption=Linear}
    boundBone {boneName=dyn_hairG_01}
    componentAppliedLinearAccClamp=<3000 3000 1300>
    componentLinearAccScale=<0.6 0.6 0.6>
    componentLinearVelScale=<0.6 0.6 0.6>
    physicsBodyDefinitions [{
      sphereCollisionRadius=0
      boundBone {boneName=dyn_hairG_01}
      boxExtents=<15 23 23>
      constraintSetup {angularLimitsMax=<0 40 40>, angularLimitsMin=<0 -40 -15>, angularTarget=<0 0 0>}
      localJointOffset=<9 0 0>
     }]
   }
   n36 = Anim Dynamics - Bone: dyn_hairW_02(ComponentPose=@n35, Alpha=0.85, ExternalForce="0,0,0") {
    angularDampingOverride=0.995
    angularSpringConstant=100
    bAngularSpring=true
    bGravityOverrideInSimSpace=true
    bOverrideAngularDamping=true
    gravityScale=0.32
    simulationSpace=RootRelative
    alphaBoolBlend {blendOption=Linear}
    boundBone {boneName=dyn_hairW_02}
    componentAppliedLinearAccClamp=<3000 3000 1300>
    componentLinearAccScale=<0.6 0.6 0.6>
    componentLinearVelScale=<0.6 0.6 0.6>
    physicsBodyDefinitions [{
      sphereCollisionRadius=0
      boundBone {boneName=dyn_hairW_02}
      boxExtents=<16 22 20>
      constraintSetup {angularLimitsMax=<0 40 40>, angularLimitsMin=<0 -40 -15>, angularTarget=<0 0 0>}
      localJointOffset=<8 0 0>
     }]
   }
   n37 = Anim Dynamics - Bone: dyn_hairBB_02(ComponentPose=@n36, Alpha=0.5, ExternalForce="0,0,0") {
    angularDampingOverride=0.995
    angularSpringConstant=100
    bAngularSpring=true
    bGravityOverrideInSimSpace=true
    bOverrideAngularDamping=true
    gravityScale=0.32
    simulationSpace=RootRelative
    alphaBoolBlend {blendOption=Linear}
    boundBone {boneName=dyn_hairBB_02}
    componentAppliedLinearAccClamp=<3000 3000 1300>
    componentLinearAccScale=<0.6 0.6 0.6>
    componentLinearVelScale=<0.6 0.6 0.6>
    physicsBodyDefinitions [{
      sphereCollisionRadius=0
      boundBone {boneName=dyn_hairBB_02}
      boxExtents=<18 25 25>
      constraintSetup {angularLimitsMax=<0 40 40>, angularLimitsMin=<0 -40 -5>, angularTarget=<0 0 0>}
      localJointOffset=<6 0 0>
     }]
   }
   n38 = Anim Dynamics - Bone: dyn_hairX_02(ComponentPose=@n37, Alpha=0.5, ExternalForce="0,0,0") {
    angularDampingOverride=0.995
    angularSpringConstant=100
    bAngularSpring=true
    bGravityOverrideInSimSpace=true
    bOverrideAngularDamping=true
    gravityScale=0.3
    simulationSpace=RootRelative
    alphaBoolBlend {blendOption=Linear}
    boundBone {boneName=dyn_hairX_02}
    componentAppliedLinearAccClamp=<3000 3000 1300>
    componentLinearAccScale=<0.6 0.6 0.6>
    componentLinearVelScale=<0.6 0.6 0.6>
    physicsBodyDefinitions [{
      sphereCollisionRadius=0
      boundBone {boneName=dyn_hairX_02}
      boxExtents=<12 15 25>
      constraintSetup {angularLimitsMax=<0 80 40>, angularLimitsMin=<0 -80 -5>, angularTarget=<0 0 0>}
      localJointOffset=<5 0 -0.25>
     }]
   }
   n39 = Anim Dynamics - Bone: dyn_hairFF_02(ComponentPose=@n38, Alpha=0.85, ExternalForce="0,0,0") {
    angularDampingOverride=0.995
    angularSpringConstant=100
    bAngularSpring=true
    bGravityOverrideInSimSpace=true
    bOverrideAngularDamping=true
    gravityScale=0.32
    simulationSpace=RootRelative
    alphaBoolBlend {blendOption=Linear}
    boundBone {boneName=dyn_hairFF_02}
    componentAppliedLinearAccClamp=<3000 3000 1300>
    componentLinearAccScale=<0.6 0.6 0.6>
    componentLinearVelScale=<0.6 0.6 0.6>
    physicsBodyDefinitions [{
      sphereCollisionRadius=0
      boundBone {boneName=dyn_hairFF_02}
      boxExtents=<4 6 12>
      constraintSetup {angularLimitsMax=<0 40 40>, angularLimitsMin=<0 -40 -15>, angularTarget=<0 0 0>}
      localJointOffset=<0.5 0.2 0>
     }]
   }
   n40 = Anim Dynamics - Bone: dyn_hairE_02(ComponentPose=@n39, Alpha=0.5, ExternalForce="0,0,0") {
    angularDampingOverride=0.995
    angularSpringConstant=100
    bAngularSpring=true
    bGravityOverrideInSimSpace=true
    bOverrideAngularDamping=true
    gravityScale=0.32
    simulationSpace=RootRelative
    alphaBoolBlend {blendOption=Linear}
    boundBone {boneName=dyn_hairE_02}
    componentAppliedLinearAccClamp=<3000 3000 1300>
    componentLinearAccScale=<0.6 0.6 0.6>
    componentLinearVelScale=<0.6 0.6 0.6>
    physicsBodyDefinitions [{
      sphereCollisionRadius=0
      boundBone {boneName=dyn_hairE_02}
      boxExtents=<18 25 25>
      constraintSetup {angularLimitsMax=<0 40 40>, angularLimitsMin=<0 -40 -1>, angularTarget=<0 0 0>}
      localJointOffset=<6 0 0>
     }]
   }
   Output Pose(Result=Component To Local(ComponentPose=@n40)) {layerGroup=None}
  ]}
 EventGraph {nodes [
   TryGetPawnOwner()
   Event BlueprintUpdateAnimation()
  ]}
]
variables []
